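# Physics asset for the coat on the M_MED_ChoiceCrave male body cosmetic, simulating its dangling panels with capsule bodies.
# Four mirrored bone chains per side hang from base_coat anchors: chains 01 and 02 swing on one axis only, while 03 and 05 allow limited swing and twist.
# Angularly locked cross-constraints tie 01 to 02 and 03 to 05 near their lower ends.
# Pelvis, thigh, calf and oversized floor capsules provide collision shapes.
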PhysicsAsset /BRCosmetics/Characters/Player/Male/Medium/Bodies/M_MED_ChoiceCrave/Meshes/M_MED_ChoiceCrave_Coat_Physics
bodies {
 base_coat_01_lf_01 [{shapeType=Capsule, length=0.505, radius=0.505, shapeName=base_coat_01_lf_01_capsule, rotation=<rot 0 -90 0>}]
 base_coat_01_rt_01 [{shapeType=Capsule, length=0.505, radius=0.505, shapeName=base_coat_01_rt_01_capsule, rotation=<rot 0 -90 0>}]
 base_coat_02_lf_01 [{shapeType=Capsule, length=0.505, radius=0.505, shapeName=base_coat_02_lf_01_capsule, rotation=<rot 0 -90 0>}]
 base_coat_02_rt_01 [{shapeType=Capsule, length=0.505, radius=0.505, shapeName=base_coat_02_rt_01_capsule, rotation=<rot 0 -90 0>}]
 base_coat_03_lf_01 [{shapeType=Capsule, length=0.505, radius=0.505, shapeName=base_coat_03_lf_01_capsule, rotation=<rot 0 -90 0>}]
 base_coat_03_rt_01 [{shapeType=Capsule, length=0.505, radius=0.505, shapeName=base_coat_03_rt_01_capsule, rotation=<rot 0 -90 0>}]
 base_coat_05_lf_01 [{shapeType=Capsule, length=0.505, radius=0.505, shapeName=base_coat_05_lf_01_capsule, rotation=<rot 0 -90 0>}]
 base_coat_05_rt_01 [{shapeType=Capsule, length=0.505, radius=0.505, shapeName=base_coat_05_rt_01_capsule, rotation=<rot 0 -90 0>}]
 calf_l [{shapeType=Capsule, length=35.6721, radius=7.5325, center=<-35.2778 -0.6578 0>, rotation=<rot -0.0461 87.6516 -89.7388>}]
 calf_r [{shapeType=Capsule, length=35.6721, radius=7.5325, center=<35.18 -0.6578 0>, rotation=<rot -0.0461 87.6516 -89.7388>}]
 dyn_coat_01_lf_01 [{shapeType=Capsule, length=10, radius=3, shapeName=dyn_coat_01_lf_01_capsule, center=<2 5.1396 -1.3335>, rotation=<rot 90 0 0>}]
 dyn_coat_01_lf_02 [{shapeType=Capsule, length=10, radius=3, shapeName=dyn_coat_01_lf_02_capsule, center=<7 4.9924 -1.3394>, rotation=<rot 90 0 0>}]
 dyn_coat_01_lf_03 [{shapeType=Capsule, length=10, radius=3, shapeName=dyn_coat_01_lf_03_capsule, center=<12 3 0>, rotation=<rot 90 0 0>}]
 dyn_coat_01_lf_04 [{shapeType=Capsule, length=10, radius=3, shapeName=dyn_coat_01_lf_04_capsule, center=<9 3 0>, rotation=<rot 90 0 0>}]
 dyn_coat_01_lf_05 [{shapeType=Capsule, length=10, radius=3, shapeName=dyn_coat_01_lf_05_capsule, center=<7 3 0>, rotation=<rot 90 0 0>}]
 dyn_coat_01_lf_06 [{shapeType=Capsule, length=10, radius=3, shapeName=dyn_coat_01_lf_06_capsule, center=<8 3 0>, rotation=<rot 90 0 0>}]
 dyn_coat_01_rt_01 [{shapeType=Capsule, length=10, radius=3, shapeName=dyn_coat_01_rt_01_capsule, center=<2 -5.1396 -1.3335>, rotation=<rot 90 0 0>}]
 dyn_coat_01_rt_02 [{shapeType=Capsule, length=10, radius=3, shapeName=dyn_coat_01_rt_02_capsule, center=<7 -4.9924 -1.3394>, rotation=<rot 90 0 0>}]
 dyn_coat_01_rt_03 [{shapeType=Capsule, length=10, radius=3, shapeName=dyn_coat_01_rt_03_capsule, center=<12 -3 0>, rotation=<rot 90 0 0>}]
 dyn_coat_01_rt_04 [{shapeType=Capsule, length=10, radius=3, shapeName=dyn_coat_01_rt_04_capsule, center=<9 -3 0>, rotation=<rot 90 0 0>}]
 dyn_coat_01_rt_05 [{shapeType=Capsule, length=10, radius=3, shapeName=dyn_coat_01_rt_05_capsule, center=<7 -3 0>, rotation=<rot 90 0 0>}]
 dyn_coat_01_rt_06 [{shapeType=Capsule, length=10, radius=3, shapeName=dyn_coat_01_rt_06_capsule, center=<8 -3 0>, rotation=<rot 90 0 0>}]
 dyn_coat_02_lf_01 [{shapeType=Capsule, length=10, radius=3, shapeName=dyn_coat_02_lf_01_capsule, center=<2 0 2.5>, rotation=<rot -82 0 0>}]
 dyn_coat_02_lf_02 [{shapeType=Capsule, length=10, radius=3, shapeName=dyn_coat_02_lf_02_capsule, center=<7 0 3>, rotation=<rot 88.5 0 0>}]
 dyn_coat_02_lf_03 [{shapeType=Capsule, length=10, radius=3, shapeName=dyn_coat_02_lf_03_capsule, center=<12 0 2>, rotation=<rot 90 0 0>}]
 dyn_coat_02_lf_04 [{shapeType=Capsule, length=10, radius=3, shapeName=dyn_coat_02_lf_04_capsule, center=<9 0 2>, rotation=<rot 90 0 0>}]
 dyn_coat_02_lf_05 [{shapeType=Capsule, length=23, radius=3, shapeName=dyn_coat_02_lf_05_capsule, center=<16 0 2>, rotation=<rot 90 0 0>}]
 dyn_coat_02_rt_01 [{shapeType=Capsule, length=10, radius=3, shapeName=dyn_coat_02_rt_01_capsule, center=<2 0 2.5>, rotation=<rot -82 0 0>}]
 dyn_coat_02_rt_02 [{shapeType=Capsule, length=10, radius=3, shapeName=dyn_coat_02_rt_02_capsule, center=<7 0 3>, rotation=<rot 88.5 0 0>}]
 dyn_coat_02_rt_03 [{shapeType=Capsule, length=10, radius=3, shapeName=dyn_coat_02_rt_03_capsule, center=<12 0 2>, rotation=<rot 90 0 0>}]
 dyn_coat_02_rt_04 [{shapeType=Capsule, length=10, radius=3, shapeName=dyn_coat_02_rt_04_capsule, center=<9 0 2>, rotation=<rot 90 0 0>}]
 dyn_coat_02_rt_05 [{shapeType=Capsule, length=23, radius=3, shapeName=dyn_coat_02_rt_05_capsule, center=<16 0 2>, rotation=<rot 90 0 0>}]
 dyn_coat_03_lf_01 [{shapeType=Capsule, length=12, radius=3, shapeName=dyn_coat_03_lf_01_capsule, center=<2 -1.32 6.31>, rotation=<rot 0 0 145>}]
 dyn_coat_03_lf_02 [{shapeType=Capsule, length=11.0629, radius=3, shapeName=dyn_coat_03_lf_02_capsule, center=<8 -1.32 6.31>, rotation=<rot 0 0 145>}]
 dyn_coat_03_lf_03 [{shapeType=Capsule, length=8.1989, radius=4.8262, shapeName=dyn_coat_03_lf_03_capsule, center=<8 -1.32 6.31>, rotation=<rot 0 0 145>}]
 dyn_coat_03_lf_04 [{shapeType=Capsule, length=7.0893, radius=5.9358, shapeName=dyn_coat_03_lf_04_capsule, center=<8 -1.32 6.31>, rotation=<rot 0 0 145>}]
 dyn_coat_03_rt_01 [{shapeType=Capsule, length=12, radius=3, shapeName=dyn_coat_03_rt_01_capsule, center=<2 1.32 6.31>, rotation=<rot 0 0 -145>}]
 dyn_coat_03_rt_02 [{shapeType=Capsule, length=11.0629, radius=3, shapeName=dyn_coat_03_rt_02_capsule, center=<8 1.32 6.31>, rotation=<rot 0 0 -145>}]
 dyn_coat_03_rt_03 [{shapeType=Capsule, length=8.1989, radius=4.8262, shapeName=dyn_coat_03_rt_03_capsule, center=<8 1.32 6.31>, rotation=<rot 0 0 -145>}]
 dyn_coat_03_rt_04 [{shapeType=Capsule, length=7.0893, radius=5.9358, shapeName=dyn_coat_03_rt_04_capsule, center=<8 1.32 6.31>, rotation=<rot 0 0 -145>}]
 dyn_coat_05_lf_01 [{shapeType=Capsule, length=10, radius=3, shapeName=dyn_coat_05_lf_01_capsule, center=<3 1 2.65>, rotation=<rot 0 0 85>}]
 dyn_coat_05_lf_02 [{shapeType=Capsule, length=10, radius=3, shapeName=dyn_coat_05_lf_02_capsule, center=<8 1 2>, rotation=<rot 0 0 85>}]
 dyn_coat_05_lf_03 [{shapeType=Capsule, length=7.1989, radius=4.8262, shapeName=dyn_coat_05_lf_03_capsule, center=<8 1 0>, rotation=<rot 0 0 85>}]
 dyn_coat_05_lf_04 [{shapeType=Capsule, length=9.0893, radius=5.9358, shapeName=dyn_coat_05_lf_04_capsule, center=<8 1 0>, rotation=<rot 0 0 85>}]
 dyn_coat_05_lf_05 [{
   shapeType=Capsule
   length=15.3433
   radius=9.6837
   shapeName=dyn_coat_05_lf_05_capsule
   center=<14.972 7.8354 -1.9526>
   rotation=<rot -0.7022 3.1625 101.6324>
  }]
 dyn_coat_05_rt_01 [{shapeType=Capsule, length=10, radius=3, shapeName=dyn_coat_05_rt_01_capsule, center=<3 -1 2.65>, rotation=<rot 0 0 -85>}]
 dyn_coat_05_rt_02 [{shapeType=Capsule, length=10, radius=3.5, shapeName=dyn_coat_05_rt_02_capsule, center=<8 -1 2>, rotation=<rot 0 0 95>}]
 dyn_coat_05_rt_03 [{shapeType=Capsule, length=7.1989, radius=4.8262, shapeName=dyn_coat_05_rt_03_capsule, center=<8 -1 0>, rotation=<rot 0 0 95>}]
 dyn_coat_05_rt_04 [{shapeType=Capsule, length=9.0893, radius=5.9358, shapeName=dyn_coat_05_rt_04_capsule, center=<8 -1 0>, rotation=<rot 0 0 95>}]
 dyn_coat_05_rt_05 [{
   shapeType=Capsule
   length=15.3433
   radius=9.6837
   shapeName=dyn_coat_05_rt_05_capsule
   center=<14.972 -7.8354 -1.9526>
   rotation=<rot 0.7022 3.1625 -101.6324>
  }]
 floor [{shapeType=Capsule, length=5, radius=510, shapeName=floor_capsule, center=<0 0 -505>}]
 pelvis [{shapeType=Capsule, length=6, radius=12.5, center=<-2.933 -0.593 0>, rotation=<rot 0 -90 0>}]
 thigh_l [{shapeType=Capsule, length=41.0072, radius=6.9563, center=<-23.0658 -0.0006 0.4587>, rotation=<rot -0.2776 -91.285 -88.281>}]
 thigh_r [{shapeType=Capsule, length=41.0072, radius=6.9563, center=<23.0658 0.0006 0.4587>, rotation=<rot -0.2776 -91.285 -88.281>}]
}
constraints [
 {bone1Name=dyn_coat_03_lf_01, bone2Name=base_coat_03_lf_01, swing1LimitDegrees=10, swing2LimitDegrees=17, twistLimitDegrees=5}
 {bone1Name=dyn_coat_03_lf_02, bone2Name=dyn_coat_03_lf_01, swing1LimitDegrees=10, swing2LimitDegrees=17, twistLimitDegrees=5}
 {bone1Name=dyn_coat_03_lf_03, bone2Name=dyn_coat_03_lf_02, swing1LimitDegrees=10, swing2LimitDegrees=17, twistLimitDegrees=5}
 {bone1Name=dyn_coat_03_lf_04, bone2Name=dyn_coat_03_lf_03, swing1LimitDegrees=10, swing2LimitDegrees=17, twistLimitDegrees=5}
 {bone1Name=dyn_coat_05_lf_01, bone2Name=base_coat_05_lf_01, swing1LimitDegrees=10, swing2LimitDegrees=17, twistLimitDegrees=5}
 {bone1Name=dyn_coat_05_lf_02, bone2Name=dyn_coat_05_lf_01, swing1LimitDegrees=10, swing2LimitDegrees=17, twistLimitDegrees=5}
 {bone1Name=dyn_coat_05_lf_03, bone2Name=dyn_coat_05_lf_02, swing1LimitDegrees=10, swing2LimitDegrees=17, twistLimitDegrees=5}
 {bone1Name=dyn_coat_05_lf_04, bone2Name=dyn_coat_05_lf_03, swing1LimitDegrees=10, swing2LimitDegrees=17, twistLimitDegrees=5}
 {bone1Name=dyn_coat_05_lf_05, bone2Name=dyn_coat_05_lf_04, swing1LimitDegrees=10, swing2LimitDegrees=17, twistLimitDegrees=5}
 {bone1Name=dyn_coat_03_rt_01, bone2Name=base_coat_03_rt_01, swing1LimitDegrees=10, swing2LimitDegrees=17, twistLimitDegrees=5}
 {bone1Name=dyn_coat_03_rt_02, bone2Name=dyn_coat_03_rt_01, swing1LimitDegrees=10, swing2LimitDegrees=17, twistLimitDegrees=5}
 {bone1Name=dyn_coat_03_rt_03, bone2Name=dyn_coat_03_rt_02, swing1LimitDegrees=10, swing2LimitDegrees=17, twistLimitDegrees=5}
 {bone1Name=dyn_coat_03_rt_04, bone2Name=dyn_coat_03_rt_03, swing1LimitDegrees=10, swing2LimitDegrees=17, twistLimitDegrees=5}
 {bone1Name=dyn_coat_05_rt_01, bone2Name=base_coat_05_rt_01, swing1LimitDegrees=10, swing2LimitDegrees=17, twistLimitDegrees=5}
 {bone1Name=dyn_coat_05_rt_02, bone2Name=dyn_coat_05_rt_01, swing1LimitDegrees=10, swing2LimitDegrees=17, twistLimitDegrees=5}
 {bone1Name=dyn_coat_05_rt_03, bone2Name=dyn_coat_05_rt_02, swing1LimitDegrees=10, swing2LimitDegrees=17, twistLimitDegrees=5}
 {bone1Name=dyn_coat_05_rt_04, bone2Name=dyn_coat_05_rt_03, swing1LimitDegrees=10, swing2LimitDegrees=17, twistLimitDegrees=5}
 {bone1Name=dyn_coat_05_rt_05, bone2Name=dyn_coat_05_rt_04, swing1LimitDegrees=10, swing2LimitDegrees=17, twistLimitDegrees=5}
 {
  bone1Name=dyn_coat_01_lf_01
  bone2Name=base_coat_01_lf_01
  swing1LimitDegrees=0
  swing1Motion=Locked
  swing2LimitDegrees=15
  twistLimitDegrees=0
  twistMotion=Locked
 }
 {
  bone1Name=dyn_coat_01_lf_02
  bone2Name=dyn_coat_01_lf_01
  swing1LimitDegrees=0
  swing1Motion=Locked
  swing2LimitDegrees=15
  twistLimitDegrees=0
  twistMotion=Locked
 }
 {
  bone1Name=dyn_coat_01_lf_04
  bone2Name=dyn_coat_01_lf_03
  swing1LimitDegrees=0
  swing1Motion=Locked
  swing2LimitDegrees=15
  twistLimitDegrees=0
  twistMotion=Locked
 }
 {
  bone1Name=dyn_coat_01_lf_03
  bone2Name=dyn_coat_01_lf_02
  swing1LimitDegrees=0
  swing1Motion=Locked
  swing2LimitDegrees=15
  twistLimitDegrees=0
  twistMotion=Locked
 }
 {
  bone1Name=dyn_coat_01_lf_05
  bone2Name=dyn_coat_01_lf_04
  swing1LimitDegrees=0
  swing1Motion=Locked
  swing2LimitDegrees=15
  twistLimitDegrees=0
  twistMotion=Locked
 }
 {
  bone1Name=dyn_coat_01_lf_06
  bone2Name=dyn_coat_01_lf_05
  swing1LimitDegrees=0
  swing1Motion=Locked
  swing2LimitDegrees=15
  twistLimitDegrees=0
  twistMotion=Locked
 }
 {
  bone1Name=dyn_coat_02_lf_01
  bone2Name=base_coat_02_lf_01
  swing1LimitDegrees=0
  swing1Motion=Locked
  swing2LimitDegrees=15
  twistLimitDegrees=0
  twistMotion=Locked
 }
 {
  bone1Name=dyn_coat_02_lf_02
  bone2Name=dyn_coat_02_lf_01
  swing1LimitDegrees=0
  swing1Motion=Locked
  swing2LimitDegrees=15
  twistLimitDegrees=0
  twistMotion=Locked
 }
 {
  bone1Name=dyn_coat_02_lf_03
  bone2Name=dyn_coat_02_lf_02
  swing1LimitDegrees=0
  swing1Motion=Locked
  swing2LimitDegrees=15
  twistLimitDegrees=0
  twistMotion=Locked
 }
 {
  bone1Name=dyn_coat_02_lf_04
  bone2Name=dyn_coat_02_lf_03
  swing1LimitDegrees=0
  swing1Motion=Locked
  swing2LimitDegrees=15
  twistLimitDegrees=0
  twistMotion=Locked
 }
 {
  bone1Name=dyn_coat_02_lf_05
  bone2Name=dyn_coat_02_lf_04
  swing1LimitDegrees=0
  swing1Motion=Locked
  swing2LimitDegrees=15
  twistLimitDegrees=0
  twistMotion=Locked
 }
 {
  bone1Name=dyn_coat_01_rt_01
  bone2Name=base_coat_01_rt_01
  swing1LimitDegrees=0
  swing1Motion=Locked
  swing2LimitDegrees=15
  twistLimitDegrees=0
  twistMotion=Locked
 }
 {
  bone1Name=dyn_coat_01_rt_02
  bone2Name=dyn_coat_01_rt_01
  swing1LimitDegrees=0
  swing1Motion=Locked
  swing2LimitDegrees=15
  twistLimitDegrees=0
  twistMotion=Locked
 }
 {
  bone1Name=dyn_coat_01_rt_03
  bone2Name=dyn_coat_01_rt_02
  swing1LimitDegrees=0
  swing1Motion=Locked
  swing2LimitDegrees=15
  twistLimitDegrees=0
  twistMotion=Locked
 }
 {
  bone1Name=dyn_coat_01_rt_04
  bone2Name=dyn_coat_01_rt_03
  swing1LimitDegrees=0
  swing1Motion=Locked
  swing2LimitDegrees=15
  twistLimitDegrees=0
  twistMotion=Locked
 }
 {
  bone1Name=dyn_coat_01_rt_05
  bone2Name=dyn_coat_01_rt_04
  swing1LimitDegrees=0
  swing1Motion=Locked
  swing2LimitDegrees=15
  twistLimitDegrees=0
  twistMotion=Locked
 }
 {
  bone1Name=dyn_coat_01_rt_06
  bone2Name=dyn_coat_01_rt_05
  swing1LimitDegrees=0
  swing1Motion=Locked
  swing2LimitDegrees=15
  twistLimitDegrees=0
  twistMotion=Locked
 }
 {
  bone1Name=dyn_coat_02_rt_01
  bone2Name=base_coat_02_rt_01
  swing1LimitDegrees=0
  swing1Motion=Locked
  swing2LimitDegrees=15
  twistLimitDegrees=0
  twistMotion=Locked
 }
 {
  bone1Name=dyn_coat_02_rt_02
  bone2Name=dyn_coat_02_rt_01
  swing1LimitDegrees=0
  swing1Motion=Locked
  swing2LimitDegrees=15
  twistLimitDegrees=0
  twistMotion=Locked
 }
 {
  bone1Name=dyn_coat_02_rt_03
  bone2Name=dyn_coat_02_rt_02
  swing1LimitDegrees=0
  swing1Motion=Locked
  swing2LimitDegrees=15
  twistLimitDegrees=0
  twistMotion=Locked
 }
 {
  bone1Name=dyn_coat_02_rt_04
  bone2Name=dyn_coat_02_rt_03
  swing1LimitDegrees=0
  swing1Motion=Locked
  swing2LimitDegrees=15
  twistLimitDegrees=0
  twistMotion=Locked
 }
 {
  bone1Name=dyn_coat_02_rt_05
  bone2Name=dyn_coat_02_rt_04
  swing1LimitDegrees=0
  swing1Motion=Locked
  swing2LimitDegrees=15
  twistLimitDegrees=0
  twistMotion=Locked
 }
 {
  bone1Name=dyn_coat_05_lf_04
  bone2Name=dyn_coat_03_lf_04
  swing1LimitDegrees=45
  swing1Motion=Locked
  swing2LimitDegrees=45
  swing2Motion=Locked
  twistLimitDegrees=45
  twistMotion=Locked
 }
 {
  bone1Name=dyn_coat_02_lf_05
  bone2Name=dyn_coat_01_lf_05
  swing1LimitDegrees=45
  swing1Motion=Locked
  swing2LimitDegrees=45
  swing2Motion=Locked
  twistLimitDegrees=45
  twistMotion=Locked
 }
 {
  bone1Name=dyn_coat_02_rt_05
  bone2Name=dyn_coat_01_rt_05
  swing1LimitDegrees=45
  swing1Motion=Locked
  swing2LimitDegrees=45
  swing2Motion=Locked
  twistLimitDegrees=45
  twistMotion=Locked
 }
 {
  bone1Name=dyn_coat_03_rt_04
  bone2Name=dyn_coat_05_rt_04
  swing1LimitDegrees=45
  swing1Motion=Locked
  swing2LimitDegrees=45
  swing2Motion=Locked
  twistLimitDegrees=45
  twistMotion=Locked
 }
]
