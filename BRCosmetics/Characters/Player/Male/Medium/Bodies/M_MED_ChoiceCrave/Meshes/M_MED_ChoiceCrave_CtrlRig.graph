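# Control Rig for the M_MED_ChoiceCrave player body that runs the shared DeformRig function, then poses arm strap, coat, chest strap and back strap bones.
# Spherical pose readers on both thighs weight offsets on coat and shirt base bones, which parent `dyn_` chains.
# Smoothed crouch (scaled by pelvis height), knee-slide and baseball-slide flags add coat offsets.
# Straps use weighted parent, rotation and aim constraints, plus upper-arm and slide-driven corrections.
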ControlRigBlueprint /BRCosmetics/Characters/Player/Male/Medium/Bodies/M_MED_ChoiceCrave/Meshes/M_MED_ChoiceCrave_CtrlRig
graphs [
 New Function_ContainedGraph {nodes [
   RigVMFunctionEntryNode()
   > &RigUnit_ParentConstraint(
    AdvancedSettings=(RotationOrderForFilter=XZY)
    Child=(Type=Bone,Name="base_arm_strap_01")
    Filter=(TranslationFilter=(bX=True,bY=True,bZ=True),RotationFilter=(bX=true,bY=true,bZ=True),ScaleFilter=(bX=True,bY=True,bZ=True))
    Parents="((Item=(Type=Bone,Name=\"spine_04\"),Weight=0.0052),(Item=(Type=Bone,Name=\"deform_upperarm_up_r\"),Weight=0.0145),(Item=(Type=Bone,Name=\"spine_03\"),Weight=0.0168),(Item=(Type=Bone,Name=\"upperarm_r\"),Weight=0.0433),(Item=(Type=Bone,Name=\"clavicle_r\"),Weight=0. ...(truncated)"
    Weight=1
   )
   > &RigUnit_RotationConstraintLocalSpaceOffset(
    AdvancedSettings=(InterpolationType=Average,RotationOrderForFilter=XZY)
    Child=(Type=Bone,Name="base_arm_strap_01")
    Filter=(bX=false,bY=True,bZ=True)
    Parents=((Item=(Type=Bone,Name="clavicle_r"),Weight=1))
    Weight=1
   )
   RigUnit_GetTransform = &RigUnit_GetTransform(Item=(Type=Bone,Name="upperarm_r"), Space=LocalSpace, bInitial=False)
   > RigUnit_SetTransform = &RigUnit_SetTransform(Item=(Type=Bone,Name="get_upperarm_r_rotation"), Space=LocalSpace, Weight=1, bInitial=False, Value=@RigUnit_GetTransform.Transform)
   RigVMFunction_MathQuaternionToEuler = &RigVMFunction_MathQuaternionToEuler(
    RotationOrder=ZYX
    Value=&RigUnit_GetTransform(Item=(Type=Bone,Name="get_upperarm_r_rotation"), Space=LocalSpace, bInitial=False).Transform.Rotation
   )
   Remap = &RigVMFunction_MathDoubleRemap(SourceMaximum=60, SourceMinimum=0, TargetMaximum=1, TargetMinimum=0, Value=@RigVMFunction_MathQuaternionToEuler.Result.Y)
   > &RigUnit_ModifyTransforms(
    ItemToModify=((Item=(Type=Bone,Name="base_arm_strap_01"),Transform=(Rotation=(X=-0.0091,Y=-0.0515,Z=-0.1734,W=0.9835),Translation=(X=0,Y=0,Z=0),Scale3D=(X=1,Y=1,Z=1))))
    WeightMinimum=0
    Weight=@Remap.Result
   )
   > RigVMFunctionReturnNode()
  ]}
 New Function_ContainedGraph {nodes [
   RigVMFunctionEntryNode()
   > n1 = &RigUnit_SphericalPoseReader(
    ActiveRegionScaleFactors=(PositiveWidth=1,NegativeWidth=1,PositiveHeight=1,NegativeHeight=1)
    ActiveRegionSize=0.05
    Debug=(bDrawDebug=false,bDraw2D=False,bDrawLocalAxes=False,DebugScale=25,DebugSegments=20,DebugThickness=0.25)
    DriverAxis=(X=-1,Y=0,Z=0)
    DriverItem=(Type=Bone,Name="thigh_l")
    FalloffRegionScaleFactors=(PositiveWidth=1,NegativeWidth=1,PositiveHeight=0.9,NegativeHeight=0.8)
    FalloffSize=0.63
    FlipHeightScaling=False
    FlipWidthScaling=False
    OptionalParentItem=(Type=Bone,Name="None")
    RotationOffset=(X=90,Y=0,Z=-30)
   )
   > &RigUnit_ModifyTransforms(
    ItemToModify=((Item=(Type=Bone,Name="base_coat_front_lf"),Transform=(Rotation=(X=-0.1865,Y=0.1843,Z=-0.0444,W=0.964),Translation=(X=14.105,Y=-1.1016,Z=-1.4615),Scale3D=(X=1,Y=1,Z=1))))
    WeightMinimum=0
    Weight=@n1.OutputParam
   )
   RigVMFunction_AnimEvalRichCurve = &RigVMFunction_AnimEvalRichCurve(
    Curve="(EditorCurveData=(Keys=((InterpMode=RCIM_Cubic,TangentMode=RCTM_User,TangentWeightMode=RCTWM_WeightedNone,Time=0,Value=0,ArriveTangent=0.4671,ArriveTangentWeight=0,LeaveTangent=0.4671,LeaveTangentWeight=0),(InterpMode=RCIM_Cubic,TangentMode=RCTM_User,Tange ...(truncated)"
    SourceMinimum=0
    TargetMinimum=0
    Value=@n1.OutputParam
   )
   > &RigUnit_ModifyTransforms(
    ItemToModify=((Item=(Type=Bone,Name="base_shirt_01_lf"),Transform=(Rotation=(X=0.0721,Y=0.8444,Z=0.1136,W=-0.5186),Translation=(X=-0.8012,Y=0,Z=-0.7155),Scale3D=(X=1,Y=1,Z=1))))
    WeightMinimum=0
    Weight=@RigVMFunction_AnimEvalRichCurve.Result
   )
   > n2 = &RigUnit_SphericalPoseReader(
    ActiveRegionScaleFactors=(PositiveWidth=1,NegativeWidth=1,PositiveHeight=1,NegativeHeight=1)
    ActiveRegionSize=0.05
    Debug=(bDrawDebug=false,bDraw2D=False,bDrawLocalAxes=False,DebugScale=25,DebugSegments=20,DebugThickness=0.25)
    DriverAxis=(X=-1,Y=0,Z=0)
    DriverItem=(Type=Bone,Name="thigh_l")
    FalloffRegionScaleFactors=(PositiveWidth=0.8,NegativeWidth=1,PositiveHeight=0.8,NegativeHeight=0.9)
    FalloffSize=0.63
    FlipHeightScaling=False
    FlipWidthScaling=False
    OptionalParentItem=(Type=Bone,Name="None")
    RotationOffset=(X=180,Y=30,Z=0)
   )
   > &RigUnit_ModifyTransforms(
    ItemToModify="((Item=(Type=Bone,Name=\"base_coat_front_lf\"),Transform=(Rotation=(X=0.0469,Y=0.6434,Z=0.0831,W=-0.7596),Translation=(X=2.355,Y=0,Z=5.7247),Scale3D=(X=1,Y=1,Z=1))),(Item=(Type=Bone,Name=\"base_coat_03_lf_01\"),Transform=(Rotation=(X=0,Y=0,Z=0,W=1),Translation ...(truncated)"
    WeightMinimum=0
    Weight=@n2.OutputParam
   )
   > n3 = &RigUnit_SphericalPoseReader(
    ActiveRegionScaleFactors=(PositiveWidth=1,NegativeWidth=1,PositiveHeight=1,NegativeHeight=1)
    ActiveRegionSize=0.05
    Debug=(bDrawDebug=false,bDraw2D=False,bDrawLocalAxes=False,DebugScale=25,DebugSegments=20,DebugThickness=0.25)
    DriverAxis=(X=-1,Y=0,Z=0)
    DriverItem=(Type=Bone,Name="thigh_l")
    FalloffRegionScaleFactors=(PositiveWidth=1,NegativeWidth=1,PositiveHeight=0.9,NegativeHeight=0.9)
    FalloffSize=0.63
    FlipHeightScaling=False
    FlipWidthScaling=False
    OptionalParentItem=(Type=Bone,Name="None")
    RotationOffset=(X=140,Y=30,Z=0)
   )
   > &RigUnit_ModifyTransforms(
    ItemToModify=((Item=(Type=Bone,Name="base_coat_front_lf"),Transform=(Rotation=(X=0,Y=-0.2164,Z=0,W=0.9763),Translation=(X=0,Y=0,Z=0),Scale3D=(X=1,Y=1,Z=1))))
    WeightMinimum=0
    Weight=@n3.OutputParam
   )
   > n4 = &RigUnit_SphericalPoseReader(
    ActiveRegionScaleFactors=(PositiveWidth=1,NegativeWidth=1,PositiveHeight=1,NegativeHeight=1)
    ActiveRegionSize=0.05
    Debug=(bDrawDebug=false,bDraw2D=False,bDrawLocalAxes=False,DebugScale=25,DebugSegments=20,DebugThickness=0.25)
    DriverAxis=(X=-1,Y=0,Z=0)
    DriverItem=(Type=Bone,Name="thigh_l")
    FalloffRegionScaleFactors=(PositiveWidth=0.75,NegativeWidth=0.75,PositiveHeight=0.9,NegativeHeight=0.9)
    FalloffSize=0.63
    FlipHeightScaling=False
    FlipWidthScaling=False
    OptionalParentItem=(Type=Bone,Name="None")
    RotationOffset=(X=-90,Y=0,Z=10)
   )
   > &RigUnit_ModifyTransforms(
    ItemToModify=((Item=(Type=Bone,Name="base_coat_back_lf"),Transform=(Rotation=(X=0.028,Y=-0.0334,Z=0.6422,W=0.7653),Translation=(X=0,Y=0,Z=0),Scale3D=(X=1,Y=1,Z=1))))
    WeightMinimum=0
    Weight=@n4.OutputParam
   )
   > n5 = &RigUnit_SphericalPoseReader(
    ActiveRegionScaleFactors=(PositiveWidth=1,NegativeWidth=1,PositiveHeight=1,NegativeHeight=1)
    ActiveRegionSize=0.05
    Debug=(bDrawDebug=false,bDraw2D=False,bDrawLocalAxes=False,DebugScale=25,DebugSegments=20,DebugThickness=0.25)
    DriverAxis=(X=-1,Y=0,Z=0)
    DriverItem=(Type=Bone,Name="thigh_l")
    FalloffRegionScaleFactors=(PositiveWidth=0.7,NegativeWidth=0.7,PositiveHeight=0.9,NegativeHeight=0.9)
    FalloffSize=0.63
    FlipHeightScaling=False
    FlipWidthScaling=false
    OptionalParentItem=(Type=Bone,Name="None")
    RotationOffset=(X=-135,Y=0,Z=0)
   )
   > &RigUnit_ModifyTransforms(
    ItemToModify=((Item=(Type=Bone,Name="base_coat_back_lf_offset"),Transform=(Rotation=(X=0,Y=0,Z=0.2276,W=0.9738),Translation=(X=0,Y=0,Z=0),Scale3D=(X=1,Y=1,Z=1))))
    WeightMinimum=0
    Weight=@n5.OutputParam
   )
   > n6 = &RigUnit_SphericalPoseReader(
    ActiveRegionScaleFactors=(PositiveWidth=1,NegativeWidth=1,PositiveHeight=1,NegativeHeight=1)
    ActiveRegionSize=0.05
    Debug=(bDrawDebug=false,bDraw2D=False,bDrawLocalAxes=False,DebugScale=25,DebugSegments=20,DebugThickness=0.25)
    DriverAxis=(X=1,Y=0,Z=0)
    DriverItem=(Type=Bone,Name="thigh_r")
    FalloffRegionScaleFactors=(PositiveWidth=1,NegativeWidth=1,PositiveHeight=0.9,NegativeHeight=0.8)
    FalloffSize=0.63
    FlipHeightScaling=False
    FlipWidthScaling=False
    OptionalParentItem=(Type=Bone,Name="None")
    RotationOffset=(X=-90,Y=0,Z=-30)
   )
   > &RigUnit_ModifyTransforms(
    ItemToModify=((Item=(Type=Bone,Name="base_coat_front_rt"),Transform=(Rotation=(X=-0.1865,Y=0.1843,Z=-0.0444,W=0.964),Translation=(X=-14.105,Y=1.1016,Z=1.4615),Scale3D=(X=1,Y=1,Z=1))))
    WeightMinimum=0
    Weight=@n6.OutputParam
   )
   n7 = &RigVMFunction_AnimEvalRichCurve(
    Curve="(EditorCurveData=(Keys=((InterpMode=RCIM_Cubic,TangentMode=RCTM_User,TangentWeightMode=RCTWM_WeightedNone,Time=0,Value=0,ArriveTangent=0.4671,ArriveTangentWeight=0,LeaveTangent=0.4671,LeaveTangentWeight=0),(InterpMode=RCIM_Cubic,TangentMode=RCTM_User,Tange ...(truncated)"
    SourceMinimum=0
    TargetMinimum=0
    Value=@n6.OutputParam
   )
   > &RigUnit_ModifyTransforms(
    ItemToModify=((Item=(Type=Bone,Name="base_shirt_01_rt"),Transform=(Rotation=(X=0.0721,Y=0.8444,Z=0.1136,W=-0.5186),Translation=(X=-0.8012,Y=0,Z=-0.7155),Scale3D=(X=1,Y=1,Z=1))))
    WeightMinimum=0
    Weight=@n7.Result
   )
   > n8 = &RigUnit_SphericalPoseReader(
    ActiveRegionScaleFactors=(PositiveWidth=1,NegativeWidth=1,PositiveHeight=1,NegativeHeight=1)
    ActiveRegionSize=0.05
    Debug=(bDrawDebug=false,bDraw2D=False,bDrawLocalAxes=False,DebugScale=25,DebugSegments=20,DebugThickness=0.25)
    DriverAxis=(X=1,Y=0,Z=0)
    DriverItem=(Type=Bone,Name="thigh_r")
    FalloffRegionScaleFactors=(PositiveWidth=1,NegativeWidth=0.8,PositiveHeight=0.8,NegativeHeight=0.9)
    FalloffSize=0.63
    FlipHeightScaling=False
    FlipWidthScaling=False
    OptionalParentItem=(Type=Bone,Name="None")
    RotationOffset=(X=0,Y=30,Z=0)
   )
   > &RigUnit_ModifyTransforms(
    ItemToModify="((Item=(Type=Bone,Name=\"base_coat_front_rt\"),Transform=(Rotation=(X=0.0469,Y=0.6434,Z=0.0831,W=-0.7596),Translation=(X=-2.355,Y=0,Z=-5.7247),Scale3D=(X=1,Y=1,Z=1))),(Item=(Type=Bone,Name=\"base_coat_03_lf_01\"),Transform=(Rotation=(X=0,Y=0,Z=0,W=1),Translati ...(truncated)"
    WeightMinimum=0
    Weight=@n8.OutputParam
   )
   > n9 = &RigUnit_SphericalPoseReader(
    ActiveRegionScaleFactors=(PositiveWidth=1,NegativeWidth=1,PositiveHeight=1,NegativeHeight=1)
    ActiveRegionSize=0.05
    Debug=(bDrawDebug=false,bDraw2D=False,bDrawLocalAxes=False,DebugScale=25,DebugSegments=20,DebugThickness=0.25)
    DriverAxis=(X=1,Y=0,Z=0)
    DriverItem=(Type=Bone,Name="thigh_r")
    FalloffRegionScaleFactors=(PositiveWidth=1,NegativeWidth=1,PositiveHeight=0.9,NegativeHeight=0.9)
    FalloffSize=0.63
    FlipHeightScaling=False
    FlipWidthScaling=False
    OptionalParentItem=(Type=Bone,Name="None")
    RotationOffset=(X=-40,Y=30,Z=0)
   )
   > &RigUnit_ModifyTransforms(
    ItemToModify=((Item=(Type=Bone,Name="base_coat_front_rt"),Transform=(Rotation=(X=0,Y=-0.2164,Z=0,W=0.9763),Translation=(X=0,Y=0,Z=0),Scale3D=(X=1,Y=1,Z=1))))
    WeightMinimum=0
    Weight=@n9.OutputParam
   )
   > n10 = &RigUnit_SphericalPoseReader(
    ActiveRegionScaleFactors=(PositiveWidth=1,NegativeWidth=1,PositiveHeight=1,NegativeHeight=1)
    ActiveRegionSize=0.05
    Debug=(bDrawDebug=false,bDraw2D=False,bDrawLocalAxes=False,DebugScale=25,DebugSegments=20,DebugThickness=0.25)
    DriverAxis=(X=1,Y=0,Z=0)
    DriverItem=(Type=Bone,Name="thigh_r")
    FalloffRegionScaleFactors=(PositiveWidth=0.75,NegativeWidth=0.75,PositiveHeight=0.9,NegativeHeight=0.9)
    FalloffSize=0.63
    FlipHeightScaling=False
    FlipWidthScaling=false
    OptionalParentItem=(Type=Bone,Name="None")
    RotationOffset=(X=90,Y=0,Z=10)
   )
   > &RigUnit_ModifyTransforms(
    ItemToModify=((Item=(Type=Bone,Name="base_coat_back_rt"),Transform=(Rotation=(X=0.028,Y=-0.0334,Z=0.6422,W=0.7653),Translation=(X=0,Y=0,Z=0),Scale3D=(X=1,Y=1,Z=1))))
    WeightMinimum=0
    Weight=@n10.OutputParam
   )
   > n11 = &RigUnit_SphericalPoseReader(
    ActiveRegionScaleFactors=(PositiveWidth=1,NegativeWidth=1,PositiveHeight=1,NegativeHeight=1)
    ActiveRegionSize=0.05
    Debug=(bDrawDebug=false,bDraw2D=False,bDrawLocalAxes=False,DebugScale=25,DebugSegments=20,DebugThickness=0.25)
    DriverAxis=(X=1,Y=0,Z=0)
    DriverItem=(Type=Bone,Name="thigh_r")
    FalloffRegionScaleFactors=(PositiveWidth=0.7,NegativeWidth=0.7,PositiveHeight=0.9,NegativeHeight=0.9)
    FalloffSize=0.63
    FlipHeightScaling=False
    FlipWidthScaling=false
    OptionalParentItem=(Type=Bone,Name="None")
    RotationOffset=(X=45,Y=0,Z=0)
   )
   > &RigUnit_ModifyTransforms(
    ItemToModify=((Item=(Type=Bone,Name="base_coat_back_rt_offset"),Transform=(Rotation=(X=0.3075,Y=0.1673,Z=-0.2718,W=-0.8964),Translation=(X=0,Y=0,Z=0.0001),Scale3D=(X=1,Y=1,Z=1))))
    WeightMinimum=0
    Weight=@n11.OutputParam
   )
   Cast = &RigVMFunction_MathBoolToFloat(Value=RigVMVariableNode() {variable=isCrouching}.Value)
   Remap = &RigVMFunction_MathDoubleRemap(
    SourceMaximum=43.5
    SourceMinimum=60
    TargetMaximum=1
    TargetMinimum=0
    Value=&RigUnit_GetRelativeTransformForItem(Child=(Type=Bone,Name="pelvis"), Parent=(Type=Bone,Name="root"), bChildInitial=False, bParentInitial=False).RelativeTransform.Translation.Z
   )
   Multiply = &RigVMFunction_MathDoubleMul(B=@Cast.Result, A=@Remap.Result)
   AlphaInterp = &RigVMFunction_AlphaInterp(InterpSpeedDecreasing=10, InterpSpeedIncreasing=10, Scale=1, bClampResult=False, Value=@Multiply.Result)
   > &RigUnit_ModifyTransforms(
    ItemToModify="((Item=(Type=Bone,Name=\"base_coat_front_lf_offset\"),Transform=(Rotation=(X=0.2121,Y=-0.325,Z=-0.074,W=-0.9187),Translation=(X=-5,Y=0,Z=-3),Scale3D=(X=1,Y=1,Z=1))),(Item=(Type=Bone,Name=\"base_coat_back_rt_offset\"),Transform=(Rotation=(X=0,Y=0,Z=0,W=1),Trans ...(truncated)"
    WeightMinimum=0
    Weight=@AlphaInterp.Result
   )
   RigVMFunction_MathBoolToFloat = &RigVMFunction_MathBoolToFloat(Value=RigVMVariableNode() {variable=isKneeSliding}.Value)
   n12 = &RigVMFunction_AlphaInterp(InterpSpeedDecreasing=10, InterpSpeedIncreasing=10, Scale=1, bClampResult=False, Value=@RigVMFunction_MathBoolToFloat.Result)
   > &RigUnit_ModifyTransforms(
    ItemToModify="((Item=(Type=Bone,Name=\"base_coat_back_lf_offset\"),Transform=(Rotation=(X=0.246,Y=-0.2932,Z=0.5939,W=0.7077),Translation=(X=0,Y=0,Z=0),Scale3D=(X=1,Y=1,Z=1))),(Item=(Type=Bone,Name=\"base_coat_back_rt_offset\"),Transform=(Rotation=(X=0.0616,Y=-0.2988,Z=0.640 ...(truncated)"
    WeightMinimum=0
    Weight=@n12.Result
   )
   n13 = &RigVMFunction_AlphaInterp(
    InterpSpeedDecreasing=10
    InterpSpeedIncreasing=10
    Scale=1
    bClampResult=False
    Value=&RigVMFunction_MathBoolToFloat(Value=RigVMVariableNode() {variable=isBaseballSliding}.Value).Result
   )
   > &RigUnit_ModifyTransforms(
    ItemToModify="((Item=(Type=Bone,Name=\"base_coat_back_lf_offset\"),Transform=(Rotation=(X=0.1373,Y=0.0617,Z=0.7076,W=0.6904),Translation=(X=0,Y=0,Z=0),Scale3D=(X=1,Y=1,Z=1))),(Item=(Type=Bone,Name=\"base_coat_back_rt_offset\"),Transform=(Rotation=(X=0.0206,Y=0.5553,Z=-0.640 ...(truncated)"
    WeightMinimum=0
    Weight=@n13.Result
   )
   > RigVMFunctionReturnNode()
  ]}
 New Function_ContainedGraph {nodes [
   RigVMFunctionEntryNode()
   > &RigUnit_ParentConstraint(
    AdvancedSettings=(RotationOrderForFilter=XZY)
    Child=(Type=Bone,Name="base_chest_strap_01")
    Parents="((Item=(Type=Bone,Name=\"upperarm_twist_01_l\"),Weight=0.0021),(Item=(Type=Bone,Name=\"spine_05\"),Weight=0.0085),(Item=(Type=Bone,Name=\"spine_01\"),Weight=0.0561),(Item=(Type=Bone,Name=\"clavicle_l\"),Weight=0.1057),(Item=(Type=Bone,Name=\"deform_pec_l\"),Weight=0 ...(truncated)"
    Weight=1
   )
   > &RigUnit_ParentConstraint(
    AdvancedSettings=(RotationOrderForFilter=XZY)
    Child=(Type=Bone,Name="base_chest_strap_aim")
    Parents=((Item=(Type=Bone,Name="spine_04"),Weight=0.051),(Item=(Type=Bone,Name="spine_01"),Weight=0.1566),(Item=(Type=Bone,Name="pelvis"),Weight=0.2427),(Item=(Type=Bone,Name="spine_03"),Weight=0.268),(Item=(Type=Bone,Name="spine_02"),Weight=0.2817))
    Weight=1
   )
   > &RigUnit_ModifyTransforms(
    ItemToModify=((Item=(Type=Bone,Name="base_chest_strap_aim"),Transform=(Rotation=(X=0,Y=0,Z=0,W=1),Translation=(X=0,Y=0,Z=0.5),Scale3D=(X=1,Y=1,Z=1))))
    Weight=1
    WeightMinimum=0
   )
   > &RigUnit_AimConstraintLocalSpaceOffset(
    Child=(Type=Bone,Name="dyn_chest_strap_01")
    Filter=(bX=True,bY=True,bZ=True)
    Parents=((Item=(Type=Bone,Name="base_chest_strap_aim"),Weight=1))
    UpAxis=(X=0,Y=1,Z=0)
    Weight=1
    WorldUp=(Target=(X=0,Y=0,Z=1),Kind=Direction,Space=(Type=Bone,Name="base_chest_strap_aim"))
   )
   > RigVMFunctionReturnNode()
  ]}
 New Function_ContainedGraph {nodes [
   RigVMFunctionEntryNode()
   > &RigUnit_ParentConstraint(
    AdvancedSettings=(RotationOrderForFilter=XZY)
    Child=(Type=Bone,Name="base_coat_straps_back_lf_offset")
    Parents="((Item=(Type=Bone,Name=\"spine_04\"),Weight=0.0129),(Item=(Type=Bone,Name=\"dyn_coat_05_lf_01\"),Weight=0.015),(Item=(Type=Bone,Name=\"dyn_coat_05_rt_01\"),Weight=0.015),(Item=(Type=Bone,Name=\"spine_03\"),Weight=0.0872),(Item=(Type=Bone,Name=\"spine_01\"),Weight=0. ...(truncated)"
    Weight=1
   )
   > &RigUnit_ParentConstraint(
    AdvancedSettings=(RotationOrderForFilter=XZY)
    Child=(Type=Bone,Name="base_coat_straps_back_rt_offset")
    Parents="((Item=(Type=Bone,Name=\"spine_04\"),Weight=0.0129),(Item=(Type=Bone,Name=\"dyn_coat_05_lf_01\"),Weight=0.015),(Item=(Type=Bone,Name=\"dyn_coat_05_rt_01\"),Weight=0.015),(Item=(Type=Bone,Name=\"spine_03\"),Weight=0.0872),(Item=(Type=Bone,Name=\"spine_01\"),Weight=0. ...(truncated)"
    Weight=1
   )
   > &RigUnit_RotationConstraintLocalSpaceOffset(
    AdvancedSettings=(InterpolationType=Average,RotationOrderForFilter=XZY)
    Child=(Type=Bone,Name="base_coat_straps_back_lf_offset")
    Filter=(bX=True,bY=True,bZ=True)
    Parents=((Item=(Type=Bone,Name="base_coat_back_lf_offset"),Weight=1))
    Weight=1
   )
   > &RigUnit_RotationConstraintLocalSpaceOffset(
    AdvancedSettings=(InterpolationType=Average,RotationOrderForFilter=XZY)
    Child=(Type=Bone,Name="base_coat_straps_back_rt_offset")
    Filter=(bX=True,bY=True,bZ=True)
    Parents=((Item=(Type=Bone,Name="base_coat_back_rt_offset"),Weight=1))
    Weight=1
   )
   n1 = &RigVMFunction_AlphaInterp(
    InterpSpeedDecreasing=10
    InterpSpeedIncreasing=10
    Scale=1
    bClampResult=False
    Value=&RigVMFunction_MathBoolToFloat(Value=RigVMVariableNode() {variable=isBaseballSliding}.Value).Result
   )
   > &RigUnit_ModifyTransforms(
    ItemToModify=((Item=(Type=Bone,Name="base_coat_straps_back_rt_offset"),Transform=(Rotation=(X=0,Y=0,Z=0.342,W=0.9397),Translation=(X=0,Y=0,Z=0),Scale3D=(X=1,Y=1,Z=1))))
    WeightMinimum=0
    Weight=@n1.Result
   )
   > RigVMFunctionReturnNode()
  ]}
 RigVMFunctionLibrary {nodes [
   arm strap = RigVMCollapseNode(ExecuteContext=())
   back straps = RigVMCollapseNode(ExecuteContext=())
   chest strap = RigVMCollapseNode(ExecuteContext=())
   coat = RigVMCollapseNode(ExecuteContext=())
  ]}
 RigVMModel {nodes [
   RigUnit_BeginExecution = &RigUnit_BeginExecution()
   FN_DeformRig_FNC = RigVMFunctionReferenceNode(
    DeformRig_ Curve="(deformRig_l_upperarm_up_alpha_72_#1=False,deformRig_r_upperarm_up_alpha_73_#2=False,deformRig_l_upperarm_fr_alpha_74_#3=False,deformRig_r_upperarm_fr_alpha_75_CB51B8 ...(truncated)"
    DeformRig_Bool="(upperarm_1_#4=True,elbow_3_#5=True,wrist_7_#6=True,groin_8_#7=True,glute_17_#8=True,knee_9_40C408324CA261 ...(truncated)"
    DeformRig_Debug=(upperarm_debug_18_#4=False,wrist_debug_20_#6=False,lat_debug_26_#9=False,groinGlute_debug_27_#7=False)
    DeformRig_Value="(upperarm_up_18_#10=(Rotation=(X=0.0,Y=-0.0,Z=0.0,W=1.0),Translation=(X=0.0,Y=0.0,Z=8.0),Scale3D=(X=0.70,Y=1.0,Z=1.0)),upperarm_fwd_fr_19_#11=(Rotation=(X=0.0,Y=-0.0,Z=0.0,W=1.0),Translation=(X=0.0, ...(truncated)"
    ExecuteContext=@RigUnit_BeginExecution.ExecutePin
   )
   > arm strap = RigVMFunctionReferenceNode()
   > coat = RigVMFunctionReferenceNode()
   > chest strap = RigVMFunctionReferenceNode()
   > back straps = RigVMFunctionReferenceNode()
  ]}
]
hierarchy [
 root {loc=<0 0 0>, rot=<0 0 0 1>}
 pelvis {parent=root, loc=<0 -0.7071 93.1191>, rot=<0.0127 0.707 -0.0127 -0.707>}
 spine_01 {parent=pelvis, loc=<6.7494 -0.6388 0>, rot=<0 0 0.034 -0.9994>}
 spine_02 {parent=spine_01, loc=<7.2518 -0.035 0>, rot=<0 0 0.0238 -0.9997>}
 spine_03 {parent=spine_02, loc=<6.0104 0 0>, rot=<0 0 0.0149 -0.9999>}
 spine_04 {parent=spine_03, loc=<8.1024 0 0>, rot=<0 0 -0.0132 -0.9999>}
 spine_05 {parent=spine_04, loc=<10.0146 0 0>, rot=<0 0 -0.0015 -1>}
 neck_01 {parent=spine_05, loc=<11.3138 0.1529 0>, rot=<0 0 -0.2003 -0.9797>}
 neck_02 {parent=neck_01, loc=<4.5627 0 0>, rot=<0 0 0.0504 -0.9987>}
 head {parent=neck_02, loc=<5.3914 0 0>, rot=<0 0 0.115 -0.9934>}
 attach_cape {parent=spine_05, loc=<9.0676 -11.7497 -0.0034>, rot=<0.0612 -0.7044 -0.0615 -0.7045>, scale=<0.9974 1.0025 1.0001>}
 cape {parent=attach_cape, type=Null, loc=<0 0.2665 0>, rot=<0 0 0 1>}
 attach_backpack {parent=spine_05, loc=<-0.0385 -13.5064 -0.0056>, rot=<-0.0097 -0.707 0.0094 -0.7071>, scale=<0.9974 1.0025 1.0001>}
 Backpack_BR {parent=attach_backpack, type=Null, loc=<0 0.8826 -0.1556>, rot=<-0.0871 0 0 0.9962>}
 clavicle_l {parent=spine_05, loc=<4.3575 -0.1466 -2.3223>, rot=<0.0465 -0.7289 -0.009 -0.6829>}
 upperarm_l {parent=clavicle_l, loc=<15.5111 0 0>, rot=<0.0226 -0.3234 -0.0088 -0.9459>}
 lowerarm_l {parent=upperarm_l, loc=<26.6514 0 0>, rot=<0.0391 -0.0099 -0.2453 -0.9686>}
 hand_l {parent=lowerarm_l, loc=<22.3491 0 0>, rot=<-0.7081 -0.0507 0.0975 -0.6975>}
 thumb_01_l {parent=hand_l, loc=<1.3118 -0.3191 -2.459>, rot=<0.6546 -0.4042 0.0664 -0.6354>}
 thumb_02_l {parent=thumb_01_l, loc=<5.9336 -0.0448 0.0401>, rot=<-0.0037 0.0213 0.298 -0.9543>}
 thumb_03_l {parent=thumb_02_l, loc=<2.2223 0.3682 -0.0907>, rot=<0.0118 -0.013 0.0066 -0.9998>}
 index_metacarpal_l {parent=hand_l, loc=<1.6535 0.0183 -2.96>, rot=<0.0158 -0.0124 -0.0173 -0.9996>}
 index_01_l {parent=index_metacarpal_l, loc=<8.3008 0.0357 -0.0524>, rot=<-0.0044 0.0485 0.0924 -0.9945>}
 index_02_l {parent=index_01_l, loc=<3.3587 -0.3004 0.0452>, rot=<-0.002 0.0027 0.3225 -0.9466>}
 index_03_l {parent=index_02_l, loc=<2.5578 -0.123 -0.0252>, rot=<-0.002 0.0067 0.2558 -0.9667>}
 middle_metacarpal_l {parent=hand_l, loc=<2.1856 -0.1725 -0.5984>, rot=<0.006 0.0168 -0.0329 -0.9993>}
 middle_01_l {parent=middle_metacarpal_l, loc=<7.3275 -0.2447 0.0129>, rot=<-0.0208 0.0042 0.0858 -0.9961>}
 middle_02_l {parent=middle_01_l, loc=<4.0123 0.1401 0.1175>, rot=<-0.0218 0.0258 0.3512 -0.9357>}
 middle_03_l {parent=middle_02_l, loc=<2.6077 0.0534 0.1373>, rot=<0.0082 0.0188 0.1708 -0.9851>}
 ring_metacarpal_l {parent=hand_l, loc=<2.2074 -0.3625 1.3638>, rot=<-0.0284 0.0722 -0.0458 -0.9959>}
 ring_01_l {parent=ring_metacarpal_l, loc=<7.0295 -0.6197 -0.1446>, rot=<-0.0489 -0.0252 0.1892 -0.9804>}
 ring_02_l {parent=ring_01_l, loc=<3.6174 0.4647 -0.1027>, rot=<0 0 0.2806 -0.9598>}
 ring_03_l {parent=ring_02_l, loc=<2.4152 0.114 0.0199>, rot=<0 0 0.1189 -0.9929>}
 pinky_metacarpal_l {parent=hand_l, loc=<1.7616 -0.9753 3.3814>, rot=<-0.0397 0.0814 -0.069 -0.9935>}
 pinky_01_l {parent=pinky_metacarpal_l, loc=<7.1709 -0.4831 -0.1094>, rot=<-0.0654 -0.0035 0.2242 -0.9723>}
 pinky_02_l {parent=pinky_01_l, loc=<2.8274 0.0893 -0.1043>, rot=<0.0037 0.0011 0.2854 -0.9584>}
 pinky_03_l {parent=pinky_02_l, loc=<2.2046 0.0307 -0.0078>, rot=<-0.0002 0.0098 0.1445 -0.9895>}
 deform_wrist_dn_base_l {parent=hand_l, loc=<0 -2.2342 0>, rot=<0 0 0 -1>}
 deform_wrist_dn_l {parent=deform_wrist_dn_base_l, loc=<0 0 0>, rot=<0 0 0 1>}
 deform_wrist_up_base_l {parent=hand_l, loc=<0 2.2682 0>, rot=<0 0 0 -1>}
 deform_wrist_up_l {parent=deform_wrist_up_base_l, loc=<0 0 0>, rot=<0 0 0 1>}
 lowerarm_twist_02_l {parent=lowerarm_l, loc=<9.6199 0 0>, rot=<0 0 0 -1>}
 lowerarm_twist_01_l {parent=lowerarm_l, loc=<17.8549 0 0>, rot=<0 0 0 -1>}
 deform_elbow_in_base_l {parent=lowerarm_l, loc=<0.2112 5.259 0.8274>, rot=<0 0 0 1>}
 deform_elbow_in_l {parent=deform_elbow_in_base_l, loc=<0 0 0>, rot=<0 0 0 1>}
 deform_elbow_out_base_l {parent=lowerarm_l, loc=<0 -4 0>, rot=<0 0 0 1>}
 deform_elbow_out_l {parent=deform_elbow_out_base_l, loc=<0 0 0>, rot=<0 0 0 1>}
 upperarm_twist_01_l {parent=upperarm_l, loc=<0.0041 0 0>, rot=<0 0 0 -1>}
 deform_upperarm_bk_base_l {parent=upperarm_twist_01_l, loc=<1.0813 -7.2193 -1.9453>, rot=<0 0 0 1>}
 deform_upperarm_bk_l {parent=deform_upperarm_bk_base_l, loc=<0 0 0>, rot=<0 0 0 1>}
 deform_upperarm_fr_base_l {parent=upperarm_twist_01_l, loc=<-0.3691 7.0506 -2.765>, rot=<0 0 0 1>}
 deform_upperarm_fr_l {parent=deform_upperarm_fr_base_l, loc=<0 0 0>, rot=<0 0 0 1>}
 deform_upperarm_up_base_l {parent=upperarm_twist_01_l, loc=<-3.2216 -1.1391 3.3311>, rot=<0 0 0 1>}
 deform_upperarm_up_l {parent=deform_upperarm_up_base_l, loc=<0 0 0>, rot=<0 0 0 1>}
 upperarm_twist_02_l {parent=upperarm_l, loc=<13.7941 0 0>, rot=<0 0 0 -1>}
 clavicle_r {parent=spine_05, loc=<4.3575 -0.1466 2.3223>, rot=<0.7289 0.0465 -0.6829 0.009>}
 upperarm_r {parent=clavicle_r, loc=<-15.5111 0 0>, rot=<0.0226 -0.3234 -0.0088 -0.9459>}
 lowerarm_r {parent=upperarm_r, loc=<-26.6514 0 0>, rot=<0.0391 -0.0099 -0.2453 -0.9686>}
 hand_r {parent=lowerarm_r, loc=<-22.3491 0 0>, rot=<-0.7081 -0.0507 0.0975 -0.6975>}
 thumb_01_r {parent=hand_r, loc=<-1.3118 0.3191 2.459>, rot=<0.6546 -0.4042 0.0664 -0.6354>}
 thumb_02_r {parent=thumb_01_r, loc=<-5.9336 0.0448 -0.0401>, rot=<-0.0037 0.0213 0.298 -0.9543>}
 thumb_03_r {parent=thumb_02_r, loc=<-2.2223 -0.3682 0.0907>, rot=<0.0118 -0.013 0.0066 -0.9998>}
 index_metacarpal_r {parent=hand_r, loc=<-1.6535 -0.0183 2.96>, rot=<0.0158 -0.0124 -0.0173 -0.9996>}
 index_01_r {parent=index_metacarpal_r, loc=<-8.3008 -0.0357 0.0524>, rot=<-0.0044 0.0485 0.0924 -0.9945>}
 index_02_r {parent=index_01_r, loc=<-3.3587 0.3004 -0.0452>, rot=<-0.002 0.0027 0.3225 -0.9466>}
 index_03_r {parent=index_02_r, loc=<-2.5578 0.123 0.0252>, rot=<-0.002 0.0067 0.2558 -0.9667>}
 middle_metacarpal_r {parent=hand_r, loc=<-2.1856 0.1725 0.5984>, rot=<0.006 0.0168 -0.0329 -0.9993>}
 middle_01_r {parent=middle_metacarpal_r, loc=<-7.3275 0.2447 -0.0129>, rot=<-0.0208 0.0042 0.0858 -0.9961>}
 middle_02_r {parent=middle_01_r, loc=<-4.0123 -0.1401 -0.1175>, rot=<-0.0218 0.0258 0.3512 -0.9357>}
 middle_03_r {parent=middle_02_r, loc=<-2.6077 -0.0534 -0.1373>, rot=<0.0082 0.0188 0.1708 -0.9851>}
 ring_metacarpal_r {parent=hand_r, loc=<-2.2074 0.3625 -1.3638>, rot=<-0.0284 0.0722 -0.0458 -0.9959>}
 ring_01_r {parent=ring_metacarpal_r, loc=<-7.0295 0.6197 0.1446>, rot=<-0.0489 -0.0252 0.1892 -0.9804>}
 ring_02_r {parent=ring_01_r, loc=<-3.6174 -0.4647 0.1027>, rot=<0 0 0.2806 -0.9598>}
 ring_03_r {parent=ring_02_r, loc=<-2.4152 -0.114 -0.0199>, rot=<0 0 0.1189 -0.9929>}
 pinky_metacarpal_r {parent=hand_r, loc=<-1.7616 0.9753 -3.3814>, rot=<-0.0397 0.0814 -0.069 -0.9935>}
 pinky_01_r {parent=pinky_metacarpal_r, loc=<-7.1709 0.4831 0.1094>, rot=<-0.0654 -0.0035 0.2242 -0.9723>}
 pinky_02_r {parent=pinky_01_r, loc=<-2.8274 -0.0893 0.1043>, rot=<0.0037 0.0011 0.2854 -0.9584>}
 pinky_03_r {parent=pinky_02_r, loc=<-2.2046 -0.0307 0.0078>, rot=<-0.0002 0.0098 0.1445 -0.9895>}
 deform_wrist_dn_base_r {parent=hand_r, loc=<-0.0004 2.2345 -0.0001>, rot=<0 0 0 -1>}
 deform_wrist_dn_r {parent=deform_wrist_dn_base_r, loc=<0 0 0>, rot=<0 0 0 1>}
 deform_wrist_up_base_r {parent=hand_r, loc=<0.0003 -2.2684 0.0001>, rot=<0 0 0 -1>}
 deform_wrist_up_r {parent=deform_wrist_up_base_r, loc=<0 0 0>, rot=<0 0 0 1>}
 lowerarm_twist_02_r {parent=lowerarm_r, loc=<-9.6199 0 0>, rot=<0 0 0 -1>}
 lowerarm_twist_01_r {parent=lowerarm_r, loc=<-17.8549 0 0>, rot=<0 0 0 -1>}
 deform_elbow_in_base_r {parent=lowerarm_r, loc=<-0.2111 -5.2591 -0.8275>, rot=<0 0 0 1>}
 deform_elbow_in_r {parent=deform_elbow_in_base_r, loc=<-0.0001 0 0.0002>, rot=<0 0 0 1>}
 deform_elbow_out_base_r {parent=lowerarm_r, loc=<0 4 -0.0001>, rot=<0 0 0 1>}
 deform_elbow_out_r {parent=deform_elbow_out_base_r, loc=<0 0 0>, rot=<0 0 0 1>}
 upperarm_twist_01_r {parent=upperarm_r, loc=<-0.0041 0 0>, rot=<0 0 0 -1>}
 deform_upperarm_bk_base_r {parent=upperarm_twist_01_r, loc=<-1.0814 7.2192 1.9455>, rot=<0 0 0 1>}
 deform_upperarm_bk_r {parent=deform_upperarm_bk_base_r, loc=<0 0 0>, rot=<0 0 0 1>}
 deform_upperarm_fr_base_r {parent=upperarm_twist_01_r, loc=<0.3689 -7.0506 2.7652>, rot=<0 0 0 1>}
 deform_upperarm_fr_r {parent=deform_upperarm_fr_base_r, loc=<0 0 0>, rot=<0 0 0 1>}
 deform_upperarm_up_base_r {parent=upperarm_twist_01_r, loc=<3.2214 1.1391 -3.3308>, rot=<0 0 0 1>}
 deform_upperarm_up_r {parent=deform_upperarm_up_base_r, loc=<0 0 0>, rot=<0 0 0 -1>}
 base_arm_strap_01 {parent=upperarm_twist_01_r, loc=<-4.6359 6.0006 1.7009>, rot=<-0.3128 -0.6666 -0.5737 0.3587>}
 dyn_arm_strap_01 {parent=base_arm_strap_01, loc=<0 0 0>, rot=<0 0 0 1>}
 dyn_arm_strap_02 {parent=dyn_arm_strap_01, loc=<8.34 0 0>, rot=<0 0 0 1>}
 dyn_arm_strap_03 {parent=dyn_arm_strap_02, loc=<8.34 0 0>, rot=<0 0 0 1>}
 upperarm_twist_02_r {parent=upperarm_r, loc=<-13.7941 0 0>, rot=<0 0 0 -1>}
 get_upperarm_r_rotation_offset {parent=clavicle_r, loc=<-15.5111 0 0>, rot=<0.0226 -0.3234 -0.0088 -0.9459>}
 get_upperarm_r_rotation {parent=get_upperarm_r_rotation_offset, loc=<0 0 0>, rot=<0 0 0 1>}
 deform_pec_aim_base_l {parent=spine_04, loc=<7.8732 12.6388 -2.4123>, rot=<0.0102 -0.4602 0.3979 -0.7936>}
 deform_pec_aim_l {parent=deform_pec_aim_base_l, loc=<0 0 0>, rot=<0 0 0 1>}
 deform_pec_base_l {parent=deform_pec_aim_l, loc=<10.766 0 0>, rot=<0 0 0 1>}
 deform_pec_l {parent=deform_pec_base_l, loc=<0.055 0 0>, rot=<0 0 0 1>}
 deform_pec_aim_base_r {parent=spine_04, loc=<7.8727 12.6388 2.4123>, rot=<-0.4602 -0.0102 0.7936 0.3979>}
 deform_pec_aim_r {parent=deform_pec_aim_base_r, loc=<0 0 0>, rot=<0 0 0 1>}
 deform_pec_base_r {parent=deform_pec_aim_r, loc=<-10.7662 -0.0006 -0.0007>, rot=<0 0 0 1>}
 deform_pec_r {parent=deform_pec_base_r, loc=<0 0 0>, rot=<0 0 0 1>}
 base_chest_strap_01 {parent=spine_04, loc=<-0.8109 10.6496 -8.4974>, rot=<-0.0053 -0.8039 -0.594 -0.0278>}
 dyn_chest_strap_01 {parent=base_chest_strap_01, loc=<0 0 0>, rot=<0 0 0 1>}
 dyn_chest_strap_02 {parent=dyn_chest_strap_01, loc=<4.5661 0 0>, rot=<0 0 0 1>}
 thigh_l {parent=pelvis, loc=<-2.5693 -0.0109 -7.7501>, rot=<0.0012 0.0572 0.0155 -0.9982>}
 calf_l {parent=thigh_l, loc=<-39.0484 0 0>, rot=<0.0006 -0.0161 -0.0391 -0.9991>}
 foot_l {parent=calf_l, loc=<-42.561 0.1077 0.4292>, rot=<0.0418 -0.0455 0.0343 -0.9975>}
 ball_l {parent=foot_l, loc=<-6.1599 13.8264 0.6806>, rot=<0.0001 -0.0001 -0.7063 -0.7079>}
 calf_twist_02_l {parent=calf_l, loc=<-18.8064 -0.0967 0.0397>, rot=<0.0237 0.0095 -0.0035 -0.9997>}
 calf_twist_01_l {parent=calf_l, loc=<-35.813 -0.0967 0.0397>, rot=<0.0237 0.0095 -0.0035 -0.9997>}
 deform_knee_in_base_l {parent=calf_l, loc=<1.298 -5.2957 -0.0496>, rot=<0 0 0 -1>}
 deform_knee_in_l {parent=deform_knee_in_base_l, loc=<0 0 0>, rot=<0 0 0 1>}
 deform_knee_out_base_l {parent=calf_l, loc=<0 4 0>, rot=<0 0 0 -1>}
 deform_knee_out_l {parent=deform_knee_out_base_l, loc=<0 0 0>, rot=<0 0 0 1>}
 thigh_twist_01_l {parent=thigh_l, loc=<-21.6303 0 0>, rot=<0.0013 0 0 -1>}
 deform_groin_base_l {parent=thigh_l, loc=<-2.1908 8.1142 -1.7442>, rot=<0 0 0 -1>}
 deform_groin_l {parent=deform_groin_base_l, loc=<0 0 0>, rot=<0 0 0 1>}
 thigh_r {parent=pelvis, loc=<-2.5693 -0.0109 7.7501>, rot=<0.0572 -0.0012 0.9982 0.0155>}
 calf_r {parent=thigh_r, loc=<39.0484 0 0>, rot=<0.0006 -0.0161 -0.0391 -0.9991>}
 foot_r {parent=calf_r, loc=<42.561 -0.1077 -0.4292>, rot=<0.0418 -0.0455 0.0343 -0.9975>}
 ball_r {parent=foot_r, loc=<6.1599 -13.8264 -0.6806>, rot=<0.0001 -0.0001 -0.7063 -0.7079>}
 calf_twist_02_r {parent=calf_r, loc=<18.8064 0.0967 -0.0397>, rot=<0.0237 0.0095 -0.0035 -0.9997>}
 calf_twist_01_r {parent=calf_r, loc=<35.813 0.0967 -0.0397>, rot=<0.0237 0.0095 -0.0035 -0.9997>}
 deform_knee_in_base_r {parent=calf_r, loc=<-1.298 5.2957 0.0496>, rot=<0 0 0 -1>}
 deform_knee_in_r {parent=deform_knee_in_base_r, loc=<0 0 0>, rot=<0 0 0 1>}
 deform_knee_out_base_r {parent=calf_r, loc=<0 -4 0>, rot=<0 0 0 -1>}
 deform_knee_out_r {parent=deform_knee_out_base_r, loc=<0 0 0>, rot=<0 0 0 1>}
 thigh_twist_01_r {parent=thigh_r, loc=<21.6303 0 0>, rot=<0.0013 0 0 -1>}
 deform_groin_base_r {parent=thigh_r, loc=<2.1908 -8.1142 1.7442>, rot=<0 0 0 -1>}
 deform_groin_r {parent=deform_groin_base_r, loc=<0 0 0>, rot=<0 0 0 1>}
 base_shirt_01_rt {parent=pelvis, loc=<9.2161 9.4997 3.2018>, rot=<0.0631 0.6989 0.7111 -0.0428>}
 dyn_shirt_01_rt {parent=base_shirt_01_rt, loc=<0 0 0>, rot=<0 0 0 1>}
 base_shirt_01_lf {parent=pelvis, loc=<9.2157 9.4997 -3.2019>, rot=<0.0428 -0.7111 -0.6989 -0.0631>}
 dyn_shirt_01_lf {parent=base_shirt_01_lf, loc=<0 0 0>, rot=<0 0 0 1>}
 base_coat_master {parent=pelvis, loc=<0 0 0>, rot=<-0.0127 -0.707 0.0127 -0.707>}
 coat_master {parent=base_coat_master, loc=<0 0 0>, rot=<0 0 0 1>}
 base_coat_front_lf {parent=coat_master, loc=<12.2309 3.847 7.7916>, rot=<0.0019 0.7464 -0.0016 -0.6655>}
 base_coat_front_lf_offset {parent=base_coat_front_lf, loc=<0 0 0>, rot=<0 0 0 1>}
 base_coat_01_lf_01 {parent=base_coat_front_lf_offset, loc=<-0.0135 2.7336 0.5568>, rot=<0.0088 -0.9921 -0.1038 -0.0697>}
 dyn_coat_01_lf_01 {parent=base_coat_01_lf_01, loc=<0 0 0>, rot=<0 0 0 1>}
 dyn_coat_01_lf_02 {parent=dyn_coat_01_lf_01, loc=<5.7653 0 0>, rot=<0 0 0 1>}
 dyn_coat_01_lf_03 {parent=dyn_coat_01_lf_02, loc=<5.7653 0 0>, rot=<0 0 0 1>}
 dyn_coat_01_lf_04 {parent=dyn_coat_01_lf_03, loc=<14.5364 0 0>, rot=<0 0 0 1>}
 dyn_coat_01_lf_05 {parent=dyn_coat_01_lf_04, loc=<14.5364 0 0>, rot=<0 0 0 1>}
 dyn_coat_01_lf_06 {parent=dyn_coat_01_lf_05, loc=<14.5364 0 0>, rot=<0 0 0 1>}
 base_coat_02_lf_01 {parent=base_coat_front_lf_offset, loc=<0.0135 -2.7336 -0.5568>, rot=<0.0234 -0.9949 -0.058 -0.0786>}
 dyn_coat_02_lf_01 {parent=base_coat_02_lf_01, loc=<0 0 0>, rot=<0 0 0 1>}
 dyn_coat_02_lf_02 {parent=dyn_coat_02_lf_01, loc=<5.7653 0 0>, rot=<0 0 0 1>}
 dyn_coat_02_lf_03 {parent=dyn_coat_02_lf_02, loc=<5.7653 0 0>, rot=<0 0 0 1>}
 dyn_coat_02_lf_04 {parent=dyn_coat_02_lf_03, loc=<14.5364 0 0>, rot=<0 0 0 1>}
 dyn_coat_02_lf_05 {parent=dyn_coat_02_lf_04, loc=<14.5364 0 0>, rot=<0 0 0 1>}
 dyn_coat_02_lf_06 {parent=dyn_coat_02_lf_05, loc=<14.5364 0 0>, rot=<0 0 0 1>}
 base_coat_back_lf {parent=coat_master, loc=<9.0897 -7.5838 7.6085>, rot=<0.0019 0.7464 -0.0016 -0.6655>}
 base_coat_back_lf_offset {parent=base_coat_back_lf, loc=<0 0 0>, rot=<0 0 0 1>}
 base_coat_03_lf_01 {parent=base_coat_back_lf_offset, loc=<-0.1757 5.1083 -3.272>, rot=<0.0041 -0.8105 0.5809 -0.0748>}
 dyn_coat_03_lf_01 {parent=base_coat_03_lf_01, loc=<0 0 0>, rot=<0 0 0 1>}
 dyn_coat_03_lf_02 {parent=dyn_coat_03_lf_01, loc=<11.2 0 0>, rot=<0 0 0 1>}
 dyn_coat_03_lf_03 {parent=dyn_coat_03_lf_02, loc=<14.6 0 0>, rot=<0 0 0 1>}
 dyn_coat_03_lf_04 {parent=dyn_coat_03_lf_03, loc=<14.6451 0 0>, rot=<0 0 0 1>}
 dyn_coat_03_lf_05 {parent=dyn_coat_03_lf_04, loc=<14.6451 0 0>, rot=<0 0 0 1>}
 base_coat_05_lf_01 {parent=base_coat_back_lf_offset, loc=<0.5061 -3.2932 5.168>, rot=<0.095 -0.8005 0.587 -0.0747>}
 dyn_coat_05_lf_01 {parent=base_coat_05_lf_01, loc=<0 0 0>, rot=<0 0 0 1>}
 dyn_coat_05_lf_02 {parent=dyn_coat_05_lf_01, loc=<11.2 0 0>, rot=<0 0 0 1>}
 dyn_coat_05_lf_03 {parent=dyn_coat_05_lf_02, loc=<14.6 0 0>, rot=<0 0 0 1>}
 dyn_coat_05_lf_04 {parent=dyn_coat_05_lf_03, loc=<14.6451 0 0>, rot=<0 0 0 1>}
 dyn_coat_05_lf_05 {parent=dyn_coat_05_lf_04, loc=<14.6451 0 0>, rot=<0 0 0 1>}
 base_coat_front_rt {parent=coat_master, loc=<-12.2309 3.847 7.7919>, rot=<-0.6655 0.0016 0.7464 -0.0019>}
 base_coat_front_rt_offset {parent=base_coat_front_rt, loc=<0 0 0>, rot=<0 0 0 1>}
 base_coat_01_rt_01 {parent=base_coat_front_rt_offset, loc=<0.0134 -2.7336 -0.5568>, rot=<-0.1038 0.0697 -0.0088 -0.9921>}
 dyn_coat_01_rt_01 {parent=base_coat_01_rt_01, loc=<0 0 0>, rot=<0 0 0 1>}
 dyn_coat_01_rt_02 {parent=dyn_coat_01_rt_01, loc=<5.7657 0 -0.0001>, rot=<0 0 0 1>}
 dyn_coat_01_rt_03 {parent=dyn_coat_01_rt_02, loc=<5.7653 0 -0.0001>, rot=<0 0 0 1>}
 dyn_coat_01_rt_04 {parent=dyn_coat_01_rt_03, loc=<14.5365 0 0>, rot=<0 0 0 1>}
 dyn_coat_01_rt_05 {parent=dyn_coat_01_rt_04, loc=<14.5364 0 0>, rot=<0 0 0 1>}
 dyn_coat_01_rt_06 {parent=dyn_coat_01_rt_05, loc=<14.5365 0 0>, rot=<0 0 0 1>}
 base_coat_02_rt_01 {parent=base_coat_front_rt_offset, loc=<-0.0134 2.7336 0.5568>, rot=<-0.058 0.0786 -0.0234 -0.9949>}
 dyn_coat_02_rt_01 {parent=base_coat_02_rt_01, loc=<0 0 0>, rot=<0 0 0 1>}
 dyn_coat_02_rt_02 {parent=dyn_coat_02_rt_01, loc=<5.7654 0 0>, rot=<0 0 0 1>}
 dyn_coat_02_rt_03 {parent=dyn_coat_02_rt_02, loc=<5.7653 0 0>, rot=<0 0 0 1>}
 dyn_coat_02_rt_04 {parent=dyn_coat_02_rt_03, loc=<14.5365 0 0>, rot=<0 0 0 1>}
 dyn_coat_02_rt_05 {parent=dyn_coat_02_rt_04, loc=<14.5365 0 0.0001>, rot=<0 0 0 1>}
 dyn_coat_02_rt_06 {parent=dyn_coat_02_rt_05, loc=<14.5365 0 0>, rot=<0 0 0 1>}
 base_coat_back_rt {parent=coat_master, loc=<-9.0897 -7.5838 7.6089>, rot=<-0.6655 0.0016 0.7464 -0.0019>}
 base_coat_back_rt_offset {parent=base_coat_back_rt, loc=<0 0 0>, rot=<0 0 0 1>}
 base_coat_03_rt_01 {parent=base_coat_back_rt_offset, loc=<0.1761 -5.1083 3.2719>, rot=<0.5809 0.0748 -0.0041 -0.8105>}
 dyn_coat_03_rt_01 {parent=base_coat_03_rt_01, loc=<0 0 0>, rot=<0 0 0 1>}
 dyn_coat_03_rt_02 {parent=dyn_coat_03_rt_01, loc=<11.2 0 0>, rot=<0 0 0 1>}
 dyn_coat_03_rt_03 {parent=dyn_coat_03_rt_02, loc=<14.6 0 0>, rot=<0 0 0 1>}
 dyn_coat_03_rt_04 {parent=dyn_coat_03_rt_03, loc=<14.645 0 0>, rot=<0 0 0 1>}
 dyn_coat_03_rt_05 {parent=dyn_coat_03_rt_04, loc=<14.645 0 0>, rot=<0 0 0 1>}
 base_coat_05_rt_01 {parent=base_coat_back_rt_offset, loc=<-0.5055 3.2933 -5.1681>, rot=<0.587 0.0747 -0.095 -0.8005>}
 dyn_coat_05_rt_01 {parent=base_coat_05_rt_01, loc=<0 0 0>, rot=<0 0 0 1>}
 dyn_coat_05_rt_02 {parent=dyn_coat_05_rt_01, loc=<11.1998 0 0.0001>, rot=<0 0 0 1>}
 dyn_coat_05_rt_03 {parent=dyn_coat_05_rt_02, loc=<14.6 0 0>, rot=<0 0 0 1>}
 dyn_coat_05_rt_04 {parent=dyn_coat_05_rt_03, loc=<14.6452 0 0>, rot=<0 0 0 1>}
 dyn_coat_05_rt_05 {parent=dyn_coat_05_rt_04, loc=<14.6451 0 0>, rot=<0 0 0 1>}
 thigh_l_2 {parent=pelvis, loc=<-2.5693 -0.0109 -7.7501>, rot=<0.0012 0.0572 0.0155 -0.9982>}
 thigh_l_3 {parent=thigh_l_2, loc=<0 0 0>, rot=<0 0 0 1>}
 thigh_l_4 {parent=pelvis, loc=<-2.5693 -0.0109 -7.7501>, rot=<0.0012 0.0572 0.0155 -0.9982>}
 thigh_l_5 {parent=thigh_l_4, loc=<0 0 0>, rot=<0 0 0 1>}
 base_coat_straps_back_lf_offset {parent=pelvis, loc=<12.6947 -10.3129 -0.3308>, rot=<-0.0127 -0.707 0.0127 -0.707>}
 base_ribbon_01_lf {parent=base_coat_straps_back_lf_offset, loc=<0.0255 -0.1729 0.2797>, rot=<-0.5045 -0.1656 0.5507 -0.644>}
 dyn_ribbon_01_lf {parent=base_ribbon_01_lf, loc=<0 0 0>, rot=<0 0 0 1>}
 dyn_ribbon_02_lf {parent=dyn_ribbon_01_lf, loc=<4.9173 0 0>, rot=<0 0 0 1>}
 dyn_ribbon_03_lf {parent=dyn_ribbon_02_lf, loc=<4.9173 0 0>, rot=<0 0 0 1>}
 base_coat_straps_01_lf {parent=base_coat_straps_back_lf_offset, loc=<0 0 0>, rot=<-0.441 -0.402 0.5246 -0.6072>}
 dyn_coat_straps_01_lf {parent=base_coat_straps_01_lf, loc=<0 0 0>, rot=<0 0 0 1>}
 dyn_coat_straps_02_lf {parent=dyn_coat_straps_01_lf, loc=<8.9 0 0>, rot=<0 0.0192 0 -0.9998>}
 dyn_coat_straps_03_lf {parent=dyn_coat_straps_02_lf, loc=<8.9 0 0>, rot=<0 0 0 1>}
 base_coat_straps_back_rt_offset {parent=pelvis, loc=<12.6946 -10.3128 0.3308>, rot=<0.707 -0.0127 -0.707 -0.0127>}
 base_ribbon_01_rt {parent=base_coat_straps_back_rt_offset, loc=<-0.0255 0.173 -0.28>, rot=<-0.5507 -0.644 -0.5045 0.1656>}
 dyn_ribbon_01_rt {parent=base_ribbon_01_rt, loc=<0 0 0>, rot=<0 0 0 1>}
 dyn_ribbon_02_rt {parent=dyn_ribbon_01_rt, loc=<4.9171 0 0>, rot=<0 0 0 1>}
 dyn_ribbon_03_rt {parent=dyn_ribbon_02_rt, loc=<4.9176 0 0>, rot=<0 0 0 1>}
 base_coat_straps_01_rt {parent=base_coat_straps_back_rt_offset, loc=<0 0 0>, rot=<-0.5187 -0.6019 -0.4478 0.41>}
 dyn_coat_straps_01_rt {parent=base_coat_straps_01_rt, loc=<0 0 0>, rot=<0 0.0131 0 -0.9999>}
 dyn_coat_straps_02_rt {parent=dyn_coat_straps_01_rt, loc=<9.0175 0 0>, rot=<0 0.0159 0 -0.9999>}
 dyn_coat_straps_03_rt {parent=dyn_coat_straps_02_rt, loc=<9.0203 0 0>, rot=<0 -0.0128 0 -0.9999>}
 dyn_coat_straps_04_rt {parent=dyn_coat_straps_03_rt, loc=<9.0206 0 0>, rot=<0 0.0078 0 -1>}
 pelvis_support {parent=root, loc=<0 -0.7071 93.1191>, rot=<0.0127 0.707 -0.0127 -0.707>}
 floor {parent=root, loc=<0 0 0>, rot=<0 0 0 1>}
 base_chest_strap_aim {parent=root, loc=<8.8865 9.2022 111.2178>, rot=<0.3915 -0.5307 -0.4462 -0.6049>}
]
variables [
 isBaseballSliding {type=bool, default=False}
 isCrouching {type=bool, default=False}
 isKneeSliding {type=bool, default=False}
]
