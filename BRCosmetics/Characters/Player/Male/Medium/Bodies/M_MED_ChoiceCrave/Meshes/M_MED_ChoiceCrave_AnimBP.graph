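# Post-process style AnimBP for the M_MED_ChoiceCrave player body cosmetic, adding secondary motion to coat, straps and ribbons.
# The parent mesh's pose is copied into a Control Rig fed crouch and slide states.
# AnimDynamics, front-end-only coat offsets, shoulder offsets, a pelvis-space RigidBody coat simulation and bone constraints then run in series.
# The EventGraph updates gravity/RBAN settings, front-end, emote and slide flags, and lowers strap-motion variables on ziplines.
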AnimBlueprint /BRCosmetics/Characters/Player/Male/Medium/Bodies/M_MED_ChoiceCrave/Meshes/M_MED_ChoiceCrave_AnimBP
target_skeleton=&/M_MED_ChoiceCrave_Skeleton
graphs [
 AnimGraph {nodes [
   Output Pose(Result=Use cached pose 'dynamicsCache'()) {layerGroup=None}
   n1 = Control Rig!(
    Source=Copy Pose From Mesh() {bUseAttachedParent=true}
    Alpha
    isCrouching:bool=Get bIsCrouching()
    isKneeSliding:bool=Get bIsKneeSliding()
    isBaseballSliding:bool=Get bIsBaseballSliding()
   ) -> (Pose) {controlRigAssetReference {blueprintRigClass=&/M_MED_ChoiceCrave_CtrlRig.M_MED_ChoiceCrave_CtrlRig_C}}
   Save cached pose 'controlRigCache'(Pose=@n1) {cacheName=controlRigCache}
   n2 = Anim Dynamics - Bone: dyn_chest_strap_01!(
    ComponentPose=Local To Component(LocalPose=Use cached pose 'controlRigCache'())
    Alpha=1
    GravityOverride:Vector="0,0,-980"
    ComponentLinearAccScale:Vector="0.5,0.5,0.5"
    ComponentLinearVelScale:Vector="0.5,0.5,0.5"
   ) -> (Pose) {
    angularDampingOverride=1
    angularSpringConstant=10000
    bAngularSpring=true
    bOverrideAngularDamping=true
    bUseGravityOverride=true
    gravityScale=0.35
    simulationSpace=RootRelative
    boundBone {boneName=dyn_chest_strap_01}
    componentAppliedLinearAccClamp=<3000 3000 3000>
    componentLinearAccScale=<0.5 0.5 0.5>
    componentLinearVelScale=<0.5 0.5 0.5>
    gravityOverride {z=-980}
    physicsBodyDefinitions [{
      boundBone {boneName=dyn_chest_strap_01}
      boxExtents=<30 30 30>
      constraintSetup {angularLimitsMax=<0 0 0>, angularLimitsMin=<0 -20 -10>, angularTarget=<1 0 0>}
      localJointOffset=<8 0 0>
     }]
    simSpaceSettings {maxAngularAcceleration=35, maxAngularVelocity=7, simSpaceAngularAlpha=1}
   }
   n3 = Anim Dynamics - Bone: dyn_chest_strap_02!(
    ComponentPose=@n2
    Alpha=1
    GravityOverride:Vector="0,0,-980"
    ComponentLinearAccScale:Vector="0.5,0.5,0.5"
    ComponentLinearVelScale:Vector="0.5,0.5,0.5"
   ) -> (Pose) {
    angularDampingOverride=1
    angularSpringConstant=10000
    bAngularSpring=true
    bOverrideAngularDamping=true
    bUseGravityOverride=true
    gravityScale=0.35
    simulationSpace=RootRelative
    boundBone {boneName=dyn_chest_strap_02}
    componentAppliedLinearAccClamp=<3000 3000 3000>
    componentLinearAccScale=<0.5 0.5 0.5>
    componentLinearVelScale=<0.5 0.5 0.5>
    gravityOverride {z=-980}
    physicsBodyDefinitions [{
      boundBone {boneName=dyn_chest_strap_02}
      boxExtents=<30 30 30>
      constraintSetup {angularLimitsMax=<0 0 0>, angularLimitsMin=<0 -20 -10>, angularTarget=<1 0 0>}
      localJointOffset=<8 0 0>
     }]
    simSpaceSettings {maxAngularAcceleration=35, maxAngularVelocity=7, simSpaceAngularAlpha=1}
   }
   n4 = Anim Dynamics - Bone: dyn_arm_strap_01!(
    ComponentPose=@n3
    Alpha=1
    GravityOverride:Vector="0,0,-980"
    ComponentLinearAccScale:Vector="0.9,0.9,0.9"
    ComponentLinearVelScale:Vector="0.9,0.9,0.9"
   ) -> (Pose) {
    angularDampingOverride=1
    angularSpringConstant=10000
    bAngularSpring=true
    bOverrideAngularDamping=true
    bUseGravityOverride=true
    bUseSphericalLimits=true
    gravityScale=0.35
    simulationSpace=RootRelative
    boundBone {boneName=dyn_arm_strap_01}
    componentAppliedLinearAccClamp=<3000 3000 3000>
    componentLinearAccScale=<0.9 0.9 0.9>
    componentLinearVelScale=<0.9 0.9 0.9>
    gravityOverride {z=-980}
    physicsBodyDefinitions [{
      collisionType=CustomSphere
      sphereCollisionRadius=3
      boundBone {boneName=dyn_arm_strap_01}
      boxExtents=<30 30 30>
      constraintSetup {angularLimitsMax=<5 30 40>, angularLimitsMin=<-5 -60 -40>, angularTarget=<1 0 0>}
      localJointOffset=<5.2 0 0>
     }]
    planarLimits [{drivingBone {boneName=upperarm_twist_01_r}, planeTransform {location=<0 3 0>, rotation=<rot 0 0 90>}}]
    simSpaceSettings {maxAngularAcceleration=35, maxAngularVelocity=7, simSpaceAngularAlpha=0.7}
    sphericalLimits [{limitRadius=15, drivingBone {boneName=spine_03}, sphereLocalOffset=<5 0 0>}]
   }
   n5 = Anim Dynamics - Bone: dyn_arm_strap_02!(
    ComponentPose=@n4
    Alpha=1
    GravityOverride:Vector="0,0,-980"
    ComponentLinearAccScale:Vector="0.9,0.9,0.9"
    ComponentLinearVelScale:Vector="0.9,0.9,0.9"
   ) -> (Pose) {
    angularDampingOverride=1
    angularSpringConstant=10000
    bAngularSpring=true
    bOverrideAngularDamping=true
    bUseGravityOverride=true
    bUseSphericalLimits=true
    gravityScale=0.35
    simulationSpace=RootRelative
    boundBone {boneName=dyn_arm_strap_02}
    componentAppliedLinearAccClamp=<3000 3000 3000>
    componentLinearAccScale=<0.9 0.9 0.9>
    componentLinearVelScale=<0.9 0.9 0.9>
    gravityOverride {z=-980}
    physicsBodyDefinitions [{
      collisionType=CustomSphere
      sphereCollisionRadius=3
      boundBone {boneName=dyn_arm_strap_02}
      boxExtents=<30 30 30>
      constraintSetup {angularLimitsMax=<5 30 40>, angularLimitsMin=<-5 -60 -40>, angularTarget=<1 0 0>}
      localJointOffset=<5.2 0 0>
     }]
    planarLimits [{drivingBone {boneName=upperarm_twist_01_r}, planeTransform {location=<0 3 0>, rotation=<rot 0 0 90>}}]
    simSpaceSettings {maxAngularAcceleration=35, maxAngularVelocity=7, simSpaceAngularAlpha=0.7}
    sphericalLimits [{limitRadius=15, drivingBone {boneName=spine_03}, sphereLocalOffset=<5 0 0>}]
   }
   n6 = Anim Dynamics - Bone: dyn_arm_strap_03!(
    ComponentPose=@n5
    Alpha=1
    GravityOverride:Vector="0,0,-980"
    ComponentLinearAccScale:Vector="0.9,0.9,0.9"
    ComponentLinearVelScale:Vector="0.9,0.9,0.9"
   ) -> (Pose) {
    angularDampingOverride=1
    angularSpringConstant=10000
    bAngularSpring=true
    bOverrideAngularDamping=true
    bUseGravityOverride=true
    bUseSphericalLimits=true
    gravityScale=0.35
    simulationSpace=RootRelative
    boundBone {boneName=dyn_arm_strap_03}
    componentAppliedLinearAccClamp=<3000 3000 3000>
    componentLinearAccScale=<0.9 0.9 0.9>
    componentLinearVelScale=<0.9 0.9 0.9>
    gravityOverride {z=-980}
    physicsBodyDefinitions [{
      collisionType=CustomSphere
      sphereCollisionRadius=3
      boundBone {boneName=dyn_arm_strap_03}
      boxExtents=<30 30 30>
      constraintSetup {angularLimitsMax=<5 30 40>, angularLimitsMin=<-5 -60 -40>, angularTarget=<1 0 0>}
      localJointOffset=<5.2 0 0>
     }]
    planarLimits [{drivingBone {boneName=upperarm_twist_01_r}, planeTransform {location=<0 3 0>, rotation=<rot 0 0 90>}}]
    simSpaceSettings {maxAngularAcceleration=35, maxAngularVelocity=7, simSpaceAngularAlpha=0.7}
    sphericalLimits [{limitRadius=15, drivingBone {boneName=spine_03}, sphereLocalOffset=<5 0 0>}]
   }
   n7 = Get isFrontEnd() -> (isFrontEnd:double)
   n8 = "Transform (Modify) Bone - Bone: base_coat_01_rt_01"(ComponentPose=@n6, Alpha=@n7, Translation="0,0,1", Rotation="-10,0,0", Scale="1,1,1") {
    rotationMode=BMM_Additive
    rotationSpace=BCS_BoneSpace
    translationMode=BMM_Additive
    translationSpace=BCS_BoneSpace
    boneToModify {boneName=base_coat_01_rt_01}
    rotation {pitch=-10}
    translation {z=1}
   }
   n9 = "Transform (Modify) Bone - Bone: base_coat_01_lf_01"(ComponentPose=@n8, Alpha=@n7, Translation="0,0,1", Rotation="10,0,0", Scale="1,1,1") {
    rotationSpace=BCS_BoneSpace
    translationMode=BMM_Additive
    translationSpace=BCS_BoneSpace
    boneToModify {boneName=base_coat_01_lf_01}
    rotation {pitch=10}
    translation {z=1}
   }
   n10 = "Transform (Modify) Bone - Bone: offset_shoulderClip_lf"(ComponentPose=@n9, Alpha=1, Translation="2,0,0", Rotation="0,0,0", Scale="1,1,1") {
    rotationSpace=BCS_BoneSpace
    translationMode=BMM_Additive
    translationSpace=BCS_BoneSpace
    boneToModify {boneName=offset_shoulderClip_lf}
    translation {x=2}
   }
   n11 = "Transform (Modify) Bone - Bone: offset_shoulderClip_rt"(ComponentPose=@n10, Alpha=1, Translation="2,0,0", Rotation="0,0,0", Scale="1,1,1") {
    rotationSpace=BCS_BoneSpace
    translationMode=BMM_Additive
    translationSpace=BCS_BoneSpace
    boneToModify {boneName=offset_shoulderClip_rt}
    translation {x=2}
   }
   n12 = RigidBody!(
    ComponentPose=@n11
    Alpha
    OverrideWorldGravity:Vector="0,0,0"
    SimSpaceSettings:Sim Space Settings Structure=(WorldAlpha=1,VelocityScaleZ=1,DampingAlpha=0,MaxLinearVelocity=500,MaxAngularVelocity=5,MaxLinearAcceleration=1000,MaxAngularAcceleration=85,ExternalLinearDragV=(X=3,Y=1.1,Z=1.1),ExternalLinearVelocity=(X=0,Y=0,Z=0),ExternalAngularVelocity=(X=0,Y=0,Z=0))
   ) -> (Pose) {
    bClampLinearTranslationLimitToRefPose=true
    bOverrideWorldGravity=true
    overridePhysicsAsset=&/M_MED_ChoiceCrave_Coat_Physics
    simulationSpace=BaseBoneSpace
    baseBoneRef {boneName=pelvis}
    simSpaceSettings {
     dampingAlpha=0
     maxAngularAcceleration=85
     maxAngularVelocity=5
     maxLinearAcceleration=1000
     maxLinearVelocity=500
     worldAlpha=1
     externalLinearDragV=<3 1.1 1.1>
    }
   }
   n13 = "Constraint - dyn_coat_01_rt_01 (R)"!(ComponentPose=@n12, bAlphaBoolEnabled:bool=True, ConstraintWeights_0=1) -> (Pose) {
    alphaInputType=Bool
    boneToModify {boneName=dyn_coat_01_rt_01}
    constraintSetup [{transformType=Rotation, targetBone {boneName=base_coat_front_rt_offset}}]
    constraintWeights [1]
   }
   n14 = "Constraint - dyn_coat_02_rt_01 (R)"!(ComponentPose=@n13, bAlphaBoolEnabled:bool=True, ConstraintWeights_0=1) -> (Pose) {
    alphaInputType=Bool
    boneToModify {boneName=dyn_coat_02_rt_01}
    constraintSetup [{transformType=Rotation, targetBone {boneName=base_coat_front_rt_offset}}]
    constraintWeights [1]
   }
   n15 = "Constraint - dyn_coat_01_rt_02 (R)"!(ComponentPose=@n14, bAlphaBoolEnabled:bool=True, ConstraintWeights_0=1) -> (Pose) {
    alphaInputType=Bool
    boneToModify {boneName=dyn_coat_01_rt_02}
    constraintSetup [{transformType=Rotation, targetBone {boneName=dyn_coat_01_rt_01}}]
    constraintWeights [1]
   }
   n16 = "Constraint - dyn_coat_02_rt_02 (R)"!(ComponentPose=@n15, bAlphaBoolEnabled:bool=True, ConstraintWeights_0=1) -> (Pose) {
    alphaInputType=Bool
    boneToModify {boneName=dyn_coat_02_rt_02}
    constraintSetup [{transformType=Rotation, targetBone {boneName=dyn_coat_01_rt_01}}]
    constraintWeights [1]
   }
   n17 = "Constraint - dyn_coat_01_rt_03 (R)"!(ComponentPose=@n16, bAlphaBoolEnabled:bool=True, ConstraintWeights_0=1) -> (Pose) {
    alphaInputType=Bool
    boneToModify {boneName=dyn_coat_01_rt_03}
    constraintSetup [{transformType=Rotation, targetBone {boneName=dyn_coat_01_rt_02}}]
    constraintWeights [1]
   }
   n18 = "Constraint - dyn_coat_02_rt_03 (R)"!(ComponentPose=@n17, bAlphaBoolEnabled:bool=True, ConstraintWeights_0=1) -> (Pose) {
    alphaInputType=Bool
    boneToModify {boneName=dyn_coat_02_rt_03}
    constraintSetup [{transformType=Rotation, targetBone {boneName=dyn_coat_01_rt_02}}]
    constraintWeights [1]
   }
   n19 = "Constraint - dyn_coat_01_lf_01 (P)"!(ComponentPose=@n18, bAlphaBoolEnabled:bool=True, ConstraintWeights_0=1) -> (Pose) {
    alphaInputType=Bool
    boneToModify {boneName=dyn_coat_01_lf_01}
    constraintSetup [{targetBone {boneName=base_coat_front_lf_offset}}]
    constraintWeights [1]
   }
   n20 = "Constraint - dyn_coat_02_lf_01 (P)"!(ComponentPose=@n19, bAlphaBoolEnabled:bool=True, ConstraintWeights_0=1) -> (Pose) {
    alphaInputType=Bool
    boneToModify {boneName=dyn_coat_02_lf_01}
    constraintSetup [{targetBone {boneName=base_coat_front_lf_offset}}]
    constraintWeights [1]
   }
   n21 = "Constraint - dyn_coat_01_lf_02 (R)"!(ComponentPose=@n20, bAlphaBoolEnabled:bool=True, ConstraintWeights_0=1) -> (Pose) {
    alphaInputType=Bool
    boneToModify {boneName=dyn_coat_01_lf_02}
    constraintSetup [{transformType=Rotation, targetBone {boneName=dyn_coat_01_lf_01}}]
    constraintWeights [1]
   }
   n22 = "Constraint - dyn_coat_02_lf_02 (R)"!(ComponentPose=@n21, bAlphaBoolEnabled:bool=True, ConstraintWeights_0=1) -> (Pose) {
    alphaInputType=Bool
    boneToModify {boneName=dyn_coat_02_lf_02}
    constraintSetup [{transformType=Rotation, targetBone {boneName=dyn_coat_02_lf_01}}]
    constraintWeights [1]
   }
   n23 = "Constraint - dyn_coat_01_lf_03 (R)"!(ComponentPose=@n22, bAlphaBoolEnabled:bool=True, ConstraintWeights_0=1) -> (Pose) {
    alphaInputType=Bool
    boneToModify {boneName=dyn_coat_01_lf_03}
    constraintSetup [{transformType=Rotation, targetBone {boneName=dyn_coat_01_lf_02}}]
    constraintWeights [1]
   }
   n24 = "Constraint - dyn_coat_02_lf_03 (R)"!(ComponentPose=@n23, bAlphaBoolEnabled:bool=True, ConstraintWeights_0=1) -> (Pose) {
    alphaInputType=Bool
    boneToModify {boneName=dyn_coat_02_lf_03}
    constraintSetup [{transformType=Rotation, targetBone {boneName=dyn_coat_02_lf_02}}]
    constraintWeights [1]
   }
   n25 = "Constraint - dyn_coat_straps_01_rt (PPPPPPP)"!(
    ComponentPose=@n24
    bAlphaBoolEnabled:bool=True
    ConstraintWeights_0=0.651
    ConstraintWeights_1:float=0.1136
    ConstraintWeights_2:float=0.1052
    ConstraintWeights_3:float=0.0872
    ConstraintWeights_4:float=0.015
    ConstraintWeights_5:float=0.015
    ConstraintWeights_6:float=0.0129
   ) -> (Pose) {
    alphaInputType=Bool
    boneToModify {boneName=dyn_coat_straps_01_rt}
    constraintSetup [
     {targetBone {boneName=pelvis}}
     {targetBone {boneName=spine_02}}
     {targetBone {boneName=spine_01}}
     {targetBone {boneName=spine_03}}
     {targetBone {boneName=dyn_coat_05_lf_01}}
     {targetBone {boneName=dyn_coat_05_rt_01}}
     {targetBone {boneName=spine_04}}
    ]
    constraintWeights [1,1,1,1,1,1,1]
   }
   n26 = "Constraint - dyn_coat_straps_02_rt (PPPPP)"!(
    ComponentPose=@n25
    bAlphaBoolEnabled:bool=True
    ConstraintWeights_0=0.435
    ConstraintWeights_1:float=0.278
    ConstraintWeights_2:float=0.1611
    ConstraintWeights_3:float=0.1077
    ConstraintWeights_4:float=0.0181
   ) -> (Pose) {
    alphaInputType=Bool
    boneToModify {boneName=dyn_coat_straps_02_rt}
    constraintSetup [
     {targetBone {boneName=pelvis}}
     {targetBone {boneName=dyn_coat_05_rt_01}}
     {targetBone {boneName=dyn_coat_05_lf_01}}
     {targetBone {boneName=dyn_coat_05_rt_02}}
     {targetBone {boneName=dyn_coat_05_lf_02}}
    ]
    constraintWeights [1,1,1,1,1]
   }
   n27 = "Constraint - dyn_coat_straps_03_rt (PPPPPPPP)"!(
    ComponentPose=@n26
    bAlphaBoolEnabled:bool=True
    ConstraintWeights_0=0.448
    ConstraintWeights_1:float=0.2518
    ConstraintWeights_2:float=0.0839
    ConstraintWeights_3:float=0.078
    ConstraintWeights_4:float=0.0685
    ConstraintWeights_5:float=0.0389
    ConstraintWeights_6:float=0.018
    ConstraintWeights_7:float=0.0129
   ) -> (Pose) {
    alphaInputType=Bool
    boneToModify {boneName=dyn_coat_straps_03_rt}
    constraintSetup [
     {targetBone {boneName=dyn_coat_05_rt_02}}
     {targetBone {boneName=dyn_coat_05_rt_01}}
     {targetBone {boneName=dyn_coat_05_lf_01}}
     {targetBone {boneName=pelvis}}
     {targetBone {boneName=dyn_coat_05_rt_03}}
     {targetBone {boneName=dyn_coat_03_rt_02}}
     {targetBone {boneName=dyn_coat_05_rt_04}}
     {targetBone {boneName=dyn_coat_03_rt_03}}
    ]
    constraintWeights [1,1,1,1,1,1,1,1]
   }
   n28 = "Constraint - dyn_coat_straps_04_rt (PPPPPPPP)"!(
    ComponentPose=@n27
    bAlphaBoolEnabled:bool=True
    ConstraintWeights_0=0.4328
    ConstraintWeights_1:float=0.1575
    ConstraintWeights_2:float=0.1484
    ConstraintWeights_3:float=0.0853
    ConstraintWeights_4:float=0.0601
    ConstraintWeights_5:float=0.0517
    ConstraintWeights_6:float=0.0361
    ConstraintWeights_7:float=0.0282
   ) -> (Pose) {
    alphaInputType=Bool
    boneToModify {boneName=dyn_coat_straps_04_rt}
    constraintSetup [
     {targetBone {boneName=dyn_coat_05_rt_02}}
     {targetBone {boneName=dyn_coat_05_rt_03}}
     {targetBone {boneName=dyn_coat_05_rt_01}}
     {targetBone {boneName=dyn_coat_03_rt_02}}
     {targetBone {boneName=pelvis}}
     {targetBone {boneName=dyn_coat_05_rt_04}}
     {targetBone {boneName=dyn_coat_03_rt_03}}
     {targetBone {boneName=dyn_coat_05_lf_01}}
    ]
    constraintWeights [1,1,1,1,1,1,1,1]
   }
   n29 = "Constraint - dyn_coat_straps_01_lf (PPPPPPP)"!(
    ComponentPose=@n28
    bAlphaBoolEnabled:bool=True
    ConstraintWeights_0=0.651
    ConstraintWeights_1:float=0.1136
    ConstraintWeights_2:float=0.1052
    ConstraintWeights_3:float=0.0872
    ConstraintWeights_4:float=0.015
    ConstraintWeights_5:float=0.015
    ConstraintWeights_6:float=0.0129
   ) -> (Pose) {
    alphaInputType=Bool
    boneToModify {boneName=dyn_coat_straps_01_lf}
    constraintSetup [
     {targetBone {boneName=pelvis}}
     {targetBone {boneName=spine_02}}
     {targetBone {boneName=spine_01}}
     {targetBone {boneName=spine_03}}
     {targetBone {boneName=dyn_coat_05_lf_01}}
     {targetBone {boneName=dyn_coat_05_rt_01}}
     {targetBone {boneName=spine_04}}
    ]
    constraintWeights [1,1,1,1,1,1,1]
   }
   n30 = "Constraint - dyn_coat_straps_02_lf (PPPPP)"!(
    ComponentPose=@n29
    bAlphaBoolEnabled:bool=True
    ConstraintWeights_0=0.435
    ConstraintWeights_1:float=0.278
    ConstraintWeights_2:float=0.1611
    ConstraintWeights_3:float=0.1077
    ConstraintWeights_4:float=0.0181
   ) -> (Pose) {
    alphaInputType=Bool
    boneToModify {boneName=dyn_coat_straps_02_lf}
    constraintSetup [
     {targetBone {boneName=pelvis}}
     {targetBone {boneName=dyn_coat_05_lf_01}}
     {targetBone {boneName=dyn_coat_05_rt_01}}
     {targetBone {boneName=dyn_coat_05_lf_02}}
     {targetBone {boneName=dyn_coat_05_rt_02}}
    ]
    constraintWeights [1,1,1,1,1]
   }
   n31 = "Constraint - dyn_coat_straps_03_lf (PPPPPPPP)"!(
    ComponentPose=@n30
    bAlphaBoolEnabled:bool=True
    ConstraintWeights_0=0.448
    ConstraintWeights_1:float=0.2518
    ConstraintWeights_2:float=0.0839
    ConstraintWeights_3:float=0.078
    ConstraintWeights_4:float=0.0685
    ConstraintWeights_5:float=0.0389
    ConstraintWeights_6:float=0.018
    ConstraintWeights_7:float=0.0129
   ) -> (Pose) {
    alphaInputType=Bool
    boneToModify {boneName=dyn_coat_straps_03_lf}
    constraintSetup [
     {targetBone {boneName=dyn_coat_05_lf_02}}
     {targetBone {boneName=dyn_coat_05_lf_01}}
     {targetBone {boneName=dyn_coat_05_rt_01}}
     {targetBone {boneName=pelvis}}
     {targetBone {boneName=dyn_coat_05_lf_03}}
     {targetBone {boneName=dyn_coat_03_lf_02}}
     {targetBone {boneName=dyn_coat_05_lf_04}}
     {targetBone {boneName=dyn_coat_03_lf_03}}
    ]
    constraintWeights [1,1,1,1,1,1,1,1]
   }
   n32 = Anim Dynamics - Bone: dyn_ribbon_01_lf!(ComponentPose=@n31, Alpha=1, ComponentLinearAccScale:Vector="0.5,0.5,0.5", ComponentLinearVelScale:Vector="0.5,0.5,0.5") -> (Pose) {
    angularDampingOverride=1
    angularSpringConstant=10000
    bAngularSpring=true
    bOverrideAngularDamping=true
    bUseGravityOverride=true
    gravityScale=0.35
    simulationSpace=RootRelative
    boundBone {boneName=dyn_ribbon_01_lf}
    componentAppliedLinearAccClamp=<3000 3000 3000>
    componentLinearAccScale=<0.5 0.5 0.5>
    componentLinearVelScale=<0.5 0.5 0.5>
    gravityOverride {z=-980}
    physicsBodyDefinitions [{
      boundBone {boneName=dyn_ribbon_01_lf}
      boxExtents=<30 30 30>
      constraintSetup {angularLimitsMax=<10 0 20>, angularLimitsMin=<-10 -20 -5>, angularTarget=<1 0 0>}
      localJointOffset=<15 0 0>
     }]
    simSpaceSettings {maxAngularAcceleration=35, maxAngularVelocity=7, simSpaceAngularAlpha=1}
   }
   n33 = Anim Dynamics - Bone: dyn_ribbon_02_lf!(ComponentPose=@n32, Alpha=1, ComponentLinearAccScale:Vector="0.5,0.5,0.5", ComponentLinearVelScale:Vector="0.5,0.5,0.5") -> (Pose) {
    angularDampingOverride=1
    angularSpringConstant=10000
    bAngularSpring=true
    bOverrideAngularDamping=true
    bUseGravityOverride=true
    gravityScale=0.35
    simulationSpace=RootRelative
    boundBone {boneName=dyn_ribbon_02_lf}
    componentAppliedLinearAccClamp=<3000 3000 3000>
    componentLinearAccScale=<0.5 0.5 0.5>
    componentLinearVelScale=<0.5 0.5 0.5>
    gravityOverride {z=-980}
    physicsBodyDefinitions [{
      boundBone {boneName=dyn_ribbon_02_lf}
      boxExtents=<30 30 30>
      constraintSetup {angularLimitsMax=<10 0 20>, angularLimitsMin=<-10 -20 -5>, angularTarget=<1 0 0>}
      localJointOffset=<15 0 0>
     }]
    simSpaceSettings {maxAngularAcceleration=35, maxAngularVelocity=7, simSpaceAngularAlpha=1}
   }
   n34 = Anim Dynamics - Bone: dyn_ribbon_03_lf!(ComponentPose=@n33, Alpha=1, ComponentLinearAccScale:Vector="0.5,0.5,0.5", ComponentLinearVelScale:Vector="0.5,0.5,0.5") -> (Pose) {
    angularDampingOverride=1
    angularSpringConstant=10000
    bAngularSpring=true
    bOverrideAngularDamping=true
    bUseGravityOverride=true
    gravityScale=0.35
    simulationSpace=RootRelative
    boundBone {boneName=dyn_ribbon_03_lf}
    componentAppliedLinearAccClamp=<3000 3000 3000>
    componentLinearAccScale=<0.5 0.5 0.5>
    componentLinearVelScale=<0.5 0.5 0.5>
    gravityOverride {z=-980}
    physicsBodyDefinitions [{
      boundBone {boneName=dyn_ribbon_03_lf}
      boxExtents=<30 30 30>
      constraintSetup {angularLimitsMax=<10 0 20>, angularLimitsMin=<-10 -20 -5>, angularTarget=<1 0 0>}
      localJointOffset=<15 0 0>
     }]
    simSpaceSettings {maxAngularAcceleration=35, maxAngularVelocity=7, simSpaceAngularAlpha=1}
   }
   n35 = Anim Dynamics - Bone: dyn_coat_straps_01_lf!(ComponentPose=@n34, Alpha=1, ComponentLinearAccScale:Vector="0.5,0.5,0.5", ComponentLinearVelScale:Vector="0.5,0.5,0.5") -> (Pose) {
    angularDampingOverride=1
    angularSpringConstant=10000
    bAngularSpring=true
    bOverrideAngularDamping=true
    bUseGravityOverride=true
    gravityScale=0.35
    simulationSpace=RootRelative
    boundBone {boneName=dyn_coat_straps_01_lf}
    componentAppliedLinearAccClamp=<3000 3000 3000>
    componentLinearAccScale=<0.5 0.5 0.5>
    componentLinearVelScale=<0.5 0.5 0.5>
    gravityOverride {z=-980}
    physicsBodyDefinitions [{
      boundBone {boneName=dyn_coat_straps_01_lf}
      boxExtents=<30 30 30>
      constraintSetup {angularLimitsMax=<5 0 10>, angularLimitsMin=<-5 -20 0>, angularTarget=<1 0 0>}
      localJointOffset=<8 0 0>
     }]
    simSpaceSettings {maxAngularAcceleration=35, maxAngularVelocity=7, simSpaceAngularAlpha=1}
   }
   n36 = Anim Dynamics - Bone: dyn_coat_straps_02_lf!(ComponentPose=@n35, Alpha=1, ComponentLinearAccScale:Vector="0.5,0.5,0.5", ComponentLinearVelScale:Vector="0.5,0.5,0.5") -> (Pose) {
    angularDampingOverride=1
    angularSpringConstant=10000
    bAngularSpring=true
    bOverrideAngularDamping=true
    bUseGravityOverride=true
    gravityScale=0.35
    simulationSpace=RootRelative
    boundBone {boneName=dyn_coat_straps_02_lf}
    componentAppliedLinearAccClamp=<3000 3000 3000>
    componentLinearAccScale=<0.5 0.5 0.5>
    componentLinearVelScale=<0.5 0.5 0.5>
    gravityOverride {z=-980}
    physicsBodyDefinitions [{
      boundBone {boneName=dyn_coat_straps_02_lf}
      boxExtents=<30 30 30>
      constraintSetup {angularLimitsMax=<5 0 10>, angularLimitsMin=<-5 -20 0>, angularTarget=<1 0 0>}
      localJointOffset=<8 0 0>
     }]
    simSpaceSettings {maxAngularAcceleration=35, maxAngularVelocity=7, simSpaceAngularAlpha=1}
   }
   n37 = Anim Dynamics - Bone: dyn_coat_straps_03_lf!(ComponentPose=@n36, Alpha=1, ComponentLinearAccScale:Vector="0.5,0.5,0.5", ComponentLinearVelScale:Vector="0.5,0.5,0.5") -> (Pose) {
    angularDampingOverride=1
    angularSpringConstant=10000
    bAngularSpring=true
    bOverrideAngularDamping=true
    bUseGravityOverride=true
    gravityScale=0.35
    simulationSpace=RootRelative
    boundBone {boneName=dyn_coat_straps_03_lf}
    componentAppliedLinearAccClamp=<3000 3000 3000>
    componentLinearAccScale=<0.5 0.5 0.5>
    componentLinearVelScale=<0.5 0.5 0.5>
    gravityOverride {z=-980}
    physicsBodyDefinitions [{
      boundBone {boneName=dyn_coat_straps_03_lf}
      boxExtents=<30 30 30>
      constraintSetup {angularLimitsMax=<5 0 10>, angularLimitsMin=<-5 -20 0>, angularTarget=<1 0 0>}
      localJointOffset=<8 0 0>
     }]
    simSpaceSettings {maxAngularAcceleration=35, maxAngularVelocity=7, simSpaceAngularAlpha=1}
   }
   n38 = Anim Dynamics - Bone: dyn_ribbon_01_rt!(ComponentPose=@n37, Alpha=1, ComponentLinearAccScale:Vector="0.5,0.5,0.5", ComponentLinearVelScale:Vector="0.5,0.5,0.5") -> (Pose) {
    angularDampingOverride=1
    angularSpringConstant=10000
    bAngularSpring=true
    bOverrideAngularDamping=true
    bUseGravityOverride=true
    gravityScale=0.35
    simulationSpace=RootRelative
    boundBone {boneName=dyn_ribbon_01_rt}
    componentAppliedLinearAccClamp=<3000 3000 3000>
    componentLinearAccScale=<0.5 0.5 0.5>
    componentLinearVelScale=<0.5 0.5 0.5>
    gravityOverride {z=-980}
    physicsBodyDefinitions [{
      boundBone {boneName=dyn_ribbon_01_rt}
      boxExtents=<30 30 30>
      constraintSetup {angularLimitsMax=<10 0 3>, angularLimitsMin=<-10 -20 -20>, angularTarget=<1 0 0>}
      localJointOffset=<10 0 0>
     }]
    simSpaceSettings {maxAngularAcceleration=35, maxAngularVelocity=7, simSpaceAngularAlpha=1}
   }
   n39 = Anim Dynamics - Bone: dyn_ribbon_02_rt!(ComponentPose=@n38, Alpha=1, ComponentLinearAccScale:Vector="0.5,0.5,0.5", ComponentLinearVelScale:Vector="0.5,0.5,0.5") -> (Pose) {
    angularDampingOverride=1
    angularSpringConstant=10000
    bAngularSpring=true
    bOverrideAngularDamping=true
    bUseGravityOverride=true
    gravityScale=0.35
    simulationSpace=RootRelative
    boundBone {boneName=dyn_ribbon_02_rt}
    componentAppliedLinearAccClamp=<3000 3000 3000>
    componentLinearAccScale=<0.5 0.5 0.5>
    componentLinearVelScale=<0.5 0.5 0.5>
    gravityOverride {z=-980}
    physicsBodyDefinitions [{
      boundBone {boneName=dyn_ribbon_02_rt}
      boxExtents=<30 30 30>
      constraintSetup {angularLimitsMax=<10 0 3>, angularLimitsMin=<-10 -20 -20>, angularTarget=<1 0 0>}
      localJointOffset=<10 0 0>
     }]
    simSpaceSettings {maxAngularAcceleration=35, maxAngularVelocity=7, simSpaceAngularAlpha=1}
   }
   n40 = Anim Dynamics - Bone: dyn_ribbon_03_rt!(ComponentPose=@n39, Alpha=1, ComponentLinearAccScale:Vector="0.5,0.5,0.5", ComponentLinearVelScale:Vector="0.5,0.5,0.5") -> (Pose) {
    angularDampingOverride=1
    angularSpringConstant=10000
    bAngularSpring=true
    bOverrideAngularDamping=true
    bUseGravityOverride=true
    gravityScale=0.35
    simulationSpace=RootRelative
    boundBone {boneName=dyn_ribbon_03_rt}
    componentAppliedLinearAccClamp=<3000 3000 3000>
    componentLinearAccScale=<0.5 0.5 0.5>
    componentLinearVelScale=<0.5 0.5 0.5>
    gravityOverride {z=-980}
    physicsBodyDefinitions [{
      boundBone {boneName=dyn_ribbon_03_rt}
      boxExtents=<30 30 30>
      constraintSetup {angularLimitsMax=<10 0 3>, angularLimitsMin=<-10 -20 -20>, angularTarget=<1 0 0>}
      localJointOffset=<10 0 0>
     }]
    simSpaceSettings {maxAngularAcceleration=35, maxAngularVelocity=7, simSpaceAngularAlpha=1}
   }
   n41 = Anim Dynamics - Bone: dyn_coat_straps_01_rt!(ComponentPose=@n40, Alpha=1, ComponentLinearAccScale:Vector="0.5,0.5,0.5", ComponentLinearVelScale:Vector="0.5,0.5,0.5") -> (Pose) {
    angularDampingOverride=1
    angularSpringConstant=10000
    bAngularSpring=true
    bOverrideAngularDamping=true
    bUseGravityOverride=true
    gravityScale=0.35
    simulationSpace=RootRelative
    boundBone {boneName=dyn_coat_straps_01_rt}
    componentAppliedLinearAccClamp=<3000 3000 3000>
    componentLinearAccScale=<0.5 0.5 0.5>
    componentLinearVelScale=<0.5 0.5 0.5>
    gravityOverride {z=-980}
    physicsBodyDefinitions [{
      boundBone {boneName=dyn_coat_straps_01_rt}
      boxExtents=<35 35 35>
      constraintSetup {angularLimitsMax=<10 0 1>, angularLimitsMin=<-10 -15 -15>, angularTarget=<1 0 0>}
      localJointOffset=<8 0 0>
     }]
    simSpaceSettings {maxAngularAcceleration=35, maxAngularVelocity=7, simSpaceAngularAlpha=1}
   }
   n42 = Anim Dynamics - Bone: dyn_coat_straps_02_rt!(ComponentPose=@n41, Alpha=1, ComponentLinearAccScale:Vector="0.5,0.5,0.5", ComponentLinearVelScale:Vector="0.5,0.5,0.5") -> (Pose) {
    angularDampingOverride=1
    angularSpringConstant=10000
    bAngularSpring=true
    bOverrideAngularDamping=true
    bUseGravityOverride=true
    gravityScale=0.35
    simulationSpace=RootRelative
    boundBone {boneName=dyn_coat_straps_02_rt}
    componentAppliedLinearAccClamp=<3000 3000 3000>
    componentLinearAccScale=<0.5 0.5 0.5>
    componentLinearVelScale=<0.5 0.5 0.5>
    gravityOverride {z=-980}
    physicsBodyDefinitions [{
      boundBone {boneName=dyn_coat_straps_02_rt}
      boxExtents=<35 35 35>
      constraintSetup {angularLimitsMax=<10 0 1>, angularLimitsMin=<-10 -20 -15>, angularTarget=<1 0 0>}
      localJointOffset=<8 0 0>
     }]
    simSpaceSettings {maxAngularAcceleration=35, maxAngularVelocity=7, simSpaceAngularAlpha=1}
   }
   n43 = Anim Dynamics - Bone: dyn_coat_straps_03_rt!(ComponentPose=@n42, Alpha=1, ComponentLinearAccScale:Vector="0.5,0.5,0.5", ComponentLinearVelScale:Vector="0.5,0.5,0.5") -> (Pose) {
    angularDampingOverride=1
    angularSpringConstant=10000
    bAngularSpring=true
    bOverrideAngularDamping=true
    bUseGravityOverride=true
    gravityScale=0.35
    simulationSpace=RootRelative
    boundBone {boneName=dyn_coat_straps_03_rt}
    componentAppliedLinearAccClamp=<3000 3000 3000>
    componentLinearAccScale=<0.5 0.5 0.5>
    componentLinearVelScale=<0.5 0.5 0.5>
    gravityOverride {z=-980}
    physicsBodyDefinitions [{
      boundBone {boneName=dyn_coat_straps_03_rt}
      boxExtents=<35 35 35>
      constraintSetup {angularLimitsMax=<10 0 1>, angularLimitsMin=<-10 -20 -15>, angularTarget=<1 0 0>}
      localJointOffset=<8 0 0>
     }]
    simSpaceSettings {maxAngularAcceleration=35, maxAngularVelocity=7, simSpaceAngularAlpha=1}
   }
   n44 = Anim Dynamics - Bone: dyn_coat_straps_04_rt!(ComponentPose=@n43, Alpha=1, ComponentLinearAccScale:Vector="0.5,0.5,0.5", ComponentLinearVelScale:Vector="0.5,0.5,0.5") -> (Pose) {
    angularDampingOverride=1
    angularSpringConstant=10000
    bAngularSpring=true
    bOverrideAngularDamping=true
    bUseGravityOverride=true
    gravityScale=0.35
    simulationSpace=RootRelative
    boundBone {boneName=dyn_coat_straps_04_rt}
    componentAppliedLinearAccClamp=<3000 3000 3000>
    componentLinearAccScale=<0.5 0.5 0.5>
    componentLinearVelScale=<0.5 0.5 0.5>
    gravityOverride {z=-980}
    physicsBodyDefinitions [{
      boundBone {boneName=dyn_coat_straps_04_rt}
      boxExtents=<35 35 35>
      constraintSetup {angularLimitsMax=<10 0 1>, angularLimitsMin=<-10 -20 -15>, angularTarget=<1 0 0>}
      localJointOffset=<8 0 0>
     }]
    simSpaceSettings {maxAngularAcceleration=35, maxAngularVelocity=7, simSpaceAngularAlpha=1}
   }
   Save cached pose 'dynamicsCache'(Pose=Component To Local(ComponentPose=@n44)) {cacheName=dynamicsCache}
  ]}
 EventGraph {nodes [
   Event BlueprintUpdateAnimation()
   > n1 = "Gravity Override State Function (Parameter Asset) (Nativized)"(
    CharacterPartAnimInstance=Self-Reference()
    GameTimeInSeconds=GetGameTimeInSeconds()
    Parameters=&/M_MED_Choice_Crave_Parameters
   )
   > Set GravityOverride(GravityOverride=@n1)
   > Set RBAN Settings Out(RBAN Settings Out=@n1.OutRigidBodyAnimNodeParameters)
   n2 = GetAnimInstance(self=GetOwnerSkeletalMeshComponent())
   > Is Valid(InputObject=@n2)
   > Set is Custom Emote Playing(is Custom Emote Playing:bool=Montage_IsPlaying(self=@n2, Montage=/BRCosmetics/Animation/Game/MainPlayer/Emotes/Eastern_Bloc/Emote_EasternBloc_CMM_M)) -> (Output_Get:bool)
   n3 = GetOwnerAnimationInterface!(self:Custom Character Part Anim Instance Object Reference=Self-Reference()) -> (ReturnValue:Fort Player Animation Interface Object Reference)
   > Is Valid(InputObject=@n3)
   > n4 = Set isZipline(isZipline:bool=Break Fort Anim Input Zipline!(FortAnimInput_Zipline:"Fort Anim Input Zipline Structure (by ref)"=GetZiplineInput!(self:Fort Player Animation Interface Object Reference=@n3) -> (ReturnValue:"Fort Anim Input Zipline Structure (by ref)")) -> (
     bIsZiplining:bool=~
     bIsOnAscender:bool=~
     bShouldPlayPivotTransition:bool=~
     LeanYaw:float=~
     PivotBlendDelayRemaining:float=~
     PivotBlendDelay:float=~
     PivotCardinalDirection:EFortCardinalDirection Enum
     WorldVelocityLastTick:Vector=~
    ) {is=title K2Node_BreakStruct}) -> (Output_Get:bool)
   > Branch(Condition=@n4)
   > Set animDyn on zipline back strap(animDyn on zipline back strap:Vector="0, 0, 0") -> (Output_Get:Vector)
   > Set animDyn on zipline arm strap(animDyn on zipline arm strap:Vector="0.1,0.1,0.1") -> (Output_Get:Vector)
   Event OnBeginSliding() {is=title K2Node_Event}
   > Set isSlidingOff(isSlidingOff:bool=false) -> (Output_Get:bool)
   Event OnEndSliding() {is=title K2Node_Event}
   > Set isSlidingOff(isSlidingOff=true)
   Event OnCharacterCustomizationLoaded()
   > Set isFrontEnd(isFrontEnd:double="To Float (Boolean)"(InBool=Get bIsInFrontEnd())) -> (Output_Get:double)
  ]}
]
variables [
 animDyn on zipline arm strap {type=Vector, default=<0.9 0.9 0.9>}
 animDyn on zipline back strap {type=Vector, default=<0.5 0.5 0.5>}
 GravityOverride {type=Vector}
 is Custom Emote Playing {type=boolean}
 isFrontEnd {type=number}
 isSlidingOff {type=boolean, default=true}
 isZipline {type=boolean}
 RBAN Settings Out {type=FortRigidBodyAnimNodeParameters}
]
