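# Physics asset of cape colliders for the M_MED_ChoiceCrave medium male player body, by its name for keeping the cape from clipping the character.
# Capsules cover the pelvis, spine, clavicles and neck, with a sphere on the head and three overlapping capsules on spine_02.
# More capsules sit on left and right dyn_coat bones, so coat segments also act as colliders.
# Arm, leg and foot bodies hold no shapes, and no constraints link the bodies.
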PhysicsAsset /BRCosmetics/Characters/Player/Male/Medium/Bodies/M_MED_ChoiceCrave/Meshes/M_MED_ChoiceCrave_CapeColliders
bodies {
 calf_l []
 calf_r []
 clavicle_l [{shapeType=Capsule, length=9.0192, radius=9.8871, center=<9.4266 -1.1806 -2.4977>, rotation=<rot 69.3297 1.4476 176.9462>}]
 clavicle_r [{shapeType=Capsule, length=9.0192, radius=9.8871, center=<-9.4266 1.1806 2.4977>, rotation=<rot 69.3297 1.4476 -3.0538>}]
 dyn_coat_05_lf_03 [{
   shapeType=Capsule
   length=27.2962
   radius=11.7402
   shapeName=dyn_coat_05_rt_03_capsule
   center=<4.4987 0.9809 -5.3571>
   rotation=<rot 0 85.0001 89.9995>
  }]
 dyn_coat_05_lf_05 [{
   shapeType=Capsule
   length=27.2457
   radius=10.1466
   shapeName=dyn_coat_05_rt_03_capsule
   center=<7.4447 0.7231 -5.6964>
   rotation=<rot 0 85.0001 89.9995>
  }]
 dyn_coat_05_rt_03 [{
   shapeType=Capsule
   length=27.2962
   radius=11.7402
   shapeName=dyn_coat_05_rt_03_capsule
   center=<4.2514 -1.846 -3.9702>
   rotation=<rot -0.0001 84.9999 89.9995>
  }]
 dyn_coat_05_rt_05 [{
   shapeType=Capsule
   length=28.2214
   radius=7.6604
   shapeName=dyn_coat_05_rt_03_capsule
   center=<6.5563 -1.3065 -4.7559>
   rotation=<rot -0.0001 84.9999 89.9995>
  }]
 foot_l []
 foot_r []
 head [{shapeType=Sphere, radius=17.3125, center=<11.028 -0.174 0>}]
 lowerarm_l []
 lowerarm_r []
 neck_02 [{shapeType=Capsule, length=13.4375, radius=7.3175, center=<-0.2642 -0.3408 0>, rotation=<rot 90 -171.1193 -175.3978>}]
 pelvis [{shapeType=Capsule, length=5.1917, radius=19.1267, center=<-2.4887 0.1881 0>}]
 spine_01 [{shapeType=Capsule, length=7.9438, radius=13.5586, center=<-1.7965 0.1884 0>}]
 spine_02 [
  {shapeType=Capsule, length=32.1272, radius=13.1324, center=<-1.711 -5.7287 0.0001>}
  {shapeType=Capsule, length=33.2382, radius=12.685, shapeName=spine_02_capsule, center=<-9.9683 -6.7611 0>}
  {shapeType=Capsule, length=20.5781, radius=9.4219, shapeName=spine_02_capsule, center=<-9.3291 -9.3655 0>}
 ]
 spine_03 [{shapeType=Capsule, length=27.3257, radius=13.6718, center=<2.766 -4.0671 0.0002>}]
 spine_04 [{shapeType=Capsule, length=9.4404, radius=13, center=<2.766 0.7765 0.0002>}]
 spine_05 [{shapeType=Capsule, length=9.4404, radius=13, center=<2.766 0.7765 0.0002>}]
 thigh_l []
 thigh_r []
 upperarm_l []
 upperarm_r []
}
constraints []
