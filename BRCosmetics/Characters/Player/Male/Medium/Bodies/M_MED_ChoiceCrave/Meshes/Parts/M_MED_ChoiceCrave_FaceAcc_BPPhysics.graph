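# Physics asset for the face accessory part of the M_MED_ChoiceCrave male character, supporting secondary motion of hair and face-framing strands.
# Capsules cover the dynamic hair and face-frame bones, plus pelvis, spine, clavicles, upper arms, coat and an arm strap.
# Twist-locked, limited-swing constraints link only hair segments 04–06 per side and the face-frame bones.
# Left face-frame bones form a chain, while right bones 02 and 03 both branch from 01.
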PhysicsAsset /BRCosmetics/Characters/Player/Male/Medium/Bodies/M_MED_ChoiceCrave/Meshes/Parts/M_MED_ChoiceCrave_FaceAcc_BPPhysics
bodies {
 base_faceFrame01_lf [{shapeType=Capsule, length=0.505, radius=0.505, shapeName=base_faceFrame01_lf_capsule}]
 clavicle_l [{shapeType=Capsule, length=8, radius=7, shapeName=clavicle_r_capsule, center=<5.0349 0.5287 0.4709>, rotation=<rot 75 9.9992 -0.0011>}]
 clavicle_r [{shapeType=Capsule, length=8, radius=7, shapeName=clavicle_r_capsule, center=<-5.0349 -0.5287 -0.4709>, rotation=<rot 75 9.9992 -0.0011>}]
 dyn_arm_strap_02 [{shapeType=Capsule, length=23, radius=2.5, shapeName=dyn_arm_strap_01_capsule, center=<9.4785 0 0>, rotation=<rot 90 0 0>}]
 dyn_coat_05_lf_03 [{
   shapeType=Capsule
   length=20
   radius=16
   shapeName=dyn_coat_05_lf_03_capsule
   center=<-1.315 4.5762 -17.4265>
   rotation=<rot 86.2474 -10.9352 171.2557>
  }]
 dyn_coat_05_rt_03 [{
   shapeType=Capsule
   length=20
   radius=16
   shapeName=dyn_coat_05_lf_03_capsule
   center=<-1.315 -4.5584 -17.4265>
   rotation=<rot 86.2474 -10.9352 171.2557>
  }]
 dyn_faceFrame01_lf [{shapeType=Capsule, length=2.5, radius=1.25, shapeName=dyn_faceFrame01_lf_capsule, center=<2.3362 0 0>, rotation=<rot 90 0 0>}]
 dyn_faceFrame01_rt [{shapeType=Capsule, length=0.505, radius=0.505, shapeName=base_faceFrame01_rt_capsule, rotation=<rot 0 -90 0>}]
 dyn_faceFrame02_lf [{shapeType=Capsule, length=3, radius=1.25, shapeName=dyn_faceFrame02_lf_capsule, center=<2.8199 0 0>, rotation=<rot 70 -179.9992 -179.9994>}]
 dyn_faceFrame02_rt [{shapeType=Capsule, length=7.3, radius=1.7, shapeName=dyn_faceFrame02_rt_capsule, center=<5.2 0 0>, rotation=<rot 90 -10.0003 0>}]
 dyn_faceFrame03_rt [{shapeType=Capsule, length=5, radius=1.5, shapeName=dyn_faceFrame03_rt_capsule, center=<4 0 0>, rotation=<rot 90 9.9997 0>}]
 dyn_hair_01_lf [{shapeType=Capsule, length=9, radius=2.8, shapeName=dyn_hair_01_lf_capsule, center=<7.45 0 0>, rotation=<rot 90 0 0>}]
 dyn_hair_01_rt [{shapeType=Capsule, length=9, radius=2.8, shapeName=dyn_hair_01_lf_capsule, center=<7.45 0 0>, rotation=<rot 90 0 0>}]
 dyn_hair_02_lf [{
   shapeType=Capsule
   length=3.5
   radius=3
   shapeName=dyn_hair_02_lf_capsule
   center=<4.6741 -0.0981 0.8612>
   rotation=<rot 84.3752 -179.9996 -179.9996>
  }]
 dyn_hair_02_rt [{
   shapeType=Capsule
   length=3.5
   radius=3
   shapeName=dyn_hair_02_lf_capsule
   center=<4.6741 -0.0981 1.0469>
   rotation=<rot 84.3752 179.9996 179.9996>
  }]
 dyn_hair_03_lf [{shapeType=Capsule, length=7, radius=3.25, shapeName=dyn_hair_03_lf_capsule, center=<6.8211 0 0.2786>, rotation=<rot 90 0 0>}]
 dyn_hair_03_rt [{shapeType=Capsule, length=7, radius=3.25, shapeName=dyn_hair_03_lf_capsule, center=<6.8211 0 0.2786>, rotation=<rot 90 0 0>}]
 dyn_hair_04_lf [{shapeType=Capsule, length=5, radius=3.5, shapeName=dyn_hair_04_lf_capsule, center=<5.9528 0 0.0965>, rotation=<rot 79.9999 -0.0003 -0.0002>}]
 dyn_hair_04_rt [{shapeType=Capsule, length=5, radius=3.5, shapeName=dyn_hair_04_lf_capsule, center=<5.9528 0 0.0965>, rotation=<rot 79.9999 -0.0003 -0.0002>}]
 dyn_hair_05_lf [{shapeType=Capsule, length=2.1, radius=5.2, shapeName=dyn_hair_05_lf_capsule, center=<6.0612 0.668 -0.8483>, rotation=<rot 90 2.8125 0>}]
 dyn_hair_05_rt [{shapeType=Capsule, length=2.1, radius=5.2, shapeName=dyn_hair_05_lf_capsule, center=<6.0612 0.668 -0.8483>, rotation=<rot 90 2.8125 0>}]
 dyn_hair_06_lf [{
   shapeType=Capsule
   length=7.8
   radius=7.2
   shapeName=dyn_hair_06_lf_capsule
   center=<12.3097 1.4202 -0.6802>
   rotation=<rot 75.8733 3.3782 0.4033>
  }]
 dyn_hair_06_rt [{
   shapeType=Capsule
   length=9.3
   radius=7.2
   shapeName=dyn_hair_06_lf_capsule
   center=<12.3097 1.4202 -0.6802>
   rotation=<rot 75.8733 3.3782 0.4033>
  }]
 pelvis [{shapeType=Capsule, length=7, radius=18.5, shapeName=pelvis_capsule, center=<-0.6641 -0.2112 0>}]
 spine_03 [{shapeType=Capsule, length=15, radius=10, center=<12.8173 1.0738 0>}]
 spine_05 [{shapeType=Capsule, length=8.5, radius=6.5, shapeName=clavicle_r_capsule, center=<5.3555 1.7946 0>}]
 upperarm_l [{shapeType=Capsule, length=18, radius=6, shapeName=upperarm_r_capsule, center=<9.9905 0 0>, rotation=<rot 90 0 0>}]
 upperarm_r [{shapeType=Capsule, length=18, radius=6, shapeName=upperarm_r_capsule, center=<-9.9905 0 0>, rotation=<rot 90 0 0>}]
}
constraints [
 {bone1Name=dyn_hair_05_lf, bone2Name=dyn_hair_04_lf, swing1LimitDegrees=8, swing2LimitDegrees=4, twistLimitDegrees=45, twistMotion=Locked}
 {bone1Name=dyn_hair_06_lf, bone2Name=dyn_hair_05_lf, swing1LimitDegrees=4, swing2LimitDegrees=4, twistLimitDegrees=45, twistMotion=Locked}
 {bone1Name=dyn_hair_05_rt, bone2Name=dyn_hair_04_rt, swing1LimitDegrees=4, swing2LimitDegrees=4, twistLimitDegrees=45, twistMotion=Locked}
 {bone1Name=dyn_hair_06_rt, bone2Name=dyn_hair_05_rt, swing1LimitDegrees=8, swing2LimitDegrees=8, twistLimitDegrees=45, twistMotion=Locked}
 {
  bone1Name=dyn_faceFrame01_lf
  bone2Name=base_faceFrame01_lf
  swing1LimitDegrees=5
  swing2LimitDegrees=5
  twistLimitDegrees=45
  twistMotion=Locked
 }
 {
  bone1Name=dyn_faceFrame02_lf
  bone2Name=dyn_faceFrame01_lf
  swing1LimitDegrees=20
  swing2LimitDegrees=25
  twistLimitDegrees=45
  twistMotion=Locked
 }
 {
  bone1Name=dyn_faceFrame02_rt
  bone2Name=dyn_faceFrame01_rt
  swing1LimitDegrees=60
  swing2LimitDegrees=15
  twistLimitDegrees=45
  twistMotion=Locked
 }
 {
  bone1Name=dyn_faceFrame03_rt
  bone2Name=dyn_faceFrame01_rt
  swing1LimitDegrees=15
  swing2LimitDegrees=15
  twistLimitDegrees=45
  twistMotion=Locked
 }
]
