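# Control Rig for the ChoiceCrave medium male character's face accessory, correctively posing hair and jewel bones to follow body movement and state.
# The Aim function weights additive bone offsets by head/neck delta rotations and spherical pose readers, plus melee, aiming and crouch-targeting flags.
# Aim constraints point hair bases and jewel at target bones, several offsets moving the head target.
# Skydiving and spine-driven standing offsets follow.
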ControlRigBlueprint /BRCosmetics/Characters/Player/Male/Medium/Bodies/M_MED_ChoiceCrave/Meshes/Parts/M_MED_ChoiceCrave_FaceAcc_CtrlRig
graphs [
 New Function_ContainedGraph {nodes [
   n1 = RigVMFunctionEntryNode()
   RigVMFunction_Sequence = &RigVMFunction_Sequence(ExecuteContext=@n1.ExecuteContext)
   RigVMFunction_Sequence_1 = RigVMAggregateNode(ExecuteContext=@RigVMFunction_Sequence.A)
   DeltaRotation = RigVMFunctionReferenceNode(
    Rotation Order=ZYX
    child=head
    debugRotation=false
    max=-40
    onlyTwist=false
    parent=head_null
    parentType=Null
    premult=1
    primAxis=2
    twistAxis=(X=1,Y=0,Z=0)
    ExecuteContext=@RigVMFunction_Sequence_1.A
   )
   > &RigUnit_ModifyTransforms(
    ItemToModify="((Item=(Type=Bone,Name=\"head_aim_tgt\"),Transform=(Rotation=(X=-0.0698,Y=0,Z=0,W=0.9976),Translation=(X=0,Y=-2.5,Z=-8),Scale3D=(X=1,Y=1,Z=1))),(Item=(Type=Bone,Name=\"dyn_hair_01_rt\"),Transform=(Rotation=(X=0.0109,Y=0.104,Z=-0.104,W=0.9891),Translation=(X=-2 ...(truncated)"
    WeightMinimum=0
    Weight=@DeltaRotation.outputWeight
   )
   > DeltaRotation_2 = RigVMFunctionReferenceNode(
    Rotation Order=ZYX
    child=neck_01
    debugRotation=false
    max=-35
    onlyTwist=false
    parent=neck_null
    parentType=Null
    premult=1
    primAxis=2
    twistAxis=(X=1,Y=0,Z=0)
   )
   > &RigUnit_ModifyTransforms(
    ItemToModify="((Item=(Type=Bone,Name=\"dyn_hair_01_lf\"),Transform=(Rotation=(X=0,Y=0,Z=0,W=1),Translation=(X=0,Y=0,Z=4),Scale3D=(X=1,Y=1,Z=1))),(Item=(Type=Bone,Name=\"dyn_hair_01_rt\"),Transform=(Rotation=(X=0,Y=0,Z=0,W=1),Translation=(X=0,Y=0,Z=4),Scale3D=(X=1,Y=1,Z=1))) ...(truncated)"
    WeightMinimum=0
    Weight=@DeltaRotation_2.outputWeight
   )
   > DeltaRotation_1 = RigVMFunctionReferenceNode(
    Rotation Order=ZYX
    child=head
    debugRotation=false
    max=30
    parent=head_null
    parentType=Null
    premult=1
    primAxis=2
    twistAxis=(X=1,Y=0,Z=0)
   )
   > &RigUnit_ModifyTransforms(
    ItemToModify="((Item=(Type=Bone,Name=\"head_aim_tgt\"),Transform=(Rotation=(X=0.4226,Y=0,Z=0,W=0.9063),Translation=(X=0,Y=-2,Z=0),Scale3D=(X=1,Y=1,Z=1))),(Item=(Type=Bone,Name=\"base_faceFrame01_rt\"),Transform=(Rotation=(X=0,Y=0,Z=-0.2588,W=0.9659),Translation=(X=0,Y=0,Z=0 ...(truncated)"
    WeightMinimum=0
    Weight=@DeltaRotation_1.outputWeight
   )
   n2 = &RigUnit_SphericalPoseReader(
    ActiveRegionScaleFactors=(PositiveWidth=0.75,NegativeWidth=0.75,PositiveHeight=1,NegativeHeight=1)
    ActiveRegionSize=0.1
    Debug=(bDrawDebug=false,bDraw2D=false,DebugScale=25,DebugSegments=20,DebugThickness=0.25)
    DriverAxis=(X=0,Y=1,Z=0)
    DriverItem=(Type=Bone,Name="head")
    FalloffRegionScaleFactors=(PositiveWidth=1,NegativeWidth=1,PositiveHeight=1,NegativeHeight=2.2)
    FalloffSize=0.2
    OptionalParentItem=(Type=Bone)
    RotationOffset=(X=150,Y=0,Z=0)
    ExecutePin=@RigVMFunction_Sequence_1.B
   )
   > &RigUnit_ModifyTransforms(
    ItemToModify="((Item=(Type=Bone,Name=\"head_aim_tgt\"),Transform=(Rotation=(X=0,Y=0,Z=-0.0436,W=0.999),Translation=(X=-6,Y=0,Z=0),Scale3D=(X=1,Y=1,Z=1))),(Item=(Type=Bone,Name=\"base_hair_lf\"),Transform=(Rotation=(X=0,Y=0,Z=0,W=1),Translation=(X=0,Y=3,Z=-4),Scale3D=(X=1,Y= ...(truncated)"
    WeightMinimum=0
    Weight=@n2.OutputParam
   )
   > n3 = &RigUnit_SphericalPoseReader(
    ActiveRegionScaleFactors=(PositiveWidth=0.75,NegativeWidth=0.75,PositiveHeight=1.5,NegativeHeight=1)
    ActiveRegionSize=0.1
    Debug=(bDrawDebug=false,bDraw2D=false,DebugScale=25,DebugSegments=20,DebugThickness=0.25)
    DriverAxis=(X=0,Y=1,Z=0)
    DriverItem=(Type=Bone,Name="head")
    FalloffRegionScaleFactors=(PositiveWidth=1,NegativeWidth=1,PositiveHeight=1,NegativeHeight=2.2)
    FalloffSize=0.2
    FlipHeightScaling=true
    FlipWidthScaling=false
    OptionalParentItem=(Type=Bone)
    RotationOffset=(X=30,Y=0,Z=0)
   )
   > &RigUnit_ModifyTransforms(
    ItemToModify="((Item=(Type=Bone,Name=\"head_aim_tgt\"),Transform=(Rotation=(X=0,Y=0,Z=-0.0436,W=0.999),Translation=(X=0,Y=-1.5,Z=0),Scale3D=(X=1,Y=1,Z=1))),(Item=(Type=Bone,Name=\"base_faceFrame01_rt\"),Transform=(Rotation=(X=0,Y=-0.5,Z=0,W=0.866),Translation=(X=0,Y=0,Z=0), ...(truncated)"
    WeightMinimum=0
    Weight=@n3.OutputParam
   )
   RigVMFunction_MathBoolToFloat = &RigVMFunction_MathBoolToFloat(Value=RigVMVariableNode() {variable=IsMeelee}.Value)
   AlphaInterp = &RigVMFunction_AlphaInterp(Bias=0.5, InterpSpeedDecreasing=2, InterpSpeedIncreasing=2, Scale=1, bClampResult=False, Value=@RigVMFunction_MathBoolToFloat.Result)
   RigVMFunction_AnimEvalRichCurve = &RigVMFunction_AnimEvalRichCurve(
    Curve="(EditorCurveData=(Keys=((InterpMode=RCIM_Cubic,TangentMode=RCTM_Auto,TangentWeightMode=RCTWM_WeightedNone,Time=0,Value=0,ArriveTangent=0,ArriveTangentWeight=0,LeaveTangent=0,LeaveTangentWeight=0),(InterpMode=RCIM_Cubic,TangentMode=RCTM_Auto,TangentWeightMo ...(truncated)"
    SourceMinimum=0
    TargetMinimum=0
    Value=@AlphaInterp.Result
   )
   > &RigUnit_ModifyTransforms(
    ItemToModify="((Item=(Type=Bone,Name=\"dyn_hair_01_rt\"),Transform=(Rotation=(X=-0.0349,Y=0,Z=0,W=0.9994),Translation=(X=0,Y=0,Z=0.5),Scale3D=(X=1,Y=1,Z=1))),(Item=(Type=Bone,Name=\"dyn_hair_01_lf\"),Transform=(Rotation=(X=0.0872,Y=0,Z=0,W=0.9962),Translation=(X=0,Y=0,Z=0), ...(truncated)"
    WeightMinimum=0
    Weight=@RigVMFunction_AnimEvalRichCurve.Result
   )
   &RigUnit_AimConstraintLocalSpaceOffset(
    Child=(Type=Bone,Name="base_hair_lf")
    Filter=(bX=True,bY=true,bZ=true)
    Parents=((Item=(Type=Bone,Name="head_aim_tgt"),Weight=1))
    UpAxis=(X=0,Y=-1,Z=0)
    Weight=1
    WorldUp=(Target=(X=0,Y=-1,Z=0),Kind=Location,Space=(Type=Bone,Name="head_aim_tgt"))
    bMaintainOffset=true
    ExecutePin=@RigVMFunction_Sequence_1.C
   )
   > &RigUnit_AimConstraintLocalSpaceOffset(
    Child=(Type=Bone,Name="base_hair_rt")
    Filter=(bX=true,bY=true,bZ=true)
    Parents=((Item=(Type=Bone,Name="head_aim_tgt"),Weight=1))
    UpAxis=(X=0,Y=-1,Z=0)
    Weight=1
    WorldUp=(Target=(X=0,Y=-1,Z=0),Kind=Location,Space=(Type=Bone,Name="head_aim_tgt"))
    bMaintainOffset=true
   )
   > &RigUnit_AimConstraintLocalSpaceOffset(
    Child=(Type=Bone,Name="aim_jewel_cr")
    Filter=(bX=true,bY=true,bZ=true)
    Parents=((Item=(Type=Bone,Name="aimtgt_jewel_cr"),Weight=1))
    UpAxis=(X=0,Y=-1,Z=0)
    Weight=1
    WorldUp=(Target=(X=0,Y=-1,Z=0),Kind=Location,Space=(Type=Bone,Name="aimtgt_jewel_cr"))
    bMaintainOffset=true
   )
   n4 = &RigUnit_SphericalPoseReader(
    ActiveRegionScaleFactors=(PositiveWidth=0.75,NegativeWidth=0.75,PositiveHeight=1,NegativeHeight=1.2)
    ActiveRegionSize=0.05
    Debug=(bDrawDebug=false,bDraw2D=false,DebugScale=25,DebugSegments=20,DebugThickness=0.25)
    DriverAxis=(X=1,Y=0,Z=-1)
    DriverItem=(Type=Bone,Name="clavicle_r")
    FalloffRegionScaleFactors=(PositiveWidth=0.5,NegativeWidth=0.4,PositiveHeight=0.5,NegativeHeight=0.75)
    FalloffSize=0.25
    FlipHeightScaling=true
    FlipWidthScaling=false
    OptionalParentItem=(Type=Bone,Name="head")
    RotationOffset=(X=75,Y=-8,Z=-117)
    ExecutePin=@RigVMFunction_Sequence_1.D
   )
   n5 = &RigVMFunction_AlphaInterp(
    InterpSpeedDecreasing=15
    InterpSpeedIncreasing=15
    Scale=1
    bClampResult=False
    bMapRange=false
    Value=&RigVMFunction_MathBoolToFloat(Value=RigVMVariableNode() {variable=IsAiming}.Value).Result
   )
   > &RigUnit_ModifyTransforms(
    ItemToModify="((Item=(Type=Bone,Name=\"head_aim_tgt\"),Transform=(Rotation=(X=-0.1294,Y=-0.017,Z=0.1294,W=0.983),Translation=(X=7,Y=-0.5,Z=5),Scale3D=(X=1,Y=1,Z=1))),(Item=(Type=Bone,Name=\"dyn_hair_01_rt\"),Transform=(Rotation=(X=0,Y=0,Z=0,W=1),Translation=(X=-4,Y=0,Z=2.5) ...(truncated)"
    WeightMinimum=0
    Weight=&RigVMFunction_MathDoubleMul(B=@n5.Result, A=@n4.OutputParam).Result
   )
   n6 = &RigVMFunction_AlphaInterp(
    InterpSpeedDecreasing=10
    InterpSpeedIncreasing=10
    Scale=1
    bClampResult=False
    bInterpResult=False
    Value=&RigVMFunction_MathBoolToFloat(Value=RigVMVariableNode() {variable=IsCrouchTargeting}.Value).Result
   )
   n7 = &RigVMFunction_AnimEvalRichCurve(
    Curve="(EditorCurveData=(Keys=((InterpMode=RCIM_Linear,TangentMode=RCTM_Auto,TangentWeightMode=RCTWM_WeightedNone,Time=0,Value=0,ArriveTangent=0,ArriveTangentWeight=0,LeaveTangent=0,LeaveTangentWeight=0),(InterpMode=RCIM_Linear,TangentMode=RCTM_Auto,TangentWeight ...(truncated)"
    SourceMinimum=0
    TargetMinimum=0
    Value=@n6.Result
   )
   > &RigUnit_ModifyTransforms(
    ItemToModify=((Item=(Type=Bone,Name="base_hair_rt"),Transform=(Rotation=(X=0,Y=-0.0958,Z=0,W=0.9954),Translation=(X=0,Y=0,Z=0),Scale3D=(X=1,Y=1,Z=1))))
    WeightMinimum=0
    Weight=@n7.Result
   )
   > n8 = &RigUnit_SphericalPoseReader(
    ActiveRegionScaleFactors=(PositiveWidth=0.8,NegativeWidth=0.8,PositiveHeight=0.8,NegativeHeight=0.8)
    ActiveRegionSize=0.12
    Debug=(bDrawDebug=true,bDraw2D=false,DebugScale=25,DebugSegments=20,DebugThickness=0.25)
    DriverAxis=(X=1,Y=0,Z=0)
    DriverItem=(Type=Bone,Name="spine_01")
    FalloffRegionScaleFactors=(PositiveWidth=0.5,NegativeWidth=0.5,PositiveHeight=0.5,NegativeHeight=0.5)
    FalloffSize=0.27
    OptionalParentItem=(Type=Bone,Name="pelvis")
    RotationOffset=(X=97,Y=0,Z=-60)
   )
   Multiply = &RigVMFunction_MathDoubleMul(B=@n6.Result, A=@n8.OutputParam)
   > &RigUnit_ModifyTransforms(
    ItemToModify=((Item=(Type=Bone,Name="base_hair_rt"),Transform=(Rotation=(X=0,Y=0.1392,Z=0,W=0.9903),Translation=(X=0,Y=0,Z=0),Scale3D=(X=1,Y=1,Z=1))))
    WeightMinimum=0
    Weight=@Multiply.Result
   )
   RigVMFunctionReturnNode(ExecuteContext=@RigVMFunction_Sequence.B)
  ]}
 New Function_ContainedGraph {nodes [
   n1 = RigVMFunctionEntryNode()
   RigVMFunction_MathBoolToInteger = &RigVMFunction_MathBoolToInteger(Value=@n1.onlyTwist)
   RigVMFunction_MathQuaternionSwingTwist = &RigVMFunction_MathQuaternionSwingTwist(
    twistAxis=@n1.twistAxis
    Input=&RigUnit_GetRelativeTransformForItem(
     bChildInitial=False
     bParentInitial=False
     child=(Type=Bone,Name="None")
     parent=(Type=Bone,Name="None")
     child.Name=@n1.child
     parent.Name=@n1.parent
     parent.Type=@n1.parentType
    ).RelativeTransform.Rotation
   )
   RigVMFunction_MathQuaternionToEuler = &RigVMFunction_MathQuaternionToEuler(
    Value=RigVMDispatchNode(
     Values=((),())
     Index=@RigVMFunction_MathBoolToInteger.Result
     Values.0=@RigVMFunction_MathQuaternionSwingTwist.Swing
     Values.1=@RigVMFunction_MathQuaternionSwingTwist.Twist
    ) {template="DISPATCH_RigVMDispatch_SelectInt32(in Index,in Values,out Result)"}.Result
    RotationOrder=@n1.Rotation Order
   )
   > DISPATCH_RigVMDispatch_Print = RigVMDispatchNode(ScreenColor=(R=1,G=1,B=1,A=1), ScreenDuration=1.5, Enabled=@n1.debugRotation, Value=@RigVMFunction_MathQuaternionToEuler.Result) {template="DISPATCH_RigVMDispatch_Print(in Prefix,in Value,in Enabled,in ScreenDuration,in ScreenColor)"}
   Clamp = &RigVMFunction_MathIntClamp(Value=@n1.primAxis)
   DISPATCH_RigVMDispatch_SelectInt32 = RigVMDispatchNode(
    Values=(0,0,0)
    Index=@Clamp.Result
    Values.0=@RigVMFunction_MathQuaternionToEuler.Result.X
    Values.1=@RigVMFunction_MathQuaternionToEuler.Result.Y
    Values.2=@RigVMFunction_MathQuaternionToEuler.Result.Z
   ) {template="DISPATCH_RigVMDispatch_SelectInt32(in Index,in Values,out Result)"}
   Remap = &RigVMFunction_MathDoubleRemap(
    TargetMaximum=1
    TargetMinimum=0
    bClamp=False
    Value=@DISPATCH_RigVMDispatch_SelectInt32.Result
    SourceMaximum=@n1.max
    SourceMinimum=@n1.min
   )
   Multiply = &RigVMFunction_MathDoubleMul(A=&RigVMFunction_MathDoubleClamp(Value=@Remap.Result).Result, B=@n1.premult)
   > RigVMFunctionReturnNode(outputWeight=@Multiply.Result)
  ]}
 RigVMFunctionLibrary {nodes [
   Aim = RigVMCollapseNode(ExecuteContext=())
   DeltaRotation = RigVMCollapseNode(ExecuteContext=(), Rotation Order=ZYX, premult=1, twistAxis=(X=1,Y=0,Z=0))
  ]}
 RigVMFunction_Sequence_2_ContainedGraph {nodes [
   n1 = RigVMFunctionEntryNode()
   n2 = &RigVMFunction_Sequence(ExecuteContext=@n1.ExecuteContext)
   n3 = &RigVMFunction_Sequence(ExecuteContext=@n2.B)
   n4 = &RigVMFunction_Sequence(ExecuteContext=@n3.B)
   RigVMFunctionReturnNode(A=@n2.A, B=@n3.A, C=@n4.A, D=@n4.B)
  ]}
 RigVMModel {nodes [
   RigUnit_BeginExecution = &RigUnit_BeginExecution()
   Aim = RigVMFunctionReferenceNode(ExecuteContext=@RigUnit_BeginExecution.ExecutePin)
   RigVMFunction_MathBoolToFloat = &RigVMFunction_MathBoolToFloat(Value=RigVMVariableNode() {variable=Is Skydiving}.Value)
   > &RigUnit_ModifyTransforms(
    ItemToModify="((Item=(Type=Bone,Name=\"base_hair_rt\"),Transform=(Rotation=(X=-0.0424,Y=-0.0447,Z=0.0242,W=0.9978),Translation=(X=0,Y=0,Z=2.5),Scale3D=(X=1,Y=1,Z=1))),(Item=(Type=Bone,Name=\"base_hair_lf\"),Transform=(Rotation=(X=-0.0466,Y=-0.1292,Z=-0.0316,W=0.99),Translat ...(truncated)"
    WeightMinimum=0
    Weight=@RigVMFunction_MathBoolToFloat.Result
   )
   > n1 = &RigUnit_SphericalPoseReader(
    ActiveRegionScaleFactors=(PositiveWidth=1,NegativeWidth=1,PositiveHeight=1,NegativeHeight=1)
    ActiveRegionSize=0.1
    Debug=(bDrawDebug=false,bDraw2D=false,DebugScale=25,DebugSegments=20,DebugThickness=0.25)
    DriverAxis=(X=1,Y=0,Z=0)
    DriverItem=(Type=Bone,Name="spine_02")
    FalloffRegionScaleFactors=(PositiveWidth=0.8,NegativeWidth=0.8,PositiveHeight=1,NegativeHeight=1)
    FalloffSize=0.2
    OptionalParentItem=(Type=Bone,Name="spine_04")
    RotationOffset=(X=90,Y=0,Z=235)
   )
   RigVMFunction_MathBoolNot = &RigVMFunction_MathBoolNot(Value=RigVMVariableNode() {variable=Is Skydiving}.Value)
   RigVMFunction_MathBoolAnd = &RigVMFunction_MathBoolAnd(B=@RigVMFunction_MathBoolNot.Result, A=RigVMVariableNode() {variable=Is Standing}.Value)
   Multiply = &RigVMFunction_MathDoubleMul(B=&RigVMFunction_MathBoolToFloat(Value=@RigVMFunction_MathBoolAnd.Result).Result, A=@n1.OutputParam)
   > &RigUnit_ModifyTransforms(
    ItemToModify=((Item=(Type=Bone,Name="base_hair_rt"),Transform=(Rotation=(X=0,Y=0,Z=0.0436,W=0.999),Translation=(X=0,Y=0,Z=0),Scale3D=(X=1,Y=1,Z=1))))
    WeightMinimum=0
    Weight=@Multiply.Result
   )
  ]}
]
hierarchy [
 root {loc=<0 0 0>, rot=<0 0 0 1>}
 pelvis {parent=root, loc=<0 -0.7071 93.1191>, rot=<0.0127 0.707 -0.0127 -0.707>}
 spine_01 {parent=pelvis, loc=<6.7494 -0.6388 0>, rot=<0 0 0.034 -0.9994>}
 spine_02 {parent=spine_01, loc=<7.2518 -0.035 0>, rot=<0 0 0.0238 -0.9997>}
 spine_03 {parent=spine_02, loc=<6.0104 0 0>, rot=<0 0 0.0149 -0.9999>}
 spine_04 {parent=spine_03, loc=<8.1024 0 0>, rot=<0 0 -0.0132 -0.9999>}
 spine_05 {parent=spine_04, loc=<10.0146 0 0>, rot=<0 0 -0.0015 -1>}
 neck_01 {parent=spine_05, loc=<11.3138 0.1529 0>, rot=<0 0 -0.2003 -0.9797>}
 neck_02 {parent=neck_01, loc=<4.5627 0 0>, rot=<0 0 0.0504 -0.9987>}
 head {parent=neck_02, loc=<5.3914 0 0>, rot=<0 0 0.115 -0.9934>}
 base_hair_lf {parent=head, loc=<8.8897 -6.6163 -2.7909>, rot=<-0.0035 -0.7061 0.7069 -0.0416>}
 dyn_hair_01_lf {parent=base_hair_lf, loc=<0 0 0>, rot=<0 0 0 1>}
 dyn_hair_02_lf {parent=dyn_hair_01_lf, loc=<14.9676 0 0>, rot=<0.0287 0.1617 -0.0837 -0.9829>}
 dyn_hair_03_lf {parent=dyn_hair_02_lf, loc=<9.5682 0 0>, rot=<-0.1374 -0.0697 -0.0097 -0.988>}
 dyn_hair_04_lf {parent=dyn_hair_03_lf, loc=<13.4946 -0.033 0.0597>, rot=<0.1362 0.0116 0.0074 -0.9906>}
 dyn_hair_05_lf {parent=dyn_hair_04_lf, loc=<11.7003 0 0>, rot=<-0.0168 -0.0006 0.0378 -0.9991>}
 dyn_hair_06_lf {parent=dyn_hair_05_lf, loc=<11.9552 0 0.4734>, rot=<0 0 0.069 -0.9976>}
 dyn_hair_07_lf {parent=dyn_hair_06_lf, loc=<8.2596 0 0>, rot=<0 0 0 1>}
 dyn_hairTopMain_lf {parent=dyn_hair_01_lf, loc=<1.8828 3.1665 -0.5984>, rot=<0.3246 0.0297 0.086 -0.9415>}
 base_hair_rt {parent=head, loc=<8.8898 -6.6163 2.7909>, rot=<0.0416 -0.7069 0.7061 0.0035>}
 dyn_hair_01_rt {parent=base_hair_rt, loc=<0 0 0>, rot=<0 0 0 1>}
 dyn_hair_02_rt {parent=dyn_hair_01_rt, loc=<14.968 0 0>, rot=<-0.0287 0.1617 0.0837 -0.9829>}
 dyn_hair_03_rt {parent=dyn_hair_02_rt, loc=<9.5681 0 0>, rot=<0.1374 -0.0697 0.0097 -0.988>}
 dyn_hair_04_rt {parent=dyn_hair_03_rt, loc=<13.495 0.033 0.0597>, rot=<-0.1361 0.0116 -0.0074 -0.9906>}
 dyn_hair_05_rt {parent=dyn_hair_04_rt, loc=<11.6998 -0.0001 0.0002>, rot=<0.0168 -0.0006 -0.0378 -0.9991>}
 dyn_hair_06_rt {parent=dyn_hair_05_rt, loc=<11.9549 -0.0001 0.4734>, rot=<0 0 -0.069 -0.9976>}
 dyn_hair_07_rt {parent=dyn_hair_06_rt, loc=<8.2597 0 0>, rot=<0 0 0 1>}
 dyn_hairTopMain_rt {parent=dyn_hair_01_rt, loc=<2.4882 -3.1931 -0.4329>, rot=<-0.3727 0.064 -0.1567 -0.9124>}
 base_hairTop01_cr {parent=head, loc=<11.4886 -6.9047 1.2444>, rot=<0.0875 -0.5993 0.7521 -0.2599>}
 dyn_hairTop01_cr {parent=base_hairTop01_cr, loc=<0 0 0>, rot=<0 0 0 1>}
 dyn_hairTop02_cr {parent=dyn_hairTop01_cr, loc=<4.0237 0 0>, rot=<0.0048 -0.1194 -0.1002 -0.9878>}
 base_faceFrame01_lf {parent=head, loc=<-2.2102 -0.487 -5.4661>, rot=<0.1134 -0.9864 0.0126 -0.1183>}
 dyn_faceFrame01_lf {parent=base_faceFrame01_lf, loc=<0 0 0>, rot=<0 0 0 1>}
 dyn_faceFrame02_lf {parent=dyn_faceFrame01_lf, loc=<4.6945 0 0>, rot=<-0.0201 0.3012 0.0635 -0.9512>}
 base_hairTop01_rt {parent=head, loc=<3.9253 1.0948 8.3285>, rot=<0.0567 0.018 0.9449 -0.3218>}
 dyn_hairTop01_rt {parent=base_hairTop01_rt, loc=<0 0 0>, rot=<0 0 0 1>}
 base_hairTop01_lf {parent=head, loc=<4.7286 1.2199 -7.9492>, rot=<0.2087 -0.9776 0.0193 0.0177>}
 dyn_hairTop01_lf {parent=base_hairTop01_lf, loc=<0 0 0>, rot=<0 0 0 1>}
 base_bangs01_lf {parent=head, loc=<9.8076 10.0102 -3.6014>, rot=<-0.0099 -0.9197 -0.3849 -0.0769>}
 dyn_bangs01_lf {parent=base_bangs01_lf, loc=<0 0 0>, rot=<0 0 0 1>}
 base_bangs02_lf {parent=head, loc=<7.9746 7.3551 -6.7272>, rot=<0.0497 -0.9867 -0.1547 0.0025>}
 dyn_bangs02_lf {parent=base_bangs02_lf, loc=<0 0 0>, rot=<0 0 0 1>}
 base_bangs02_rt {parent=head, loc=<8.588 7.732 6.8466>, rot=<-0.0553 0.285 0.9567 -0.0223>}
 dyn_bangs02_rt {parent=base_bangs02_rt, loc=<0 0 0>, rot=<0 0 0 1>}
 base_bangs01_rt {parent=head, loc=<9.6177 11.1971 2.9448>, rot=<-0.0007 0.6047 0.7958 -0.0308>}
 dyn_bangs01_rt {parent=base_bangs01_rt, loc=<0 0 0>, rot=<0 0 0 1>}
 base_lowerHairLick01_rt {parent=head, loc=<13.0252 9.3945 3.8373>, rot=<0.3227 0.1086 0.9386 -0.0556>}
 dyn_lowerHairLick01_rt {parent=base_lowerHairLick01_rt, loc=<0 0 0>, rot=<-0.2397 0 0 -0.9708>}
 dyn_lowerHairLick02_rt {parent=dyn_lowerHairLick01_rt, loc=<3.5686 0 0>, rot=<0.2182 -0.0803 -0.29 -0.9284>}
 base_hairLick01_rt {parent=head, loc=<16.1169 7.3407 1.0747>, rot=<0.497 0.0504 0.8239 -0.2675>}
 dyn_hairLick01_rt {parent=base_hairLick01_rt, loc=<0 0 0>, rot=<0 0 0 1>}
 dyn_hairLick02_rt {parent=dyn_hairLick01_rt, loc=<4.2377 0 0>, rot=<-0.073 -0.1803 0.1226 -0.9732>}
 dyn_hairLick03_rt {parent=dyn_hairLick02_rt, loc=<5.2697 0 0>, rot=<0.0946 -0.1738 0.1251 -0.9722>}
 base_hairLick01_lf {parent=head, loc=<15.2585 7.5803 -2.0551>, rot=<0.0469 -0.8825 -0.2418 -0.4008>}
 dyn_hairLick01_lf {parent=base_hairLick01_lf, loc=<0 0 0>, rot=<0 0 0 1>}
 dyn_hairLick02_lf {parent=dyn_hairLick01_lf, loc=<3.9387 0 0>, rot=<0.0166 -0.1621 -0.0186 -0.9865>}
 dyn_hairLick03_lf {parent=dyn_hairLick02_lf, loc=<4.4678 0 0>, rot=<0.0702 -0.0078 0.4091 -0.9098>}
 base_faceFrame01_rt {parent=head, loc=<-1.0324 -0.6805 5.3538>, rot=<0.0371 -0.132 0.9558 0.2601>}
 dyn_faceFrame01_rt {parent=base_faceFrame01_rt, loc=<0 0 0>, rot=<0 0 0 1>}
 dyn_faceFrame02_rt {parent=dyn_faceFrame01_rt, loc=<3.3175 0.7062 -0.0509>, rot=<-0.1818 -0.057 0.1153 -0.9749>}
 dyn_faceFrame03_rt {parent=dyn_faceFrame01_rt, loc=<2.6875 1.8983 1.7524>, rot=<-0.2604 0.2311 -0.5439 -0.7636>}
 base_hairLick_cr {parent=head, loc=<14.2632 8.7498 2.6121>, rot=<0.4959 0.2157 0.057 -0.8392>}
 dyn_hairLick_cr {parent=base_hairLick_cr, loc=<0 0 0>, rot=<0 0 0 1>}
 base_hairTop02_rt {parent=head, loc=<6.1779 -2.3814 8.5114>, rot=<0.4716 0.1099 0.8502 -0.2067>}
 dyn_hairTop02_rt {parent=base_hairTop02_rt, loc=<0 0 0>, rot=<0 0 0 1>}
 base_hairTop03_rt {parent=head, loc=<11.6819 -0.9902 7.9774>, rot=<0.1895 0.0116 0.953 -0.236>}
 dyn_hairTop03_rt {parent=base_hairTop03_rt, loc=<0 0 0>, rot=<0 0 0 1>}
 base_hairTop02_lf {parent=head, loc=<6.1937 -2.2063 -8.4828>, rot=<0.249 -0.9455 0.0595 -0.2011>}
 dyn_hairTop02_lf {parent=base_hairTop02_lf, loc=<0 0 0>, rot=<0 0 0 1>}
 base_jewel_cr {parent=head, loc=<0.4483 -7.6052 0.0036>, rot=<0.0475 -0.7055 0.7055 -0.0475>}
 aim_jewel_cr {parent=base_jewel_cr, loc=<0 0 0>, rot=<0 0 0 1>}
 base_earring_lf {parent=head, loc=<-0.8597 2.0226 -5.9354>, rot=<0.001 0.7071 0.7071 0.0034>}
 dyn_earring_lf {parent=base_earring_lf, loc=<0 0 0>, rot=<0 0 0 1>}
 aimtgt_jewel_cr {parent=head, loc=<-5.9705 -7.8433 0.0036>, rot=<0.5029 -0.4971 0.4971 -0.5029>}
 head_null {parent=neck_01, type=Null, loc=<9.9266 -0.5431 0>, rot=<0 0 -0.1649 0.9863>}
 base_head_aim_tgt {parent=spine_05, loc=<9.017 -11.714 0>, rot=<0.0616 -0.7044 -0.0616 -0.7044>}
 head_aim_tgt {parent=base_head_aim_tgt, loc=<0 0 0>, rot=<0 0 0 -1>}
 clavicle_l {parent=spine_05, loc=<4.3575 -0.1466 -2.3223>, rot=<0.0465 -0.7289 -0.009 -0.6829>}
 upperarm_l {parent=clavicle_l, loc=<15.5111 0 0>, rot=<0.0226 -0.3234 -0.0088 -0.9459>}
 lowerarm_l {parent=upperarm_l, loc=<26.6514 0 0>, rot=<0.0391 -0.0099 -0.2453 -0.9686>}
 clavicle_r {parent=spine_05, loc=<4.3575 -0.1466 2.3223>, rot=<0.7289 0.0465 -0.6829 0.009>}
 upperarm_r {parent=clavicle_r, loc=<-15.5111 0 0>, rot=<0.0226 -0.3234 -0.0088 -0.9459>}
 lowerarm_r {parent=upperarm_r, loc=<-26.6514 0 0>, rot=<0.0391 -0.0099 -0.2453 -0.9686>}
 neck_null {parent=spine_05, type=Null, loc=<0 0 0>, rot=<0 0 0 1>}
]
variables [
 Is Skydiving {type=bool, default=False}
 Is Standing {type=bool, default=False}
 IsAiming {type=bool, default=False}
 IsCrouchTargeting {type=bool, default=False}
 IsMeelee {type=bool, default=False}
]
