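# Post-process style secondary-motion AnimBP for the ChoiceCrave male character's face accessory part, animating hair, earring and hair jewel.
# Poses copied from parent and body meshes feed a state-driven Control Rig, then bone constraints.
# A bool selects between two neck-relative RigidBody sims, one with corrective offsets and its own physics asset for cape and backpack.
# Chained AnimDynamics then add spring jiggle, with cape-driven alpha on extra hair bones.
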AnimBlueprint /BRCosmetics/Characters/Player/Male/Medium/Bodies/M_MED_ChoiceCrave/Meshes/Parts/M_MED_ChoiceCrave_FaceAcc_AnimBP
target_skeleton=&/M_MED_ChoiceCrave_FaceAcc_Skeleton
graphs [
 AnimGraph {nodes [
   n1 = Anim Dynamics - Bone: dyn_earring_lf!(
    ComponentPose=Local To Component(LocalPose=Use cached pose 'PostPhysBody'())
    Alpha=1
    ComponentLinearAccScale:Vector="0.2,0.2,0.2"
    ComponentLinearVelScale:Vector="0.2,0.2,0.2"
    ExternalForce="0,0,0"
   ) -> (Pose) {
    angularDampingOverride=0.9999
    bOverrideAngularDamping=true
    bUseGravityOverride=true
    gravityScale=0.35
    linearSpringConstant=10000
    simulationSpace=RootRelative
    worldSpaceMinimumScale=0
    boundBone {boneName=dyn_earring_lf}
    componentAppliedLinearAccClamp=<20000 20000 20000>
    componentLinearAccScale=<0.2 0.2 0.2>
    componentLinearVelScale=<0.2 0.2 0.2>
    gravityOverride {z=-980}
    physicsBodyDefinitions [{
      boundBone {boneName=dyn_earring_lf}
      boxExtents=<18 15 15>
      constraintSetup {angularLimitsMax=<0 20 10>, angularLimitsMin=<0 -40 -10>, angularTarget=<1 0 0>}
      localJointOffset=<3 0 0>
     }]
    relativeSpaceBone {boneName=neck_02}
    simSpaceSettings {maxAngularAcceleration=20, maxAngularVelocity=10, simSpaceAngularAlpha=0.25}
   }
   n2 = Get Cape Dyn Alpha() -> (Cape Dyn Alpha:double)
   n3 = Anim Dynamics - Bone: dyn_hairExtra_05_lf!(
    ComponentPose=@n1
    Alpha=@n2
    ComponentLinearAccScale:Vector="0.2,0.2,0.2"
    ComponentLinearVelScale:Vector="0.2,0.2,0.2"
    ExternalForce="0,0,0"
   ) -> (Pose) {
    angularDampingOverride=0.9999
    angularSpringConstant=4000
    bAngularSpring=true
    bOverrideAngularDamping=true
    bUseGravityOverride=true
    gravityScale=0.35
    linearSpringConstant=10000
    simulationSpace=RootRelative
    worldSpaceMinimumScale=0
    boundBone {boneName=dyn_hairExtra_05_lf}
    componentAppliedLinearAccClamp=<20000 20000 20000>
    componentLinearAccScale=<0.2 0.2 0.2>
    componentLinearVelScale=<0.2 0.2 0.2>
    gravityOverride {z=-200}
    physicsBodyDefinitions [{
      boundBone {boneName=dyn_hairExtra_05_lf}
      boxExtents=<16 15 15>
      constraintSetup {angularLimitsMax=<0 10 -2>, angularLimitsMin=<0 -10 -5>, angularTarget=<1 0 0>}
      localJointOffset=<2 0 0>
     }]
    relativeSpaceBone {boneName=neck_02}
    simSpaceSettings {maxAngularAcceleration=20, maxAngularVelocity=10, simSpaceAngularAlpha=0.25}
   }
   n4 = Anim Dynamics - Bone: dyn_hairExtra_06_lf!(
    ComponentPose=@n3
    Alpha=@n2
    ComponentLinearAccScale:Vector="0.2,0.2,0.2"
    ComponentLinearVelScale:Vector="0.2,0.2,0.2"
    ExternalForce="0,0,0"
   ) -> (Pose) {
    angularDampingOverride=0.9999
    angularSpringConstant=4000
    bAngularSpring=true
    bOverrideAngularDamping=true
    bUseGravityOverride=true
    gravityScale=0.35
    linearSpringConstant=10000
    simulationSpace=RootRelative
    worldSpaceMinimumScale=0
    boundBone {boneName=dyn_hairExtra_06_lf}
    componentAppliedLinearAccClamp=<20000 20000 20000>
    componentLinearAccScale=<0.2 0.2 0.2>
    componentLinearVelScale=<0.2 0.2 0.2>
    gravityOverride {z=-200}
    physicsBodyDefinitions [{
      boundBone {boneName=dyn_hairExtra_06_lf}
      boxExtents=<16 15 15>
      constraintSetup {angularLimitsMax=<0 15 2.5>, angularLimitsMin=<0 -15 -2.5>, angularTarget=<1 0 0>}
      localJointOffset=<2 0 0>
     }]
    relativeSpaceBone {boneName=neck_02}
    simSpaceSettings {maxAngularAcceleration=20, maxAngularVelocity=10, simSpaceAngularAlpha=0.25}
   }
   n5 = Anim Dynamics - Bone: dyn_hairExtra_05_rt!(
    ComponentPose=@n4
    Alpha=@n2
    ComponentLinearAccScale:Vector="0.2,0.2,0.2"
    ComponentLinearVelScale:Vector="0.2,0.2,0.2"
    ExternalForce="0,0,0"
   ) -> (Pose) {
    angularDampingOverride=0.9999
    angularSpringConstant=4000
    bAngularSpring=true
    bOverrideAngularDamping=true
    bUseGravityOverride=true
    gravityScale=0.35
    linearSpringConstant=10000
    simulationSpace=RootRelative
    worldSpaceMinimumScale=0
    boundBone {boneName=dyn_hairExtra_05_rt}
    componentAppliedLinearAccClamp=<20000 20000 20000>
    componentLinearAccScale=<0.2 0.2 0.2>
    componentLinearVelScale=<0.2 0.2 0.2>
    gravityOverride {z=-200}
    physicsBodyDefinitions [{
      boundBone {boneName=dyn_hairExtra_05_rt}
      boxExtents=<16 15 15>
      constraintSetup {angularLimitsMax=<0 10 5>, angularLimitsMin=<0 -10 2>, angularTarget=<1 0 0>}
      localJointOffset=<2 0 0>
     }]
    relativeSpaceBone {boneName=neck_02}
    simSpaceSettings {maxAngularAcceleration=20, maxAngularVelocity=10, simSpaceAngularAlpha=0.25}
   }
   n6 = Anim Dynamics - Bone: dyn_hairExtra_06_rt!(
    ComponentPose=@n5
    Alpha=@n2
    ComponentLinearAccScale:Vector="0.2,0.2,0.2"
    ComponentLinearVelScale:Vector="0.2,0.2,0.2"
    ExternalForce="0,0,0"
   ) -> (Pose) {
    angularDampingOverride=0.9999
    angularSpringConstant=4000
    bAngularSpring=true
    bOverrideAngularDamping=true
    bUseGravityOverride=true
    gravityScale=0.35
    linearSpringConstant=10000
    simulationSpace=RootRelative
    worldSpaceMinimumScale=0
    boundBone {boneName=dyn_hairExtra_06_rt}
    componentAppliedLinearAccClamp=<20000 20000 20000>
    componentLinearAccScale=<0.2 0.2 0.2>
    componentLinearVelScale=<0.2 0.2 0.2>
    gravityOverride {z=-200}
    physicsBodyDefinitions [{
      boundBone {boneName=dyn_hairExtra_06_rt}
      boxExtents=<16 15 15>
      constraintSetup {angularLimitsMax=<0 15 2.5>, angularLimitsMin=<0 -15 -2.5>, angularTarget=<1 0 0>}
      localJointOffset=<2 0 0>
     }]
    relativeSpaceBone {boneName=neck_02}
    simSpaceSettings {maxAngularAcceleration=20, maxAngularVelocity=10, simSpaceAngularAlpha=0.25}
   }
   n7 = Anim Dynamics - Bone: dyn_hairTopMain_lf!(
    ComponentPose=@n6
    Alpha=1
    ComponentLinearAccScale:Vector="0.2,0.2,0.2"
    ComponentLinearVelScale:Vector="0.2,0.2,0.2"
    ExternalForce="0,0,0"
   ) -> (Pose) {
    angularDampingOverride=0.9999
    angularSpringConstant=4000
    bAngularSpring=true
    bOverrideAngularDamping=true
    bUseGravityOverride=true
    gravityScale=0.35
    linearSpringConstant=10000
    simulationSpace=RootRelative
    worldSpaceMinimumScale=0
    boundBone {boneName=dyn_hairTopMain_lf}
    componentAppliedLinearAccClamp=<20000 20000 20000>
    componentLinearAccScale=<0.2 0.2 0.2>
    componentLinearVelScale=<0.2 0.2 0.2>
    gravityOverride {z=-200}
    physicsBodyDefinitions [{
      boundBone {boneName=dyn_hairTopMain_lf}
      boxExtents=<20 17 17>
      constraintSetup {angularLimitsMax=<0 0 20>, angularLimitsMin=<0 -30 -20>, angularTarget=<1 0 0>}
      localJointOffset=<2 0 0>
     }]
    relativeSpaceBone {boneName=neck_02}
    simSpaceSettings {maxAngularAcceleration=20, maxAngularVelocity=10, simSpaceAngularAlpha=0.25}
   }
   n8 = Anim Dynamics - Bone: dyn_hairTopMain_rt!(
    ComponentPose=@n7
    Alpha=1
    ComponentLinearAccScale:Vector="0.2,0.2,0.2"
    ComponentLinearVelScale:Vector="0.2,0.2,0.2"
    ExternalForce="0,0,0"
   ) -> (Pose) {
    angularDampingOverride=0.9999
    angularSpringConstant=4000
    bAngularSpring=true
    bOverrideAngularDamping=true
    bUseGravityOverride=true
    gravityScale=0.35
    linearSpringConstant=10000
    simulationSpace=RootRelative
    worldSpaceMinimumScale=0
    boundBone {boneName=dyn_hairTopMain_rt}
    componentAppliedLinearAccClamp=<20000 20000 20000>
    componentLinearAccScale=<0.2 0.2 0.2>
    componentLinearVelScale=<0.2 0.2 0.2>
    gravityOverride {z=-200}
    physicsBodyDefinitions [{
      boundBone {boneName=dyn_hairTopMain_rt}
      boxExtents=<24 22 22>
      constraintSetup {angularLimitsMax=<0 0 20>, angularLimitsMin=<0 -30 -20>, angularTarget=<1 0 0>}
      localJointOffset=<2 0 0>
     }]
    relativeSpaceBone {boneName=neck_02}
    simSpaceSettings {maxAngularAcceleration=20, maxAngularVelocity=10, simSpaceAngularAlpha=0.25}
   }
   n9 = Anim Dynamics - Bone: dyn_hairTop01_cr!(
    ComponentPose=@n8
    Alpha=1
    ComponentLinearAccScale:Vector="0.2,0.2,0.2"
    ComponentLinearVelScale:Vector="0.2,0.2,0.2"
    ExternalForce="0,0,0"
   ) -> (Pose) {
    angularDampingOverride=0.9999
    angularSpringConstant=4000
    bAngularSpring=true
    bOverrideAngularDamping=true
    bUseGravityOverride=true
    gravityScale=0.35
    linearSpringConstant=10000
    simulationSpace=RootRelative
    worldSpaceMinimumScale=0
    boundBone {boneName=dyn_hairTop01_cr}
    componentAppliedLinearAccClamp=<20000 20000 20000>
    componentLinearAccScale=<0.2 0.2 0.2>
    componentLinearVelScale=<0.2 0.2 0.2>
    gravityOverride {z=-200}
    physicsBodyDefinitions [{
      boundBone {boneName=dyn_hairTop01_cr}
      boxExtents=<24 22 22>
      constraintSetup {angularLimitsMax=<0 0 0>, angularLimitsMin=<0 -20 0>, angularTarget=<1 0 0>}
      localJointOffset=<2 0 0>
     }]
    relativeSpaceBone {boneName=neck_02}
    simSpaceSettings {maxAngularAcceleration=20, maxAngularVelocity=10, simSpaceAngularAlpha=0.25}
   }
   n10 = Anim Dynamics - Bone: dyn_hairTop02_cr!(
    ComponentPose=@n9
    Alpha=1
    ComponentLinearAccScale:Vector="0.2,0.2,0.2"
    ComponentLinearVelScale:Vector="0.2,0.2,0.2"
    ExternalForce="0,0,0"
   ) -> (Pose) {
    angularDampingOverride=0.9999
    angularSpringConstant=4000
    bAngularSpring=true
    bOverrideAngularDamping=true
    bUseGravityOverride=true
    gravityScale=0.35
    linearSpringConstant=10000
    simulationSpace=RootRelative
    worldSpaceMinimumScale=0
    boundBone {boneName=dyn_hairTop02_cr}
    componentAppliedLinearAccClamp=<20000 20000 20000>
    componentLinearAccScale=<0.2 0.2 0.2>
    componentLinearVelScale=<0.2 0.2 0.2>
    gravityOverride {z=-200}
    physicsBodyDefinitions [{
      boundBone {boneName=dyn_hairTop02_cr}
      boxExtents=<20 17 17>
      constraintSetup {angularLimitsMax=<0 0 2>, angularLimitsMin=<0 -20 -2>, angularTarget=<1 0 0>}
      localJointOffset=<2 0 0>
     }]
    relativeSpaceBone {boneName=neck_02}
    simSpaceSettings {maxAngularAcceleration=20, maxAngularVelocity=10, simSpaceAngularAlpha=0.25}
   }
   n11 = Anim Dynamics - Bone: dyn_hairTop02_lf!(
    ComponentPose=@n10
    Alpha=1
    ComponentLinearAccScale:Vector="0.2,0.2,0.2"
    ComponentLinearVelScale:Vector="0.2,0.2,0.2"
    ExternalForce="0,0,0"
   ) -> (Pose) {
    angularDampingOverride=0.9999
    angularSpringConstant=4000
    bAngularSpring=true
    bOverrideAngularDamping=true
    bUseGravityOverride=true
    gravityScale=0.35
    linearSpringConstant=10000
    simulationSpace=RootRelative
    worldSpaceMinimumScale=0
    boundBone {boneName=dyn_hairTop02_lf}
    componentAppliedLinearAccClamp=<20000 20000 20000>
    componentLinearAccScale=<0.2 0.2 0.2>
    componentLinearVelScale=<0.2 0.2 0.2>
    gravityOverride {z=-200}
    physicsBodyDefinitions [{
      boundBone {boneName=dyn_hairTop02_lf}
      boxExtents=<20 17 17>
      constraintSetup {angularLimitsMax=<0 20 20>, angularLimitsMin=<0 -20 -20>, angularTarget=<1 0 0>}
      localJointOffset=<2 0 0>
     }]
    relativeSpaceBone {boneName=neck_02}
    simSpaceSettings {maxAngularAcceleration=20, maxAngularVelocity=10, simSpaceAngularAlpha=0.25}
   }
   n12 = Anim Dynamics - Bone: dyn_hairTop01_lf!(
    ComponentPose=@n11
    Alpha=1
    ComponentLinearAccScale:Vector="0.2,0.2,0.2"
    ComponentLinearVelScale:Vector="0.2,0.2,0.2"
    ExternalForce="0,0,0"
   ) -> (Pose) {
    angularDampingOverride=0.9999
    angularSpringConstant=4000
    bAngularSpring=true
    bOverrideAngularDamping=true
    bUseGravityOverride=true
    gravityScale=0.35
    linearSpringConstant=10000
    simulationSpace=RootRelative
    worldSpaceMinimumScale=0
    boundBone {boneName=dyn_hairTop01_lf}
    componentAppliedLinearAccClamp=<20000 20000 20000>
    componentLinearAccScale=<0.2 0.2 0.2>
    componentLinearVelScale=<0.2 0.2 0.2>
    gravityOverride {z=-200}
    physicsBodyDefinitions [{
      boundBone {boneName=dyn_hairTop01_lf}
      boxExtents=<24 22 22>
      constraintSetup {angularLimitsMax=<0 0 10>, angularLimitsMin=<0 -25 -25>, angularTarget=<1 0 0>}
      localJointOffset=<2 0 0>
     }]
    relativeSpaceBone {boneName=neck_02}
    simSpaceSettings {maxAngularAcceleration=20, maxAngularVelocity=10, simSpaceAngularAlpha=0.25}
   }
   n13 = Anim Dynamics - Bone: dyn_hairTop03_rt!(
    ComponentPose=@n12
    Alpha=1
    ComponentLinearAccScale:Vector="0.2,0.2,0.2"
    ComponentLinearVelScale:Vector="0.2,0.2,0.2"
    ExternalForce="0,0,0"
   ) -> (Pose) {
    angularDampingOverride=0.9999
    angularSpringConstant=4000
    bAngularSpring=true
    bOverrideAngularDamping=true
    bUseGravityOverride=true
    gravityScale=0.35
    linearSpringConstant=10000
    simulationSpace=RootRelative
    worldSpaceMinimumScale=0
    boundBone {boneName=dyn_hairTop03_rt}
    componentAppliedLinearAccClamp=<20000 20000 20000>
    componentLinearAccScale=<0.2 0.2 0.2>
    componentLinearVelScale=<0.2 0.2 0.2>
    gravityOverride {z=-200}
    physicsBodyDefinitions [{
      boundBone {boneName=dyn_hairTop03_rt}
      boxExtents=<17 15 15>
      constraintSetup {angularLimitsMax=<0 0 10>, angularLimitsMin=<0 -20 -10>, angularTarget=<1 0 0>}
      localJointOffset=<2 0 0>
     }]
    relativeSpaceBone {boneName=neck_02}
    simSpaceSettings {maxAngularAcceleration=20, maxAngularVelocity=10, simSpaceAngularAlpha=0.25}
   }
   n14 = Anim Dynamics - Bone: dyn_hairTop02_rt!(
    ComponentPose=@n13
    Alpha=1
    ComponentLinearAccScale:Vector="0.2,0.2,0.2"
    ComponentLinearVelScale:Vector="0.2,0.2,0.2"
    ExternalForce="0,0,0"
   ) -> (Pose) {
    angularDampingOverride=0.9999
    angularSpringConstant=4000
    bAngularSpring=true
    bOverrideAngularDamping=true
    bUseGravityOverride=true
    gravityScale=0.35
    linearSpringConstant=10000
    simulationSpace=RootRelative
    worldSpaceMinimumScale=0
    boundBone {boneName=dyn_hairTop02_rt}
    componentAppliedLinearAccClamp=<20000 20000 20000>
    componentLinearAccScale=<0.2 0.2 0.2>
    componentLinearVelScale=<0.2 0.2 0.2>
    gravityOverride {z=-200}
    physicsBodyDefinitions [{
      boundBone {boneName=dyn_hairTop02_rt}
      boxExtents=<24 22 22>
      constraintSetup {angularLimitsMax=<0 20 20>, angularLimitsMin=<0 -20 -20>, angularTarget=<1 0 0>}
      localJointOffset=<2 0 0>
     }]
    relativeSpaceBone {boneName=neck_02}
    simSpaceSettings {maxAngularAcceleration=20, maxAngularVelocity=10, simSpaceAngularAlpha=0.25}
   }
   n15 = Anim Dynamics - Bone: dyn_hairTop01_rt!(
    ComponentPose=@n14
    Alpha=1
    ComponentLinearAccScale:Vector="0.2,0.2,0.2"
    ComponentLinearVelScale:Vector="0.2,0.2,0.2"
    ExternalForce="0,0,0"
   ) -> (Pose) {
    angularDampingOverride=0.9999
    angularSpringConstant=4000
    bAngularSpring=true
    bOverrideAngularDamping=true
    bUseGravityOverride=true
    gravityScale=0.35
    linearSpringConstant=10000
    simulationSpace=RootRelative
    worldSpaceMinimumScale=0
    boundBone {boneName=dyn_hairTop01_rt}
    componentAppliedLinearAccClamp=<20000 20000 20000>
    componentLinearAccScale=<0.2 0.2 0.2>
    componentLinearVelScale=<0.2 0.2 0.2>
    gravityOverride {z=-200}
    physicsBodyDefinitions [{
      boundBone {boneName=dyn_hairTop01_rt}
      boxExtents=<24 22 22>
      constraintSetup {angularLimitsMax=<0 10 15>, angularLimitsMin=<0 -20 -15>, angularTarget=<1 0 0>}
      localJointOffset=<2 0 0>
     }]
    relativeSpaceBone {boneName=neck_02}
    simSpaceSettings {maxAngularAcceleration=20, maxAngularVelocity=10, simSpaceAngularAlpha=0.25}
   }
   n16 = Anim Dynamics - Bone: dyn_hairTop02_cr(ComponentPose=@n15, Alpha=1, ComponentLinearAccScale="0.2,0.2,0.2", ComponentLinearVelScale="0.2,0.2,0.2", ExternalForce="0,0,0") {
    angularDampingOverride=0.9999
    angularSpringConstant=4000
    bAngularSpring=true
    bOverrideAngularDamping=true
    bUseGravityOverride=true
    gravityScale=0.35
    linearSpringConstant=10000
    simulationSpace=RootRelative
    worldSpaceMinimumScale=0
    boundBone {boneName=dyn_hairTop02_cr}
    componentAppliedLinearAccClamp=<20000 20000 20000>
    componentLinearAccScale=<0.2 0.2 0.2>
    componentLinearVelScale=<0.2 0.2 0.2>
    gravityOverride {z=-200}
    physicsBodyDefinitions [{
      boundBone {boneName=dyn_hairTop02_cr}
      boxExtents=<20 17 17>
      constraintSetup {angularLimitsMax=<0 0 8>, angularLimitsMin=<0 -8 -8>, angularTarget=<1 0 0>}
      localJointOffset=<2 0 0>
     }]
    relativeSpaceBone {boneName=neck_02}
    simSpaceSettings {maxAngularAcceleration=20, maxAngularVelocity=10, simSpaceAngularAlpha=0.25}
   }
   n17 = Anim Dynamics - Bone: dyn_bangs01_lf!(
    ComponentPose=@n16
    Alpha=1
    ComponentLinearAccScale:Vector="0.2,0.2,0.2"
    ComponentLinearVelScale:Vector="0.2,0.2,0.2"
    ExternalForce="0,0,0"
   ) -> (Pose) {
    angularDampingOverride=0.9999
    angularSpringConstant=4000
    bAngularSpring=true
    bOverrideAngularDamping=true
    bUseGravityOverride=true
    gravityScale=0.35
    linearSpringConstant=10000
    simulationSpace=RootRelative
    worldSpaceMinimumScale=0
    boundBone {boneName=dyn_bangs01_lf}
    componentAppliedLinearAccClamp=<20000 20000 20000>
    componentLinearAccScale=<0.2 0.2 0.2>
    componentLinearVelScale=<0.2 0.2 0.2>
    gravityOverride {z=-200}
    physicsBodyDefinitions [{
      boundBone {boneName=dyn_bangs01_lf}
      boxExtents=<20 20 20>
      constraintSetup {angularLimitsMax=<0 0 10>, angularLimitsMin=<0 -25 -10>, angularTarget=<1 0 0>}
      localJointOffset=<2 0 0>
     }]
    relativeSpaceBone {boneName=neck_02}
    simSpaceSettings {maxAngularAcceleration=20, maxAngularVelocity=10, simSpaceAngularAlpha=0.25}
   }
   n18 = Anim Dynamics - Bone: dyn_bangs02_lf!(
    ComponentPose=@n17
    Alpha=1
    ComponentLinearAccScale:Vector="0.2,0.2,0.2"
    ComponentLinearVelScale:Vector="0.2,0.2,0.2"
    ExternalForce="0,0,0"
   ) -> (Pose) {
    angularDampingOverride=0.9999
    angularSpringConstant=4000
    bAngularSpring=true
    bOverrideAngularDamping=true
    bUseGravityOverride=true
    gravityScale=0.35
    linearSpringConstant=10000
    simulationSpace=RootRelative
    worldSpaceMinimumScale=0
    boundBone {boneName=dyn_bangs02_lf}
    componentAppliedLinearAccClamp=<20000 20000 20000>
    componentLinearAccScale=<0.2 0.2 0.2>
    componentLinearVelScale=<0.2 0.2 0.2>
    gravityOverride {z=-200}
    physicsBodyDefinitions [{
      boundBone {boneName=dyn_bangs02_lf}
      boxExtents=<20 20 20>
      constraintSetup {angularLimitsMax=<0 0 10>, angularLimitsMin=<0 -25 -10>, angularTarget=<1 0 0>}
      localJointOffset=<2 0 0>
     }]
    relativeSpaceBone {boneName=neck_02}
    simSpaceSettings {maxAngularAcceleration=20, maxAngularVelocity=10, simSpaceAngularAlpha=0.25}
   }
   n19 = Anim Dynamics - Bone: dyn_bangs01_rt!(
    ComponentPose=@n18
    Alpha=1
    ComponentLinearAccScale:Vector="0.2,0.2,0.2"
    ComponentLinearVelScale:Vector="0.2,0.2,0.2"
    ExternalForce="0,0,0"
   ) -> (Pose) {
    angularDampingOverride=0.9999
    angularSpringConstant=4000
    bAngularSpring=true
    bOverrideAngularDamping=true
    bUseGravityOverride=true
    gravityScale=0.35
    linearSpringConstant=10000
    simulationSpace=RootRelative
    worldSpaceMinimumScale=0
    boundBone {boneName=dyn_bangs01_rt}
    componentAppliedLinearAccClamp=<20000 20000 20000>
    componentLinearAccScale=<0.2 0.2 0.2>
    componentLinearVelScale=<0.2 0.2 0.2>
    gravityOverride {z=-200}
    physicsBodyDefinitions [{
      boundBone {boneName=dyn_bangs01_rt}
      boxExtents=<20 20 20>
      constraintSetup {angularLimitsMax=<0 0 10>, angularLimitsMin=<0 -25 -10>, angularTarget=<1 0 0>}
      localJointOffset=<2 0 0>
     }]
    relativeSpaceBone {boneName=neck_02}
    simSpaceSettings {maxAngularAcceleration=20, maxAngularVelocity=10, simSpaceAngularAlpha=0.25}
   }
   n20 = Anim Dynamics - Bone: dyn_bangs02_rt!(
    ComponentPose=@n19
    Alpha=1
    ComponentLinearAccScale:Vector="0.2,0.2,0.2"
    ComponentLinearVelScale:Vector="0.2,0.2,0.2"
    ExternalForce="0,0,0"
   ) -> (Pose) {
    angularDampingOverride=0.9999
    angularSpringConstant=4000
    bAngularSpring=true
    bOverrideAngularDamping=true
    bUseGravityOverride=true
    gravityScale=0.35
    linearSpringConstant=10000
    simulationSpace=RootRelative
    worldSpaceMinimumScale=0
    boundBone {boneName=dyn_bangs02_rt}
    componentAppliedLinearAccClamp=<20000 20000 20000>
    componentLinearAccScale=<0.2 0.2 0.2>
    componentLinearVelScale=<0.2 0.2 0.2>
    gravityOverride {z=-200}
    physicsBodyDefinitions [{
      boundBone {boneName=dyn_bangs02_rt}
      boxExtents=<20 20 20>
      constraintSetup {angularLimitsMax=<0 0 8>, angularLimitsMin=<0 -25 -4>, angularTarget=<1 0 0>}
      localJointOffset=<2 0 0>
     }]
    relativeSpaceBone {boneName=neck_02}
    simSpaceSettings {maxAngularAcceleration=20, maxAngularVelocity=10, simSpaceAngularAlpha=0.25}
   }
   n21 = Anim Dynamics - Bone: dyn_hairLick_cr!(
    ComponentPose=@n20
    Alpha=1
    ComponentLinearAccScale:Vector="0.2,0.2,0.2"
    ComponentLinearVelScale:Vector="0.2,0.2,0.2"
    ExternalForce="0,0,0"
   ) -> (Pose) {
    angularDampingOverride=0.9999
    angularSpringConstant=4000
    bAngularSpring=true
    bOverrideAngularDamping=true
    bUseGravityOverride=true
    gravityScale=0.35
    linearSpringConstant=10000
    simulationSpace=RootRelative
    worldSpaceMinimumScale=0
    boundBone {boneName=dyn_hairLick_cr}
    componentAppliedLinearAccClamp=<20000 20000 20000>
    componentLinearAccScale=<0.2 0.2 0.2>
    componentLinearVelScale=<0.2 0.2 0.2>
    gravityOverride {z=-200}
    physicsBodyDefinitions [{
      boundBone {boneName=dyn_hairLick_cr}
      boxExtents=<17 17 17>
      constraintSetup {angularLimitsMax=<0 20 20>, angularLimitsMin=<0 -20 -20>, angularTarget=<1 0 0>}
      localJointOffset=<1 0 0>
     }]
    relativeSpaceBone {boneName=base_hairLick_cr}
    simSpaceSettings {maxAngularAcceleration=20, maxAngularVelocity=10, simSpaceAngularAlpha=0.25}
   }
   n22 = Anim Dynamics - Bone: dyn_hairLick02_rt!(
    ComponentPose=@n21
    Alpha=1
    ComponentLinearAccScale:Vector="0.2,0.2,0.2"
    ComponentLinearVelScale:Vector="0.2,0.2,0.2"
    ExternalForce="0,0,0"
   ) -> (Pose) {
    angularDampingOverride=0.9999
    angularSpringConstant=4000
    bAngularSpring=true
    bOverrideAngularDamping=true
    bUseGravityOverride=true
    gravityScale=0.35
    linearSpringConstant=10000
    simulationSpace=RootRelative
    worldSpaceMinimumScale=0
    boundBone {boneName=dyn_hairLick02_rt}
    componentAppliedLinearAccClamp=<20000 20000 20000>
    componentLinearAccScale=<0.2 0.2 0.2>
    componentLinearVelScale=<0.2 0.2 0.2>
    gravityOverride {z=-200}
    physicsBodyDefinitions [{
      boundBone {boneName=dyn_hairLick02_rt}
      boxExtents=<18 18 18>
      constraintSetup {angularLimitsMax=<0 0 8>, angularLimitsMin=<0 -20 -8>, angularTarget=<1 0 0>}
      localJointOffset=<3 0 0>
     }]
    relativeSpaceBone {boneName=neck_02}
    simSpaceSettings {maxAngularAcceleration=20, maxAngularVelocity=10, simSpaceAngularAlpha=0.25}
   }
   n23 = Anim Dynamics - Bone: dyn_hairLick03_rt!(
    ComponentPose=@n22
    Alpha=1
    ComponentLinearAccScale:Vector="0.2,0.2,0.2"
    ComponentLinearVelScale:Vector="0.2,0.2,0.2"
    ExternalForce="0,0,0"
   ) -> (Pose) {
    angularDampingOverride=0.9999
    angularSpringConstant=4000
    bAngularSpring=true
    bOverrideAngularDamping=true
    bUseGravityOverride=true
    gravityScale=0.35
    linearSpringConstant=10000
    simulationSpace=RootRelative
    worldSpaceMinimumScale=0
    boundBone {boneName=dyn_hairLick03_rt}
    componentAppliedLinearAccClamp=<20000 20000 20000>
    componentLinearAccScale=<0.2 0.2 0.2>
    componentLinearVelScale=<0.2 0.2 0.2>
    gravityOverride {z=-200}
    physicsBodyDefinitions [{
      boundBone {boneName=dyn_hairLick03_rt}
      boxExtents=<18 18 18>
      constraintSetup {angularLimitsMax=<0 0 15>, angularLimitsMin=<0 -25 -15>, angularTarget=<1 0 0>}
      localJointOffset=<3 0 0>
     }]
    relativeSpaceBone {boneName=neck_02}
    simSpaceSettings {maxAngularAcceleration=20, maxAngularVelocity=10, simSpaceAngularAlpha=0.25}
   }
   n24 = Anim Dynamics - Bone: dyn_hairLick02_lf!(
    ComponentPose=@n23
    Alpha=1
    ComponentLinearAccScale:Vector="0.2,0.2,0.2"
    ComponentLinearVelScale:Vector="0.2,0.2,0.2"
    ExternalForce="0,0,0"
   ) -> (Pose) {
    angularDampingOverride=0.9999
    angularSpringConstant=4000
    bAngularSpring=true
    bOverrideAngularDamping=true
    bUseGravityOverride=true
    gravityScale=0.35
    linearSpringConstant=10000
    simulationSpace=RootRelative
    worldSpaceMinimumScale=0
    boundBone {boneName=dyn_hairLick02_lf}
    componentAppliedLinearAccClamp=<20000 20000 20000>
    componentLinearAccScale=<0.2 0.2 0.2>
    componentLinearVelScale=<0.2 0.2 0.2>
    gravityOverride {z=-200}
    physicsBodyDefinitions [{
      boundBone {boneName=dyn_hairLick02_lf}
      boxExtents=<18 18 18>
      constraintSetup {angularLimitsMax=<0 0 8>, angularLimitsMin=<0 -20 -8>, angularTarget=<1 0 0>}
      localJointOffset=<3 0 0>
     }]
    relativeSpaceBone {boneName=neck_02}
    simSpaceSettings {maxAngularAcceleration=20, maxAngularVelocity=10, simSpaceAngularAlpha=0.25}
   }
   n25 = Anim Dynamics - Bone: dyn_hairLick03_lf!(
    ComponentPose=@n24
    Alpha=1
    ComponentLinearAccScale:Vector="0.2,0.2,0.2"
    ComponentLinearVelScale:Vector="0.2,0.2,0.2"
    ExternalForce="0,0,0"
   ) -> (Pose) {
    angularDampingOverride=0.9999
    angularSpringConstant=4000
    bAngularSpring=true
    bOverrideAngularDamping=true
    bUseGravityOverride=true
    gravityScale=0.35
    linearSpringConstant=10000
    simulationSpace=RootRelative
    worldSpaceMinimumScale=0
    boundBone {boneName=dyn_hairLick03_lf}
    componentAppliedLinearAccClamp=<20000 20000 20000>
    componentLinearAccScale=<0.2 0.2 0.2>
    componentLinearVelScale=<0.2 0.2 0.2>
    gravityOverride {z=-200}
    physicsBodyDefinitions [{
      boundBone {boneName=dyn_hairLick03_lf}
      boxExtents=<18 18 18>
      constraintSetup {angularLimitsMax=<0 0 15>, angularLimitsMin=<0 -25 -15>, angularTarget=<1 0 0>}
      localJointOffset=<3 0 0>
     }]
    relativeSpaceBone {boneName=neck_02}
    simSpaceSettings {maxAngularAcceleration=20, maxAngularVelocity=10, simSpaceAngularAlpha=0.25}
   }
   n26 = Anim Dynamics - Bone: dyn_lowerHairLick01_rt!(
    ComponentPose=@n25
    Alpha=1
    ComponentLinearAccScale:Vector="0.2,0.2,0.2"
    ComponentLinearVelScale:Vector="0.2,0.2,0.2"
    ExternalForce="0,0,0"
   ) -> (Pose) {
    angularDampingOverride=0.9999
    angularSpringConstant=4000
    bAngularSpring=true
    bOverrideAngularDamping=true
    bUseGravityOverride=true
    gravityScale=0.35
    linearSpringConstant=10000
    simulationSpace=RootRelative
    worldSpaceMinimumScale=0
    boundBone {boneName=dyn_lowerHairLick01_rt}
    componentAppliedLinearAccClamp=<20000 20000 20000>
    componentLinearAccScale=<0.2 0.2 0.2>
    componentLinearVelScale=<0.2 0.2 0.2>
    gravityOverride {z=-200}
    physicsBodyDefinitions [{
      boundBone {boneName=dyn_lowerHairLick01_rt}
      boxExtents=<18 18 18>
      constraintSetup {angularLimitsMax=<0 -3 8>, angularLimitsMin=<0 -20 -8>, angularTarget=<1 0 0>}
      localJointOffset=<3 0 0>
     }]
    relativeSpaceBone {boneName=neck_02}
    simSpaceSettings {maxAngularAcceleration=20, maxAngularVelocity=10, simSpaceAngularAlpha=0.25}
   }
   Output Pose(Result=Component To Local(ComponentPose=@n26)) {layerGroup=None}
   n27 = Layered blend per bone(BasePose=Copy Pose From Mesh() {bUseAttachedParent=true}, BlendPoses_0=Copy Pose From Mesh(SourceMeshComponent=Get Body_Mesh())) {layerSetup [{branchFilters [
       {boneName=clavicle_l}
       {boneName=clavicle_r}
       {boneName=upperarm_r}
       {boneName=upperarm_l}
       {boneName=base_coat_back_rt}
       {boneName=base_coat_back_lf}
       {boneName=base_coat_straps_back_rt_offset}
       {boneName=base_coat_straps_back_lf_offset}
       {boneName=dyn_coat_05_lf_03}
       {boneName=dyn_coat_05_rt_03}
       {boneName=dyn_coat_05_lf_01}
       {boneName=dyn_coat_05_rt_01}
       {boneName=dyn_coat_05_lf_02}
       {boneName=dyn_coat_05_rt_02}
      ]}]}
   Save cached pose 'StartPose'(Pose=@n27) {cacheName=StartPose}
   n28 = Control Rig!(
    Source=Use cached pose 'StartPose'()
    Alpha
    IsAiming:bool=Get bIsTargeting()
    IsMeelee:bool=Get bIsPlayingMeleeAnim()
    IsCrouchTargeting:bool=Get IsCrouchTargeting() -> (IsCrouchTargeting:bool)
    Is Skydiving:bool=Get bIsSkyDiving()
    Is Standing:bool=NOT Boolean(A=Get bIsCrouching())
   ) -> (Pose) {controlRigAssetReference {blueprintRigClass=&/M_MED_ChoiceCrave_FaceAcc_CtrlRig.M_MED_ChoiceCrave_FaceAcc_CtrlRig_C}}
   n29 = "Constraint - base_jewel_cr (PPPPPPPP)"!(
    ComponentPose=Local To Component(LocalPose=@n28)
    Alpha=1
    ConstraintWeights_0=0.4408
    ConstraintWeights_1:float=0.225
    ConstraintWeights_2:float=0.2249
    ConstraintWeights_3:float=0.0436
    ConstraintWeights_4:float=0.0435
    ConstraintWeights_5:float=0.0108
    ConstraintWeights_6:float=0.0107
    ConstraintWeights_7:float=0.0007
   ) -> (Pose) {
    boneToModify {boneName=base_jewel_cr}
    constraintSetup [
     {targetBone {boneName=head}}
     {targetBone {boneName=dyn_hair_01_lf}}
     {targetBone {boneName=dyn_hair_01_rt}}
     {targetBone {boneName=dyn_hair_02_lf}}
     {targetBone {boneName=dyn_hair_02_rt}}
     {targetBone {boneName=dyn_hair_03_lf}}
     {targetBone {boneName=dyn_hair_03_rt}}
     {targetBone {boneName=dyn_hair_04_lf}}
    ]
    constraintWeights [1,1,1,1,1,1,1,1]
   }
   n30 = "Constraint - aimtgt_jewel_cr (PPPPPPPP)"!(
    ComponentPose=@n29
    Alpha=1
    ConstraintWeights_0=0.3658
    ConstraintWeights_1:float=0.2187
    ConstraintWeights_2:float=0.1864
    ConstraintWeights_3:float=0.1344
    ConstraintWeights_4:float=0.068
    ConstraintWeights_5:float=0.0172
    ConstraintWeights_6:float=0.0079
    ConstraintWeights_7:float=0.0017
   ) -> (Pose) {
    boneToModify {boneName=aimtgt_jewel_cr}
    constraintSetup [
     {targetBone {boneName=aim_jewel_cr}}
     {targetBone {boneName=dyn_hair_02_rt}}
     {targetBone {boneName=dyn_hair_01_rt}}
     {targetBone {boneName=head}}
     {targetBone {boneName=dyn_hair_03_rt}}
     {targetBone {boneName=dyn_hair_04_rt}}
     {targetBone {boneName=dyn_hair_01_lf}}
     {targetBone {boneName=dyn_hair_02_lf}}
    ]
    constraintWeights [1,1,1,1,1,1,1,1]
   }
   n31 = "Constraint - dyn_hairTopMain_rt (PP)"!(ComponentPose=@n30, Alpha=0.75, ConstraintWeights_0=0.6249, ConstraintWeights_1:float=0.3751) -> (Pose) {
    boneToModify {boneName=dyn_hairTopMain_rt}
    constraintSetup [{targetBone {boneName=head}}, {targetBone {boneName=dyn_hairTopMain_rt}}]
    constraintWeights [1,1]
   }
   n32 = "Constraint - dyn_hairTopMain_lf (PPPPPPPP)"!(
    ComponentPose=@n31
    Alpha=0.5
    ConstraintWeights_0=0.8324
    ConstraintWeights_1:float=0.127
    ConstraintWeights_2:float=0.0199
    ConstraintWeights_3:float=0.0147
    ConstraintWeights_4:float=0.0037
    ConstraintWeights_5:float=0.0013
    ConstraintWeights_6:float=0.0007
    ConstraintWeights_7:float=0.0003
   ) -> (Pose) {
    boneToModify {boneName=dyn_hairTopMain_lf}
    constraintSetup [
     {targetBone {boneName=head}}
     {targetBone {boneName=dyn_hair_01_lf}}
     {targetBone {boneName=dyn_hair_01_rt}}
     {targetBone {boneName=dyn_hair_02_lf}}
     {targetBone {boneName=dyn_hair_03_lf}}
     {targetBone {boneName=dyn_hair_02_rt}}
     {targetBone {boneName=dyn_hair_04_lf}}
     {targetBone {boneName=dyn_hair_03_rt}}
    ]
    constraintWeights [1,1,1,1,1,1,1,1]
   }
   n33 = "Constraint - base_faceFrame01_rt (RR)"!(ComponentPose=@n32, Alpha=0, ConstraintWeights_0=0.25, ConstraintWeights_1:float=0.75) -> (Pose) {
    boneToModify {boneName=base_faceFrame01_rt}
    constraintSetup [{transformType=Rotation, perAxis {bX=false}, targetBone {boneName=dyn_hair_02_rt}}, {transformType=Rotation, targetBone {boneName=head}}]
    constraintWeights [1,1]
   }
   n34 = "Constraint - base_faceFrame01_lf (RR)"!(ComponentPose=@n33, Alpha=0, ConstraintWeights_0=0.5, ConstraintWeights_1:float=0.5) -> (Pose) {
    boneToModify {boneName=base_faceFrame01_lf}
    constraintSetup [
     {transformType=Rotation, perAxis {bX=false, bZ=false}, targetBone {boneName=dyn_hair_02_lf}}
     {transformType=Rotation, targetBone {boneName=head}}
    ]
    constraintWeights [1,1]
   }
   n35 = "Transform (Modify) Bone - Bone: base_jewel_cr"(ComponentPose=@n34, Alpha=Get HairClipCrouch() -> (HairClipCrouch:double), Translation="-0.5,0,0", Rotation="0,0,0", Scale="1,1,1") {
    rotationMode=BMM_Additive
    rotationSpace=BCS_BoneSpace
    translationMode=BMM_Additive
    translationSpace=BCS_BoneSpace
    boneToModify {boneName=base_jewel_cr}
    translation {x=-0.5}
   }
   Save cached pose 'RigidBodyPose'(Pose=Component To Local(ComponentPose=@n35)) {cacheName=RigidBodyPose}
   n36 = RigidBody!(
    ComponentPose=Local To Component(LocalPose=Use cached pose 'RigidBodyPose'())
    Alpha
    OverrideWorldGravity:Vector="0,0,0"
    ExternalForce
    SimSpaceSettings:Sim Space Settings Structure=(WorldAlpha=1,VelocityScaleZ=1,DampingAlpha=1,MaxLinearVelocity=10000,MaxAngularVelocity=10000,MaxLinearAcceleration=10000,MaxAngularAcceleration=10000,ExternalLinearDragV=(X=0,Y=3,Z=0),ExternalLinearVelocity=(X=0,Y=0,Z=0),ExternalAngularVelocity=(X=0,Y=0,Z=0))
   ) -> (Pose) {
    bOverrideWorldGravity=true
    overridePhysicsAsset=&/M_MED_ChoiceCrave_FaceAcc_Physics
    simulationSpace=BaseBoneSpace
    baseBoneRef {boneName=neck_02}
    simSpaceSettings {worldAlpha=1, externalLinearDragV {y=3}}
   }
   n37 = "Transform (Modify) Bone - Bone: dyn_hair_04_lf"(
    ComponentPose=Local To Component(LocalPose=Use cached pose 'RigidBodyPose'())
    Alpha=1
    Translation="0,0,0"
    Rotation="-10,-10,0"
    Scale="1,1,1"
   ) {rotationMode=BMM_Additive, rotationSpace=BCS_BoneSpace, boneToModify {boneName=dyn_hair_04_lf}, rotation {pitch=-10, yaw=-10}}
   n38 = "Transform (Modify) Bone - Bone: dyn_hair_03_rt"(ComponentPose=@n37, Alpha=1, Translation="0,0,0", Rotation="3,7,0", Scale="1,1,1") {rotationMode=BMM_Additive, rotationSpace=BCS_BoneSpace, boneToModify {boneName=dyn_hair_03_rt}, rotation {pitch=3, yaw=7}}
   n39 = "Transform (Modify) Bone - Bone: dyn_hair_04_rt"(ComponentPose=@n38, Alpha=1, Translation="0,0,0", Rotation="-10,8,0", Scale="1,1,1") {rotationMode=BMM_Additive, rotationSpace=BCS_BoneSpace, boneToModify {boneName=dyn_hair_04_rt}, rotation {pitch=-10, yaw=8}}
   n40 = "Constraint - dyn_hair_02_rt (P)"(ComponentPose=@n39, Alpha=1, ConstraintWeights_0=1) {boneToModify {boneName=dyn_hair_02_rt}, constraintSetup [{targetBone {boneName=spine_05}}], constraintWeights [1]}
   n41 = "Constraint - dyn_hair_02_lf (P)"(ComponentPose=@n40, Alpha=1, ConstraintWeights_0=1) {boneToModify {boneName=dyn_hair_02_lf}, constraintSetup [{targetBone {boneName=spine_05}}], constraintWeights [1]}
   n42 = Get Cape Con Alpha() -> (Cape Con Alpha:double)
   n43 = "Transform (Modify) Bone - Bone: dyn_hair_06_lf"(ComponentPose=@n41, Alpha=@n42, Translation="0,-4,-3", Rotation="0,0,0", Scale="1,1,1") {
    rotationSpace=BCS_BoneSpace
    scaleSpace=BCS_BoneSpace
    translationMode=BMM_Additive
    translationSpace=BCS_BoneSpace
    boneToModify {boneName=dyn_hair_06_lf}
    translation {y=-4, z=-3}
   }
   n44 = "Transform (Modify) Bone - Bone: dyn_hair_06_rt"(ComponentPose=@n43, Alpha=@n42, Translation="0,3,-3", Rotation="0,0,0", Scale="1,1,1") {
    rotationSpace=BCS_BoneSpace
    translationMode=BMM_Additive
    translationSpace=BCS_BoneSpace
    boneToModify {boneName=dyn_hair_06_rt}
    translation {y=3, z=-3}
   }
   n45 = Get Cape Crouch Tranform Alpha() -> (Cape Crouch Tranform Alpha:double)
   n46 = "Transform (Modify) Bone - Bone: dyn_hair_05_lf"(ComponentPose=@n44, Alpha=@n45, Translation="0,0,0", Rotation="0,-15,0", Scale="1,1,1") {rotationMode=BMM_Additive, rotationSpace=BCS_ParentBoneSpace, boneToModify {boneName=dyn_hair_05_lf}, rotation {yaw=-15}}
   n47 = "Transform (Modify) Bone - Bone: dyn_hair_05_rt"(ComponentPose=@n46, Alpha=@n45, Translation="0,0,0", Rotation="0,15,0", Scale="1,1,1") {
    rotationMode=BMM_Additive
    rotationSpace=BCS_ParentBoneSpace
    translationMode=BMM_Additive
    translationSpace=BCS_BoneSpace
    boneToModify {boneName=dyn_hair_05_rt}
    rotation {yaw=15}
   }
   n48 = Get Cape Con Alpha()
   n49 = "Constraint - dyn_hair_04_lf (R)"(ComponentPose=@n47, Alpha=@n48, ConstraintWeights_0=1) {boneToModify {boneName=dyn_hair_04_lf}, constraintSetup [{transformType=Rotation, targetBone {boneName=spine_01}}], constraintWeights [1]}
   n50 = "Constraint - dyn_hair_04_rt (R)"(ComponentPose=@n49, Alpha=@n48, ConstraintWeights_0=1) {boneToModify {boneName=dyn_hair_04_rt}, constraintSetup [{transformType=Rotation, targetBone {boneName=spine_01}}], constraintWeights [1]}
   n51 = "Constraint - dyn_hair_06_lf (R)"(ComponentPose=@n50, Alpha=@n42, ConstraintWeights_0=1) {boneToModify {boneName=dyn_hair_06_lf}, constraintSetup [{transformType=Rotation, targetBone {boneName=spine_01}}], constraintWeights [1]}
   n52 = "Constraint - dyn_hair_06_rt (R)"(ComponentPose=@n51, Alpha=@n42, ConstraintWeights_0=1) {boneToModify {boneName=dyn_hair_06_rt}, constraintSetup [{transformType=Rotation, targetBone {boneName=spine_01}}], constraintWeights [1]}
   n53 = "Transform (Modify) Bone - Bone: dyn_hairExtra_06_lf"(ComponentPose=@n52, Alpha=@n48, Translation="0,-12,0", Rotation="0,0,0", Scale="1,1,1") {
    rotationMode=BMM_Additive
    rotationSpace=BCS_ParentBoneSpace
    translationMode=BMM_Additive
    translationSpace=BCS_BoneSpace
    boneToModify {boneName=dyn_hairExtra_06_lf}
    translation {y=-12}
   }
   n54 = "Transform (Modify) Bone - Bone: dyn_hairExtra_06_rt"(ComponentPose=@n53, Alpha=@n48, Translation="0,12,0", Rotation="0,0,0", Scale="1,1,1") {
    rotationMode=BMM_Additive
    rotationSpace=BCS_ParentBoneSpace
    translationMode=BMM_Additive
    translationSpace=BCS_BoneSpace
    boneToModify {boneName=dyn_hairExtra_06_rt}
    translation {y=12}
   }
   n55 = "Transform (Modify) Bone - Bone: dyn_hairExtra_05_lf"(ComponentPose=@n54, Alpha=@n48, Translation="0,0,0", Rotation="8,-18,0", Scale="1,1,1") {
    rotationMode=BMM_Additive
    rotationSpace=BCS_ParentBoneSpace
    translationMode=BMM_Additive
    translationSpace=BCS_BoneSpace
    boneToModify {boneName=dyn_hairExtra_05_lf}
    rotation {pitch=8, yaw=-18}
   }
   n56 = "Transform (Modify) Bone - Bone: dyn_hairExtra_05_rt"(ComponentPose=@n55, Alpha=@n48, Translation="0,0,0", Rotation="5,12,0", Scale="1,1,1") {
    rotationMode=BMM_Additive
    rotationSpace=BCS_BoneSpace
    translationSpace=BCS_BoneSpace
    boneToModify {boneName=dyn_hairExtra_05_rt}
    rotation {pitch=5, yaw=12}
   }
   n57 = RigidBody(
    ComponentPose=@n56
    Alpha=Get Cape Dyn Alpha()
    OverrideWorldGravity="0,0,0"
    SimSpaceSettings=(WorldAlpha=1,VelocityScaleZ=1,DampingAlpha=1,MaxLinearVelocity=10000,MaxAngularVelocity=10000,MaxLinearAcceleration=10000,MaxAngularAcceleration=10000,ExternalLinearDragV=(X=0,Y=3,Z=0),ExternalLinearVelocity=(X=0,Y=0,Z=0),ExternalAngularVelocity=(X=0,Y=0,Z=0))
   ) {
    bOverrideWorldGravity=true
    overridePhysicsAsset=&/M_MED_ChoiceCrave_FaceAcc_BPPhysics
    simulationSpace=BaseBoneSpace
    baseBoneRef {boneName=neck_02}
    simSpaceSettings {worldAlpha=1, externalLinearDragV {y=3}}
   }
   Save cached pose 'PostPhysBody'(Pose=Blend Poses by bool(
     BlendPose_0=Component To Local(ComponentPose=@n57)
     BlendPose_1=Component To Local(ComponentPose=@n36)
     bActiveValue=Get is Wearing cape and backpack() -> (is Wearing cape and backpack:bool)
    ) {blendType=Linear}) {cacheName=PostPhysBody}
  ]}
 EventGraph {nodes [
   Event BlueprintUpdateAnimation()
   > n1 = "Gravity Override State Function (Parameter Asset) (Nativized)"(
    CharacterPartAnimInstance=Self-Reference()
    GameTimeInSeconds=GetGameTimeInSeconds()
    Parameters=&/M_MED_ChoiceCrave_FaceAcc_Parameters
   )
   > Set GravityOverride(GravityOverride=@n1)
   > Set RBAN Settings Out(RBAN Settings Out=@n1.OutRigidBodyAnimNodeParameters)
   n2 = Get bIsCrouching()
   > Set HairClipCrouch(HairClipCrouch:double="To Float (Boolean)"(InBool=@n2)) -> (Output_Get:double)
   n3 = Get is Wearing cape and backpack()
   > Set IsCrouchTargeting(IsCrouchTargeting:bool=AND Boolean!(A=@n2, B=Get bIsTargeting(), C:bool=NOT Boolean(A=@n3)) -> (ReturnValue)) -> (Output_Get:bool)
   > Set Cape Crouch Tranform Alpha(Cape Crouch Tranform Alpha:double=float * float(A:bool=@n2, B:bool=Get Is Wearing Cape() -> (Is Wearing Cape:bool)) {is=title K2Node_PromotableOperator}) -> (Output_Get:double)
   n4 = GetOwnerAnimationInterface!(self:Custom Character Part Anim Instance Object Reference=Self-Reference()) -> (ReturnValue:Fort Player Animation Interface Object Reference)
   > Is Valid(InputObject=@n4)
   > n5 = Set isZipline(isZipline:bool=Break Fort Anim Input Zipline!(FortAnimInput_Zipline:"Fort Anim Input Zipline Structure (by ref)"=GetZiplineInput!(self:Fort Player Animation Interface Object Reference=@n4) -> (ReturnValue:"Fort Anim Input Zipline Structure (by ref)")) -> (
     bIsZiplining:bool=~
     bIsOnAscender:bool=~
     bShouldPlayPivotTransition:bool=~
     LeanYaw:float=~
     PivotBlendDelayRemaining:float=~
     PivotBlendDelay:float=~
     PivotCardinalDirection:EFortCardinalDirection Enum
     WorldVelocityLastTick:Vector=~
    ) {is=title K2Node_BreakStruct}) -> (Output_Get:bool)
   > n6 = Branch(Condition=@n5)
   > Set animDyn on zipline(animDyn on zipline:Vector="0, 0, 0") -> (Output_Get:Vector)
   Set animDyn on zipline(execute=@n6.else, animDyn on zipline="0.3,0.3,0.3")
   Event OnCharacterCustomizationLoaded()
   > n7 = Cast To PlayerPawn_Athena(Object=TryGetPawnOwner())
   > Set Body_Mesh(Body_Mesh=GetSkeletalMeshForPartType(self=@n7))
   n8 = CheckForBackbling()
   > n9 = Set isWearingBackbling(isWearingBackbling=AND Boolean(A=@n8, B=NOT Boolean(A=@n8.bIsBackblingACape), C=NOT Boolean(A=@n8.bIsBackblingLong)))
   > n10 = Set Is Wearing Cape(Is Wearing Cape=OR Boolean(A=@n8.bIsBackblingACape, B=@n8.bIsBackblingLong))
   > Set is Wearing cape and backpack(is Wearing cape and backpack:bool=OR Boolean(A=@n9, B=@n10)) -> (Output_Get:bool)
   > n11 = Branch(Condition=Get Is Wearing Cape())
   > Set Cape Dyn Alpha(Cape Dyn Alpha:double=0) -> (Output_Get:double)
   > Set Cape Con Alpha(Cape Con Alpha:double=0.8) -> (Output_Get:double)
   Set Cape Dyn Alpha(execute=@n11.else, Cape Dyn Alpha=1)
   > Set Cape Con Alpha(Cape Con Alpha=0)
  ]}
]
variables [
 animDyn on zipline {type=Vector, default=<0.3 0.3 0.3>}
 Body_Mesh {type=SkeletalMeshComponent}
 Cape Con Alpha {type=number}
 Cape Crouch Tranform Alpha {type=number}
 Cape Dyn Alpha {type=number, default=1}
 GravityOverride {type=Vector}
 HairClipCrouch {type=number}
 HairCrouchAimFix {type=number}
 Is Wearing Cape {type=boolean}
 is Wearing cape and backpack {type=boolean}
 IsCrouchTargeting {type=boolean}
 IsCrouchTargetMoving {type=boolean}
 isWearingBackbling {type=boolean}
 isZipline {type=boolean}
 RBAN Settings Out {type=FortRigidBodyAnimNodeParameters}
 Source Mesh Component {type=SkeletalMeshComponent}
]
